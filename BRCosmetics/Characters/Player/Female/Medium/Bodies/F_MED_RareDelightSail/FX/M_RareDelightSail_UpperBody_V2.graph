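# Unlit translucent Niagara ribbon material for upper-body cosmetic FX on RareDelightSail, drawing waving, color-cycling trails.
# Emissive multiplies a time-animated hue shift by a two-color gradient with blurred wave texture, scaled by particle color and dynamic brightness.
# Opacity combines a panning trail texture, sphere-mask edge fades, depth fade and particle alpha.
# World position offset is a tapered travelling sine wave, scaled by event audio amplitude and a dynamic parameter.
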Material /BRCosmetics/Characters/Player/Female/Medium/Bodies/F_MED_RareDelightSail/FX/M_RareDelightSail_UpperBody_V2
graph {
 properties {bUsedWithNiagaraRibbons=true, blendMode=BLEND_Translucent, pixelDepthOffsetMode=PDOM_Legacy, shadingModel=MSM_Unlit, usedShadingModels=Unlit}
 nodes [
  Constant2Vector() {g=0.2}
  n1 = Multiply(
   A=Time()
   B=LinearInterpolate(
    B=DynamicParameter() {defaultValue=<rgba 1 1 5 1>, paramNames [WPO, Bright Intensity, Use custom speed, Param4]}.Use custom speed
    Alpha=ScalarParameter() {parameterName=Use Custom speed}
   ) {constA=5}
  ) {constB=5}
  n2 = OneMinus(None=ComponentMask(None=TextureCoordinate()) {g=true})
  n3 = MaterialFunctionCall(Hue Shift Percentage=LinearInterpolate(
    A=ScalarParameter() {parameterName=HueShift RangeMin}
    B=ScalarParameter() {defaultValue=0.5, parameterName=HueShift RangeMax}
    Alpha=Add(A=@n1, B=@n2)
   ) {constB=0.5}) {materialFunction=&HueShift}
  n4 = Desaturation(
   None=Multiply(A=VectorParameter() {parameterName=HueShift Color, defaultValue=<rgba 8 1.1662 5.4451 1>}.RGB, B=Multiply(A=@n3.Result))
   Fraction=ScalarParameter() {parameterName=HueShift Des}
  )
  n5 = DynamicParameter() {defaultValue=<rgba 1 1 5 1>, paramNames [WPO, Bright Intensity, Use custom speed, Param4]}
  n6 = ParticleColor()
  n7 = Multiply(A=@n6.RGB, B=ScalarParameter() {defaultValue=0.2, parameterName=Bright_Intensity}) {constB=0.2}
  n8 = VectorParameter() {parameterName=Lerp Color1, defaultValue=<rgba 1.6041 0.403 2 1>}
  n9 = VectorParameter() {parameterName=Lerp Color2, defaultValue=<rgba 1 0.7133 0.277 1>}
  n10 = Power(Base=ComponentMask(None=TextureCoordinate()) {g=true}) {constExponent=1}
  n11 = Multiply(A=Desaturation(None=LinearInterpolate(A=@n8.RGB, B=@n9.RGB, Alpha=@n10), Fraction=Constant() {r=-0.2})) {constB=4}
  n12 = TextureSample(UVs=MaterialFunctionCall(XY=TextureCoordinate() {vTiling=0.3}) {materialFunction=&Swizzle}.YX) {texture=/Game/Athena/Environments/Nexus/Textures/T_ZeroPoint_MagneticWaves_Blurred}
  n13 = Min(
   A=Multiply(A=VectorParameter() {parameterName=Color, defaultValue=<rgba 0 11.7044 30 1>}.RGB, B=Power(Base=@n12.R) {constExponent=1})
   B=@n11
  )
  n14 = Constant() {r=-1}
  n15 = Add(B=Add(A=@n11, B=Desaturation(None=@n13, Fraction=@n14)))
  MaterialFunctionCall(Emissive=Multiply(A=Multiply(A=@n4, B=@n5.Bright Intensity), B=Multiply(A=@n7, B=@n15))) {materialFunction=&MF_FX_EmissiveAdjustments}
  n16 = VectorParameter() {parameterName=UV_Scale, defaultValue=<rgba 1 0.7 0 0>}
  n17 = Add(A=Multiply(A=TextureCoordinate(), B=ComponentMask(None=@n16.RGB) {g=true, r=true})) {constB=0}
  n18 = ComponentMask(None=VectorParameter() {parameterName=UV_Offset, defaultValue=<rgba 0 0.05 0 0>}.RGB) {g=true, r=true}
  n19 = VectorParameter() {parameterName=Panner speed, defaultValue=<rgba -0.6 0 0 0>}
  n20 = TextureSampleParameter2D(UVs=Panner(Coordinate=Add(A=@n17, B=@n18), Speed=ComponentMask(None=@n19.RGB) {g=true, r=true}) {speedX=-0.3}) {
   parameterName=Tex_Default
   samplerType=SAMPLERTYPE_LinearColor
   texture=/BRCosmetics/Athena/Effects/Particles/Skydiving/NS_Contrail_CubeNinja/T_CubeNinja_Trail
  }
  n21 = TextureCoordinate() {uTiling=0}
  n22 = Constant2Vector() {g=0.5}
  n23 = Multiply(A=@n20.R, B=Power(Base=SphereMask(A=@n21, B=@n22) {attenuationRadius=0.5, hardnessPercent=0}) {constExponent=4})
  n24 = TextureCoordinate() {vTiling=0}
  n25 = Constant2Vector() {r=0.5}
  n26 = Saturate(None=Multiply(A=SphereMask(A=@n24, B=@n25) {attenuationRadius=0.5, hardnessPercent=0}) {constB=5})
  Multiply(A=@n6.A, B=DepthFade(Opacity=Multiply(A=@n23, B=@n26)) {fadeDistanceDefault=20})
  n27 = VectorParameter() {parameterName=WPO UV Scale, defaultValue=<rgba 3.5 1 0 0>}
  n28 = Add(A=Multiply(A=TextureCoordinate(), B=ComponentMask(None=@n27.RGB) {g=true, r=true})) {constB=0}
  n29 = Add(
   A=Multiply(A=ComponentMask(None=@n28) {r=true})
   B=Multiply(A=Time(), B=ScalarParameter() {defaultValue=1.5, parameterName=Speed}) {constB=2}
  )
  n30 = CollectionParameter() {collection=&FortniteMaterialParameters, parameterName=EventAudio_Amplitude}
  n31 = Multiply(A=LinearInterpolate(Alpha=Sine(None=@n29)) {constA=-0.2}, B=Clamp(None=@n30) {maxDefault=3, minDefault=0.5})
  n32 = TextureCoordinate()
  n33 = Power(Base=Sine(None=ComponentMask(None=@n32) {r=true}) {period=2})
  n34 = Multiply(A=Multiply(A=Multiply(A=@n31, B=@n33)) {constB=0.3}, B=VectorParameter() {defaultValue=<rgba 0 10 0 1>}.RGB)
  n35 = DynamicParameter() {defaultValue=<rgba 1 1 5 1>, paramNames [WPO, Bright Intensity, Use custom speed, Param4]}
  Reroute(None=Transform(None=Multiply(A=@n34, B=@n35.WPO)) {transformType=TRANSFORM_Instance})
 ]
}
