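# Niagara system for the RareDelightSail female body cosmetic, drawing music-reactive flowing ribbons along the arms and shoulder plus drifting flare sprites.
# FN_BoneFlow bends burst-spawned particles around skeletal mesh bones into layered ribbons with curve-shaped radius and width.
# System music, time-of-day and LOD-distance values drive color, opacity and material parameters, with front-end overrides.
# GPU flares spawn from the mesh, stopping during skydiving, in vehicles or by owner speed.
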NiagaraSystem /BRCosmetics/Characters/Player/Female/Medium/Bodies/F_MED_RareDelightSail/FX/NS_RareDelightSail_Skin
dependencies {
 usedDataInterfaces [
  &NiagaraDataInterfaceCurve
  &NiagaraDataInterfaceCamera
  &NiagaraDataInterfacePlatformSet
  &NiagaraDataInterfaceVector2DCurve
  &NiagaraDataInterfaceSkeletalMesh
 ]
 usedDynamicInputs [&FNTimeOfDayScalar, &Waveform, &Power]
 usedModules [
  &self:SystemSpawnScript.SystemScriptSource.SystemScriptGraph.NiagaraNodeAssignment_4.SetVariables_#1
  &self:SystemSpawnScript.SystemScriptSource.SystemScriptGraph.NiagaraNodeAssignment_5.SetVariables_#2
  &self:SystemSpawnScript.SystemScriptSource.SystemScriptGraph.NiagaraNodeAssignment_6.SetVariables_#3
  &self:SystemSpawnScript.SystemScriptSource.SystemScriptGraph.NiagaraNodeAssignment_1.SetVariables_#4
  &SystemState
  &self:SystemSpawnScript.SystemScriptSource.SystemScriptGraph.NiagaraNodeAssignment_0.SetVariables_#5
  &self:SystemSpawnScript.SystemScriptSource.SystemScriptGraph.NiagaraNodeAssignment_3.SetVariables_#6
  &self:SystemSpawnScript.SystemScriptSource.SystemScriptGraph.NiagaraNodeAssignment_8.SetVariables_#7
  &self:SystemSpawnScript.SystemScriptSource.SystemScriptGraph.NiagaraNodeAssignment_2.SetVariables_#8
  &self:SystemSpawnScript.SystemScriptSource.SystemScriptGraph.NiagaraNodeAssignment_7.SetVariables_#9
  &self:SystemSpawnScript.SystemScriptSource.SystemScriptGraph.NiagaraNodeAssignment_10.SetVariables_#10
  &self:SystemSpawnScript.SystemScriptSource.SystemScriptGraph.NiagaraNodeAssignment_9.SetVariables_#11
  &InterpolateOverTime
 ]
}
emitters {
 Bone_Flow_LeftArm {
  emitterUpdateScript [
   &EmitterState(
    Inactive Response=::NewEnumerator1
    Life Cycle Mode=::NewEnumerator1
    Loop Behavior=::NewEnumerator1
    Loop Duration Mode=::NewEnumerator1
    UseLoopDelay=false
   )
   &SpawnBurst_Instantaneous(Spawn Count=10, Use Spawn Probability=false, UseLoopCountLimit=false)
  ]
  particleSpawnScript [
   &FN_SetSpawnGroup(Inclusivity=true, Normalize Spawn Group=true, Particles Per Group=10, Pattern=::NewEnumerator0, Set Ribbon ID and Ribbon Link Order=true)
   &InitializeParticle(
    Color=<rgba 1 1 1 1.5>
    Ribbon Width=12
    Ribbon Width Mode=::NewEnumerator1
    Sprite Size Mode=::NewEnumerator3
    Uniform Sprite Size=1
    UsePositionOffset=false
    Write Material Random=true
   )
   &CameraOffset(Camera Offset Amount=30)
  ]
  particleUpdateScript [
   /SetVariables(Particles.Ribbon width02:NiagaraFloat=&Multiply_Float(
     A=20.4
     B=&FloatFromCurve(
      FloatCurve=DataInterface() {propertyValues {Curve {keys [
          {interpMode=RCIM_Cubic, tangentMode=RCTM_User, time=0, value=0.3}
          {interpMode=RCIM_Cubic, tangentMode=RCTM_User, time=0.3, value=1}
          {arriveTangent=-1, interpMode=RCIM_Cubic, leaveTangent=-1, time=0.7, value=1}
          {tangentMode=RCTM_User, time=1, value=0.3}
         ]}}}
      CurveIndex=$Particles.Normalized Spawn Group
     )
    )) {isSetParameters=true, module=SetVariables_#12}
   /SetVariables(Particles.Ribbon width03:NiagaraFloat=&Multiply_Float(
     A=13.2
     B=&FloatFromCurve(
      FloatCurve=DataInterface() {propertyValues {Curve {keys [
          {interpMode=RCIM_Cubic, tangentMode=RCTM_User, time=0, value=0.3}
          {interpMode=RCIM_Cubic, tangentMode=RCTM_User, time=0.3, value=1}
          {arriveTangent=-1, interpMode=RCIM_Cubic, leaveTangent=-1, time=0.7, value=1}
          {tangentMode=RCTM_User, time=1, value=0.3}
         ]}}}
      CurveIndex=$Particles.Normalized Spawn Group
     )
    )) {isSetParameters=true, module=SetVariables_#13}
   &ParticleState(Kill Particles When Lifetime Has Elapsed=false, Let Infinitely Lived Particles Die When Emitter Deactivates=true)
   &FN_BoneFlow(
    Angle:NiagaraFloat=&Multiply_Float(A=$Particles.Normalized Spawn Group, B=&FN_FrontEndFloatModifier(Gameplay=0.5, Front End=0.8))
    BoneID:NiagaraInt32=&ModuloInt(A=$Particles.SpawnGroup, B=1)
    Height:NiagaraFloat=&Multiply_Float(A=$Particles.Normalized Spawn Group, B=1)
    Radius:NiagaraFloat=&FloatFromCurve(
     FloatCurve=DataInterface() {propertyValues {Curve {keys [
         {interpMode=RCIM_Cubic, tangentMode=RCTM_User, time=0, value=0.4}
         {interpMode=RCIM_Cubic, tangentMode=RCTM_User, time=0.5, value=1}
         {tangentMode=RCTM_User, time=1, value=0.4}
        ]}}}
     CurveIndex=$Particles.Normalized Spawn Group
     Scale Curve=10
    )
    SkeletalMesh:NiagaraDataInterfaceSkeletalMesh=$System.Skeletal Mesh L
   )
   &ScaleRibbonWidth(Ribbon Width Scale=&FloatFromCurve(
     FloatCurve=DataInterface() {propertyValues {Curve {keys [
         {interpMode=RCIM_Cubic, tangentMode=RCTM_User, time=0, value=0.5}
         {interpMode=RCIM_Cubic, tangentMode=RCTM_User, time=0.3, value=1}
         {arriveTangent=-0.7143, interpMode=RCIM_Cubic, leaveTangent=-0.7143, time=0.7, value=1}
         {tangentMode=RCTM_User, time=1, value=0.5}
        ]}}}
     CurveIndex=$Particles.Normalized Spawn Group
    ))
   /SetVariables(Particles.Ribbon02 Color:LinearColor=&MakeLinearColorFromVectorAndFloat(
     "Vector (RGB)"=&MakeVectorFromLinearColorRGB(Color=&MultiplyLinearColor(
       Linear Color=<rgba 0.5 0.5 0.5 1>
       Scale Factor=&MakeVector4_FromFloat(Float=&FN_FrontEndFloatModifier(Gameplay=$System.TOD, Front End=$System.Frontend Bright NoMusic))
      ))
     "Float (Alpha)"=&FN_FrontEndFloatModifier(
      Gameplay=&Multiply_Float(
       A=&Multiply_Float(A=3, B=&Lerp_Float(A=0, B=1, Alpha=&ClampFloat(Float=$System.Music_Reactive_preview)))
       B=&Multiply_Float(
        A=$User.Opacity
        B=&FN_LOD_Distance_Remap(
         Start Distance=200
         End Distance=300
         Minimum Output=&Lerp_Float(A=0.1, B=1, Alpha=$System.LOD_Fade_preview)
         Maximum Output=1
         Clamp Output=true
        )
       )
      )
      Front End=$System.Frontend Alpha NoMusic
     )
    )) {isSetParameters=true, module=SetVariables_#14}
   /SetVariables(Particles.Position=&AddVectorToPosition(
     Position=$Particles.Position
     Vector=&Multiply_VectorByFloat(
      Vector=&NormalizeVector(
       Vector To Normalize=&SubtractVectorFromPosition(
        Position=&ReturnCameraPosition(Coordinate Space:ENiagaraCoordinateSpace=::Simulation, Camera Query:NiagaraDataInterfaceCamera=DataInterface())
        Vector=Linked() {linkedVariable {name=Particles.Position, type=NiagaraPosition}}
       )
       Safe Normalize=true
       Safe Fallback Threshold=0
       Safe Fallback Vector=<0 0 0.001>
      )
      Float=0
     )
    )) {isSetParameters=true, module=SetVariables_#15}
   &DynamicMaterialParameters(
    Index 0 Param 1=$System.WPO
    Index 0 Param 2=&Lerp_Float(
     A=1
     B=&FN_FrontEndFloatModifier(Gameplay=1.5, Front End=1)
     Alpha=$System.Music_Reactive_preview
    )
    Module.Index0Param0Write=true
    Module.Index0Param1Write=true
    Module.Index0Param2Write=true
    Module.Index0Param3Write=true
   )
   &ScaleColor(
    Scale Alpha=&FN_FrontEndFloatModifier(
     Gameplay=&Multiply_Float(
      A=&Lerp_Float(A=0, B=1, Alpha=$System.InterpolateOverTime.InterpolateOverTime_Float.Moving Average)
      B=&Multiply_Float(
       A=$User.Opacity
       B=&FN_LOD_Distance_Remap(
        Start Distance=200
        End Distance=300
        Minimum Output=&Lerp_Float(A=0.01, B=1, Alpha=$System.LOD_Fade_preview)
        Maximum Output=1
        Clamp Output=true
       )
      )
     )
     Front End=1
    )
    Scale RGB=&Multiply_VectorByFloat(
     Vector=&Multiply_VectorByFloat(
      Vector=<1 1 1>
      Float=&FN_PickFloatDependOnPlatform(TargetPlatformActive=1, TargetPlatformNoActive=0.75)
     )
     Float=&FN_FrontEndFloatModifier(Gameplay=$System.TOD, Front End=1)
    )
    ScaleA=true
    ScaleRGB=true
   )
  ]
  rendererClasses [&NiagaraRibbonRendererProperties]
  renderers [
   {rendererClass=&NiagaraRibbonRendererProperties}
   {rendererClass=&NiagaraRibbonRendererProperties, rendererIndex=1}
   {rendererClass=&NiagaraRibbonRendererProperties, rendererIndex=2}
  ]
 }
 Bone_Flow_LeftArm001 {
  bEnabled=false
  emitterUpdateScript [
   &EmitterState(
    Inactive Response=::NewEnumerator1
    Life Cycle Mode=::NewEnumerator1
    Loop Behavior=::NewEnumerator1
    Loop Duration Mode=::NewEnumerator1
    UseLoopDelay=false
   )
   &SpawnBurst_Instantaneous(
    Spawn Count=&MakeCustomIntFromBool(Bool=$User.bIsFrontEndPreview, True INT=10, False INT=0)
    Use Spawn Probability=false
    UseLoopCountLimit=false
   )
  ]
  particleSpawnScript [
   &FN_SetSpawnGroup(Inclusivity=true, Normalize Spawn Group=true, Particles Per Group=10, Pattern=::NewEnumerator0, Set Ribbon ID and Ribbon Link Order=true)
   &InitializeParticle(
    Color=<rgba 1 1 1 1.5>
    Ribbon Width=12
    Ribbon Width Mode=::NewEnumerator1
    Sprite Size Mode=::NewEnumerator3
    Uniform Sprite Size=1
    UsePositionOffset=false
    Write Material Random=true
   )
   &CameraOffset(Camera Offset Amount=30)
  ]
  particleUpdateScript [
   /SetVariables(Particles.Ribbon width02:NiagaraFloat=&Multiply_Float(
     A=20.4
     B=&FloatFromCurve(
      FloatCurve=DataInterface() {propertyValues {Curve {keys [
          {interpMode=RCIM_Cubic, tangentMode=RCTM_User, time=0, value=0.3}
          {interpMode=RCIM_Cubic, tangentMode=RCTM_User, time=0.3, value=1}
          {arriveTangent=-1, interpMode=RCIM_Cubic, leaveTangent=-1, time=0.7, value=1}
          {tangentMode=RCTM_User, time=1, value=0.3}
         ]}}}
      CurveIndex=$Particles.Normalized Spawn Group
     )
    )) {isSetParameters=true, module=SetVariables_#12}
   /SetVariables(Particles.Ribbon width03:NiagaraFloat=&Multiply_Float(
     A=13.2
     B=&FloatFromCurve(
      FloatCurve=DataInterface() {propertyValues {Curve {keys [
          {interpMode=RCIM_Cubic, tangentMode=RCTM_User, time=0, value=0.3}
          {interpMode=RCIM_Cubic, tangentMode=RCTM_User, time=0.3, value=1}
          {arriveTangent=-1, interpMode=RCIM_Cubic, leaveTangent=-1, time=0.7, value=1}
          {tangentMode=RCTM_User, time=1, value=0.3}
         ]}}}
      CurveIndex=$Particles.Normalized Spawn Group
     )
    )) {isSetParameters=true, module=SetVariables_#13}
   &ParticleState(Kill Particles When Lifetime Has Elapsed=false, Let Infinitely Lived Particles Die When Emitter Deactivates=true)
   &FN_BoneFlow(
    Angle:NiagaraFloat=&FN_FrontEndFloatModifier(Gameplay=&Multiply_Float(A=$Particles.Normalized Spawn Group, B=0.3), Front End=0.1)
    BoneID:NiagaraInt32=&ModuloInt(A=$Particles.SpawnGroup, B=1)
    Height:NiagaraFloat=&Multiply_Float(A=$Particles.Normalized Spawn Group, B=&FN_FrontEndFloatModifier(Gameplay=0.7, Front End=0.6))
    Radius:NiagaraFloat=&FloatFromCurve(
     FloatCurve=DataInterface() {propertyValues {Curve {keys [
         {arriveTangent=1.3929, interpMode=RCIM_Cubic, leaveTangent=1.3929, tangentMode=RCTM_User, time=0, value=0.3}
         {interpMode=RCIM_Cubic, tangentMode=RCTM_User, time=0.5, value=1.3}
         {arriveTangent=-2.2123, tangentMode=RCTM_User, time=1, value=0.2}
        ]}}}
     CurveIndex=$Particles.Normalized Spawn Group
     Scale Curve=12
    )
    SkeletalMesh:NiagaraDataInterfaceSkeletalMesh=$System.Skeletal Mesh RHand
   ) {module=FN_BoneFlow001}
   &ScaleRibbonWidth(Ribbon Width Scale=&FloatFromCurve(
     FloatCurve=DataInterface() {propertyValues {Curve {keys [
         {interpMode=RCIM_Cubic, tangentMode=RCTM_User, time=0, value=0.5}
         {interpMode=RCIM_Cubic, tangentMode=RCTM_User, time=0.3, value=1}
         {arriveTangent=-0.7143, interpMode=RCIM_Cubic, leaveTangent=-0.7143, time=0.7, value=1}
         {tangentMode=RCTM_User, time=1, value=0.5}
        ]}}}
     CurveIndex=$Particles.Normalized Spawn Group
    ))
   /SetVariables(Particles.Ribbon02 Color:LinearColor=&MakeLinearColorFromVectorAndFloat(
     "Vector (RGB)"=&MakeVectorFromLinearColorRGB(Color=&MultiplyLinearColor(
       Linear Color=<rgba 0.5 0.5 0.5 1>
       Scale Factor=&MakeVector4_FromFloat(Float=&FN_FrontEndFloatModifier(Gameplay=$System.TOD, Front End=$System.Frontend Bright NoMusic))
      ))
     "Float (Alpha)"=&FN_FrontEndFloatModifier(
      Gameplay=&Multiply_Float(
       A=&Multiply_Float(A=2, B=&Lerp_Float(A=0, B=1, Alpha=&ClampFloat(Float=$System.Music_Reactive_preview)))
       B=&Multiply_Float(
        A=$User.Opacity
        B=&FN_LOD_Distance_Remap(
         Start Distance=200
         End Distance=300
         Minimum Output=&Lerp_Float(A=0.1, B=1, Alpha=$System.LOD_Fade_preview)
         Maximum Output=1
         Clamp Output=true
        )
       )
      )
      Front End=$System.Frontend Alpha NoMusic
     )
    )) {isSetParameters=true, module=SetVariables_#14}
   /SetVariables(Particles.Position=&AddVectorToPosition(
     Position=$Particles.Position
     Vector=&Multiply_VectorByFloat(
      Vector=&NormalizeVector(
       Vector To Normalize=&SubtractVectorFromPosition(
        Position=&ReturnCameraPosition(Coordinate Space:ENiagaraCoordinateSpace=::Simulation, Camera Query:NiagaraDataInterfaceCamera=DataInterface())
        Vector=Linked() {linkedVariable {name=Particles.Position, type=NiagaraPosition}}
       )
       Safe Normalize=true
       Safe Fallback Threshold=0
       Safe Fallback Vector=<0 0 0.001>
      )
      Float=0
     )
    )) {enabled=false, isSetParameters=true, module=SetVariables_#15}
   &DynamicMaterialParameters(
    Index 0 Param 1=$System.WPO
    Index 0 Param 2=&Lerp_Float(A=1, B=1.5, Alpha=$System.Music_Reactive_preview)
    Module.Index0Param0Write=true
    Module.Index0Param1Write=true
    Module.Index0Param2Write=true
    Module.Index0Param3Write=true
   )
   &ScaleColor(
    Scale Alpha=&FN_FrontEndFloatModifier(
     Gameplay=&Multiply_Float(
      A=&Lerp_Float(A=0, B=1, Alpha=&ClampFloat(Float=$System.Music_Reactive_preview))
      B=&Multiply_Float(
       A=$User.Opacity
       B=&FN_LOD_Distance_Remap(
        Start Distance=200
        End Distance=300
        Minimum Output=&Lerp_Float(A=0.01, B=1, Alpha=$System.LOD_Fade_preview)
        Maximum Output=1
        Clamp Output=true
       )
      )
     )
     Front End=1
    )
    Scale RGB=&Multiply_VectorByFloat(
     Vector=&Multiply_VectorByFloat(
      Vector=<1 1 1>
      Float=&FN_PickFloatDependOnPlatform(TargetPlatformActive=1, TargetPlatformNoActive=0.75)
     )
     Float=&FN_FrontEndFloatModifier(Gameplay=$System.TOD, Front End=1)
    )
    ScaleA=true
    ScaleRGB=true
   )
  ]
  rendererClasses [&NiagaraRibbonRendererProperties]
  renderers [
   {rendererClass=&NiagaraRibbonRendererProperties}
   {rendererClass=&NiagaraRibbonRendererProperties, rendererIndex=1}
   {rendererClass=&NiagaraRibbonRendererProperties, rendererIndex=2}
  ]
 }
 Bone_Flow_LeftArm003 {
  emitterUpdateScript [
   &EmitterState(
    Inactive Response=::NewEnumerator1
    Life Cycle Mode=::NewEnumerator1
    Loop Behavior=::NewEnumerator1
    Loop Duration Mode=::NewEnumerator1
    UseLoopDelay=false
   )
   &SpawnBurst_Instantaneous(
    Spawn Count=&MakeCustomIntFromBool(Bool=$User.bIsFrontEndPreview, True INT=10, False INT=0)
    Use Spawn Probability=false
    UseLoopCountLimit=false
   )
  ]
  particleSpawnScript [
   &FN_SetSpawnGroup(Inclusivity=true, Normalize Spawn Group=true, Particles Per Group=10, Pattern=::NewEnumerator0, Set Ribbon ID and Ribbon Link Order=true)
   &InitializeParticle(
    Color=<rgba 1 1 1 1.5>
    Ribbon Width=10
    Ribbon Width Mode=::NewEnumerator1
    Sprite Size Mode=::NewEnumerator3
    Uniform Sprite Size=1
    UsePositionOffset=false
    Write Material Random=true
   )
   &CameraOffset(Camera Offset Amount=30)
  ]
  particleUpdateScript [
   /SetVariables(Particles.Ribbon width02:NiagaraFloat=&Multiply_Float(
     A=13
     B=&FloatFromCurve(
      FloatCurve=DataInterface() {propertyValues {Curve {keys [
          {interpMode=RCIM_Cubic, tangentMode=RCTM_User, time=0, value=0.3}
          {interpMode=RCIM_Cubic, tangentMode=RCTM_User, time=0.3, value=1}
          {arriveTangent=-1, interpMode=RCIM_Cubic, leaveTangent=-1, time=0.7, value=1}
          {tangentMode=RCTM_User, time=1, value=0.3}
         ]}}}
      CurveIndex=$Particles.Normalized Spawn Group
     )
    )) {isSetParameters=true, module=SetVariables_#12}
   /SetVariables(Particles.Ribbon width03:NiagaraFloat=&Multiply_Float(
     A=9
     B=&FloatFromCurve(
      FloatCurve=DataInterface() {propertyValues {Curve {keys [
          {interpMode=RCIM_Cubic, tangentMode=RCTM_User, time=0, value=0.3}
          {interpMode=RCIM_Cubic, tangentMode=RCTM_User, time=0.3, value=1}
          {arriveTangent=-1, interpMode=RCIM_Cubic, leaveTangent=-1, time=0.7, value=1}
          {tangentMode=RCTM_User, time=1, value=0.3}
         ]}}}
      CurveIndex=$Particles.Normalized Spawn Group
     )
    )) {isSetParameters=true, module=SetVariables_#13}
   &ParticleState(Kill Particles When Lifetime Has Elapsed=false, Let Infinitely Lived Particles Die When Emitter Deactivates=true)
   &FN_BoneFlow(
    Angle:NiagaraFloat=&FN_FrontEndFloatModifier(Gameplay=&Multiply_Float(A=$Particles.Normalized Spawn Group, B=0.3), Front End=0.1)
    BoneID:NiagaraInt32=&ModuloInt(A=$Particles.SpawnGroup, B=1)
    Height:NiagaraFloat=&Multiply_Float(A=$Particles.Normalized Spawn Group, B=&FN_FrontEndFloatModifier(Gameplay=0.7, Front End=0.6))
    Radius:NiagaraFloat=&FloatFromCurve(
     FloatCurve=DataInterface() {propertyValues {Curve {keys [
         {arriveTangent=1.3929, interpMode=RCIM_Cubic, leaveTangent=1.3929, tangentMode=RCTM_User, time=0, value=0.3}
         {interpMode=RCIM_Cubic, tangentMode=RCTM_User, time=0.5, value=1}
         {arriveTangent=-2.2123, tangentMode=RCTM_User, time=1, value=0.2}
        ]}}}
     CurveIndex=$Particles.Normalized Spawn Group
     Scale Curve=12
    )
    SkeletalMesh:NiagaraDataInterfaceSkeletalMesh=$System.Skeletal Mesh RHand
   ) {module=FN_BoneFlow001}
   &ScaleRibbonWidth(Ribbon Width Scale=&FloatFromCurve(
     FloatCurve=DataInterface() {propertyValues {Curve {keys [
         {interpMode=RCIM_Cubic, tangentMode=RCTM_User, time=0, value=0.5}
         {interpMode=RCIM_Cubic, tangentMode=RCTM_User, time=0.3, value=1}
         {arriveTangent=-0.7143, interpMode=RCIM_Cubic, leaveTangent=-0.7143, time=0.7, value=1}
         {tangentMode=RCTM_User, time=1, value=0.5}
        ]}}}
     CurveIndex=$Particles.Normalized Spawn Group
    ))
   /SetVariables(Particles.Ribbon02 Color:LinearColor=&MakeLinearColorFromVectorAndFloat(
     "Vector (RGB)"=&MakeVectorFromLinearColorRGB(Color=&MultiplyLinearColor(
       Linear Color=<rgba 0.5 0.5 0.5 1>
       Scale Factor=&MakeVector4_FromFloat(Float=&FN_FrontEndFloatModifier(Gameplay=$System.TOD, Front End=$System.Frontend Bright NoMusic))
      ))
     "Float (Alpha)"=&FN_FrontEndFloatModifier(
      Gameplay=&Multiply_Float(
       A=&Multiply_Float(A=2, B=&Lerp_Float(A=0, B=1, Alpha=&ClampFloat(Float=$System.Music_Reactive_preview)))
       B=&Multiply_Float(
        A=$User.Opacity
        B=&FN_LOD_Distance_Remap(
         Start Distance=200
         End Distance=300
         Minimum Output=&Lerp_Float(A=0.1, B=1, Alpha=$System.LOD_Fade_preview)
         Maximum Output=1
         Clamp Output=true
        )
       )
      )
      Front End=$System.Frontend Alpha NoMusic
     )
    )) {isSetParameters=true, module=SetVariables_#14}
   /SetVariables(Particles.Position=&AddVectorToPosition(
     Position=$Particles.Position
     Vector=&Multiply_VectorByFloat(
      Vector=&NormalizeVector(
       Vector To Normalize=&SubtractVectorFromPosition(
        Position=&ReturnCameraPosition(Coordinate Space:ENiagaraCoordinateSpace=::Simulation, Camera Query:NiagaraDataInterfaceCamera=DataInterface())
        Vector=Linked() {linkedVariable {name=Particles.Position, type=NiagaraPosition}}
       )
       Safe Normalize=true
       Safe Fallback Threshold=0
       Safe Fallback Vector=<0 0 0.001>
      )
      Float=0
     )
    )) {isSetParameters=true, module=SetVariables_#15}
   &DynamicMaterialParameters(
    Index 0 Param 1=&Multiply_Float(A=$System.WPO, B=0.5)
    Index 0 Param 2=&Lerp_Float(
     A=1
     B=&FN_FrontEndFloatModifier(Gameplay=1.5, Front End=1)
     Alpha=$System.Music_Reactive_preview
    )
    Module.Index0Param0Write=true
    Module.Index0Param1Write=true
    Module.Index0Param2Write=true
    Module.Index0Param3Write=true
   )
   &ScaleColor(
    Scale Alpha=&FN_FrontEndFloatModifier(
     Gameplay=&Multiply_Float(
      A=&Lerp_Float(A=0, B=1, Alpha=&ClampFloat(Float=$System.Music_Reactive_preview))
      B=&Multiply_Float(
       A=$User.Opacity
       B=&FN_LOD_Distance_Remap(
        Start Distance=200
        End Distance=300
        Minimum Output=&Lerp_Float(A=0.01, B=1, Alpha=$System.LOD_Fade_preview)
        Maximum Output=1
        Clamp Output=true
       )
      )
     )
     Front End=1
    )
    Scale RGB=&Multiply_VectorByFloat(
     Vector=&Multiply_VectorByFloat(
      Vector=<1 1 1>
      Float=&FN_PickFloatDependOnPlatform(TargetPlatformActive=1, TargetPlatformNoActive=0.75)
     )
     Float=&FN_FrontEndFloatModifier(Gameplay=$System.TOD, Front End=1)
    )
    ScaleA=true
    ScaleRGB=true
   )
  ]
  rendererClasses [&NiagaraRibbonRendererProperties]
  renderers [
   {rendererClass=&NiagaraRibbonRendererProperties}
   {rendererClass=&NiagaraRibbonRendererProperties, rendererIndex=1}
   {rendererClass=&NiagaraRibbonRendererProperties, rendererIndex=2}
  ]
 }
 Bone_Flow_RightArm {
  emitterUpdateScript [
   &EmitterState(
    Inactive Response=::NewEnumerator1
    Life Cycle Mode=::NewEnumerator1
    Loop Behavior=::NewEnumerator1
    Loop Duration Mode=::NewEnumerator1
    UseLoopDelay=false
   )
   &SpawnBurst_Instantaneous(
    Spawn Count=&MakeCustomIntFromBool(Bool=$User.bIsFrontEndPreview, True INT=0, False INT=10)
    Use Spawn Probability=false
    UseLoopCountLimit=false
   )
  ]
  particleSpawnScript [
   &FN_SetSpawnGroup(Inclusivity=true, Normalize Spawn Group=true, Particles Per Group=10, Pattern=::NewEnumerator0, Set Ribbon ID and Ribbon Link Order=true)
   &InitializeParticle(
    Color=<rgba 1 1 1 1.5>
    Ribbon Width=12
    Ribbon Width Mode=::NewEnumerator1
    Sprite Size Mode=::NewEnumerator3
    Uniform Sprite Size=1
    UsePositionOffset=false
    Write Material Random=true
   )
   &CameraOffset(Camera Offset Amount=30)
  ]
  particleUpdateScript [
   /SetVariables(Particles.Ribbon width02:NiagaraFloat=&Multiply_Float(
     A=12
     B=&FloatFromCurve(
      FloatCurve=DataInterface() {propertyValues {Curve {keys [
          {interpMode=RCIM_Cubic, tangentMode=RCTM_User, time=0, value=0.3}
          {interpMode=RCIM_Cubic, tangentMode=RCTM_User, time=0.3, value=0.8}
          {arriveTangent=-0.7143, interpMode=RCIM_Cubic, leaveTangent=-0.7143, time=0.7, value=0.8}
          {tangentMode=RCTM_User, time=1, value=0.3}
         ]}}}
      CurveIndex=$Particles.Normalized Spawn Group
     )
    )) {isSetParameters=true, module=SetVariables_#12}
   &ParticleState(Kill Particles When Lifetime Has Elapsed=false, Let Infinitely Lived Particles Die When Emitter Deactivates=true)
   &FN_BoneFlow(
    Angle:NiagaraFloat=&FN_FrontEndFloatModifier(Gameplay=&Multiply_Float(A=$Particles.Normalized Spawn Group, B=0.3), Front End=0.2)
    BoneID:NiagaraInt32=&ModuloInt(A=$Particles.SpawnGroup, B=1)
    Enable Position Offsets=true
    Height:NiagaraFloat=&Multiply_Float(A=$Particles.Normalized Spawn Group, B=&FN_FrontEndFloatModifier(Gameplay=0.7, Front End=0.6))
    Offset Mid:Vector3f=<0 0 5>
    Offset Start:Vector3f=<-10 10 0>
    Radius:NiagaraFloat=&FloatFromCurve(
     FloatCurve=DataInterface() {propertyValues {Curve {keys [
         {arriveTangent=1.3929, interpMode=RCIM_Cubic, leaveTangent=1.3929, tangentMode=RCTM_User, time=0, value=0.3}
         {interpMode=RCIM_Cubic, tangentMode=RCTM_User, time=0.5, value=1.3}
         {arriveTangent=-2.2123, tangentMode=RCTM_User, time=1, value=0.2}
        ]}}}
     CurveIndex=$Particles.Normalized Spawn Group
     Scale Curve=10
    )
    SkeletalMesh:NiagaraDataInterfaceSkeletalMesh=$System.Skeletal Mesh RHand
   )
   &ScaleRibbonWidth(Ribbon Width Scale=&FloatFromCurve(
     FloatCurve=DataInterface() {propertyValues {Curve {keys [
         {interpMode=RCIM_Cubic, tangentMode=RCTM_User, time=0, value=0.5}
         {interpMode=RCIM_Cubic, tangentMode=RCTM_User, time=0.3, value=0.8}
         {arriveTangent=-0.4286, interpMode=RCIM_Cubic, leaveTangent=-0.4286, time=0.7, value=0.8}
         {tangentMode=RCTM_User, time=1, value=0.5}
        ]}}}
     CurveIndex=$Particles.Normalized Spawn Group
    ))
   /SetVariables(Particles.Ribbon02 Color:LinearColor=&MakeLinearColorFromVectorAndFloat(
     "Vector (RGB)"=&MakeVectorFromLinearColorRGB(Color=&MultiplyLinearColor(
       Linear Color=<rgba 0.5 0.5 0.5 1>
       Scale Factor=&MakeVector4_FromFloat(Float=&FN_FrontEndFloatModifier(Gameplay=$System.TOD, Front End=$System.Frontend Bright NoMusic))
      ))
     "Float (Alpha)"=&FN_FrontEndFloatModifier(
      Gameplay=&Multiply_Float(
       A=3
       B=&Multiply_Float(
        A=&Lerp_Float(A=0, B=1, Alpha=&ClampFloat(Float=$System.Music_Reactive_preview))
        B=&Multiply_Float(
         A=&Multiply_Float(
          A=$User.Opacity
          B=&FloatFromCurve(
           FloatCurve=DataInterface() {propertyValues {Curve {keys [{tangentMode=RCTM_User, time=0.2, value=0}, {tangentMode=RCTM_User, time=0.5, value=1}]}}}
           CurveIndex=$Particles.Normalized Spawn Group
          )
         )
         B=&FN_LOD_Distance_Remap(
          Start Distance=200
          End Distance=300
          Minimum Output=&Lerp_Float(A=0.1, B=1, Alpha=$System.LOD_Fade_preview)
          Maximum Output=1
          Clamp Output=true
         )
        )
       )
      )
      Front End=$System.Frontend Alpha NoMusic
     )
    )) {isSetParameters=true, module=SetVariables_#14}
   /SetVariables(Particles.Position=&AddVectorToPosition(
     Position=$Particles.Position
     Vector=&Multiply_VectorByFloat(
      Vector=&NormalizeVector(
       Vector To Normalize=&SubtractVectorFromPosition(
        Position=&ReturnCameraPosition(Coordinate Space:ENiagaraCoordinateSpace=::Simulation, Camera Query:NiagaraDataInterfaceCamera=DataInterface())
        Vector=Linked() {linkedVariable {name=Particles.Position, type=NiagaraPosition}}
       )
       Safe Normalize=true
       Safe Fallback Threshold=0
       Safe Fallback Vector=<0 0 0.001>
      )
      Float=0
     )
    )) {isSetParameters=true, module=SetVariables_#15}
   &DynamicMaterialParameters(
    Index 0 Param 1=$System.WPO
    Index 0 Param 2=&Lerp_Float(
     A=1
     B=&FN_FrontEndFloatModifier(Gameplay=1.5, Front End=1)
     Alpha=$System.Music_Reactive_preview
    )
    Module.Index0Param0Write=true
    Module.Index0Param1Write=true
    Module.Index0Param2Write=true
    Module.Index0Param3Write=true
   )
   &ScaleColor(
    Scale Alpha=&FN_FrontEndFloatModifier(
     Gameplay=&Multiply_Float(
      A=&Lerp_Float(A=0, B=1, Alpha=$System.InterpolateOverTime.InterpolateOverTime_Float.Moving Average)
      B=&Multiply_Float(
       A=&Multiply_Float(
        A=$User.Opacity
        B=&FloatFromCurve(
         FloatCurve=DataInterface() {propertyValues {Curve {keys [{tangentMode=RCTM_User, time=0.2, value=1}, {tangentMode=RCTM_User, time=0.5, value=1}]}}}
         CurveIndex=$Particles.Normalized Spawn Group
        )
       )
       B=&FN_LOD_Distance_Remap(
        Start Distance=200
        End Distance=300
        Minimum Output=&Lerp_Float(A=0.05, B=1, Alpha=$System.LOD_Fade_preview)
        Maximum Output=1
        Clamp Output=true
       )
      )
     )
     Front End=1
    )
    Scale RGB=&Multiply_VectorByFloat(
     Vector=&Multiply_VectorByFloat(
      Vector=<1 1 1>
      Float=&FN_PickFloatDependOnPlatform(TargetPlatformActive=1, TargetPlatformNoActive=0.75)
     )
     Float=&FN_FrontEndFloatModifier(Gameplay=$System.TOD, Front End=1)
    )
    ScaleA=true
    ScaleRGB=true
   )
  ]
  rendererClasses [&NiagaraRibbonRendererProperties]
  renderers [
   {rendererClass=&NiagaraRibbonRendererProperties}
   {rendererClass=&NiagaraRibbonRendererProperties, rendererIndex=1}
  ]
 }
 Bone_Flow_Right_shoulder {
  emitterUpdateScript [
   &EmitterState(
    Inactive Response=::NewEnumerator1
    Life Cycle Mode=::NewEnumerator1
    Loop Behavior=::NewEnumerator1
    Loop Duration Mode=::NewEnumerator1
    UseLoopDelay=false
   )
   &SpawnBurst_Instantaneous(Spawn Count=10, Use Spawn Probability=false, UseLoopCountLimit=false)
  ]
  particleSpawnScript [
   &FN_SetSpawnGroup(Inclusivity=true, Normalize Spawn Group=true, Particles Per Group=10, Pattern=::NewEnumerator0, Set Ribbon ID and Ribbon Link Order=true)
   &InitializeParticle(
    Color=<rgba 1 1 1 1.5>
    Ribbon Width=12
    Ribbon Width Mode=::NewEnumerator1
    Sprite Size Mode=::NewEnumerator3
    Uniform Sprite Size=1
    UsePositionOffset=false
    Write Material Random=true
   )
   &CameraOffset(Camera Offset Amount=30)
  ]
  particleUpdateScript [
   /SetVariables(Particles.Ribbon width02:NiagaraFloat=&Multiply_Float(
     A=20
     B=&FloatFromCurve(
      FloatCurve=DataInterface() {propertyValues {Curve {keys [
          {interpMode=RCIM_Cubic, tangentMode=RCTM_User, time=0, value=0}
          {arriveTangent=0.3, interpMode=RCIM_Cubic, leaveTangent=0.3, time=0.7, value=1}
          {tangentMode=RCTM_User, time=1, value=0.3}
         ]}}}
      CurveIndex=$Particles.Normalized Spawn Group
     )
    )) {isSetParameters=true}
   /SetVariables(Particles.Ribbon width03:NiagaraFloat=&Multiply_Float(
     A=13.2
     B=&FloatFromCurve(
      FloatCurve=DataInterface() {propertyValues {Curve {keys [
          {interpMode=RCIM_Cubic, tangentMode=RCTM_User, time=0, value=0.3}
          {interpMode=RCIM_Cubic, tangentMode=RCTM_User, time=0.3, value=1}
          {arriveTangent=-1, interpMode=RCIM_Cubic, leaveTangent=-1, time=0.7, value=1}
          {tangentMode=RCTM_User, time=1, value=0.3}
         ]}}}
      CurveIndex=$Particles.Normalized Spawn Group
     )
    )) {isSetParameters=true}
   &ParticleState(Kill Particles When Lifetime Has Elapsed=false, Let Infinitely Lived Particles Die When Emitter Deactivates=true)
   &FN_BoneFlow(
    Angle:NiagaraFloat=&Multiply_Float(A=$Particles.Normalized Spawn Group, B=0.5)
    BoneID:NiagaraInt32=&ModuloInt(A=$Particles.SpawnGroup, B=1)
    Height:NiagaraFloat=&Multiply_Float(A=$Particles.Normalized Spawn Group, B=1)
    Radius:NiagaraFloat=&FloatFromCurve(
     FloatCurve=DataInterface() {propertyValues {Curve {keys [
         {interpMode=RCIM_Cubic, tangentMode=RCTM_User, time=0, value=0.4}
         {interpMode=RCIM_Cubic, tangentMode=RCTM_User, time=0.5, value=1}
         {tangentMode=RCTM_User, time=1, value=0.4}
        ]}}}
     CurveIndex=$Particles.Normalized Spawn Group
     Scale Curve=10
    )
    SkeletalMesh:NiagaraDataInterfaceSkeletalMesh=$System.Skeletal Mesh R
   )
   &ScaleRibbonWidth(Ribbon Width Scale=&FloatFromCurve(
     FloatCurve=DataInterface() {propertyValues {Curve {keys [
         {interpMode=RCIM_Cubic, tangentMode=RCTM_User, time=0, value=0.5}
         {interpMode=RCIM_Cubic, tangentMode=RCTM_User, time=0.3, value=1}
         {arriveTangent=-0.7143, interpMode=RCIM_Cubic, leaveTangent=-0.7143, time=0.7, value=1}
         {tangentMode=RCTM_User, time=1, value=0.5}
        ]}}}
     CurveIndex=$Particles.Normalized Spawn Group
    ))
   /SetVariables(Particles.Ribbon02 Color:LinearColor=&MakeLinearColorFromVectorAndFloat(
     "Vector (RGB)"=&MakeVectorFromLinearColorRGB(Color=&MultiplyLinearColor(
       Linear Color=<rgba 0.5 0.5 0.5 1>
       Scale Factor=&MakeVector4_FromFloat(Float=&FN_FrontEndFloatModifier(Gameplay=$System.TOD, Front End=$System.Frontend Bright NoMusic))
      ))
     "Float (Alpha)"=&FN_FrontEndFloatModifier(
      Gameplay=&Multiply_Float(
       A=3
       B=&Multiply_Float(
        A=&Lerp_Float(A=0, B=1, Alpha=&ClampFloat(Float=$System.Music_Reactive_preview))
        B=&Multiply_Float(
         A=$User.Opacity
         B=&FN_LOD_Distance_Remap(
          Start Distance=200
          End Distance=300
          Minimum Output=&Lerp_Float(A=0, B=1, Alpha=$System.LOD_Fade_preview)
          Maximum Output=1
          Clamp Output=true
         )
        )
       )
      )
      Front End=$System.Frontend Alpha NoMusic
     )
    )) {isSetParameters=true}
   /SetVariables(Particles.Position=&AddVectorToPosition(
     Position=$Particles.Position
     Vector=&Multiply_VectorByFloat(
      Vector=&NormalizeVector(
       Vector To Normalize=&SubtractVectorFromPosition(
        Position=&ReturnCameraPosition(Coordinate Space:ENiagaraCoordinateSpace=::Simulation, Camera Query:NiagaraDataInterfaceCamera=DataInterface())
        Vector=Linked() {linkedVariable {name=Particles.Position, type=NiagaraPosition}}
       )
       Safe Normalize=true
       Safe Fallback Threshold=0
       Safe Fallback Vector=<0 0 0.001>
      )
      Float=1
     )
    )) {isSetParameters=true}
   &DynamicMaterialParameters(
    Index 0 Param 1=$System.WPO
    Index 0 Param 2=&Lerp_Float(
     A=1
     B=&FN_FrontEndFloatModifier(Gameplay=1.5, Front End=1)
     Alpha=$System.Music_Reactive_preview
    )
    Module.Index0Param0Write=true
    Module.Index0Param1Write=true
    Module.Index0Param2Write=true
    Module.Index0Param3Write=true
   )
   &ScaleColor(
    Scale Alpha=&FN_FrontEndFloatModifier(
     Gameplay=&Multiply_Float(
      A=&Multiply_Float(
       A=&Lerp_Float(A=0, B=1, Alpha=$System.InterpolateOverTime.InterpolateOverTime_Float.Moving Average)
       B=$User.Opacity
      )
      B=&FN_LOD_Distance_Remap(
       Start Distance=200
       End Distance=300
       Minimum Output=&Lerp_Float(A=0, B=1, Alpha=$System.LOD_Fade_preview)
       Maximum Output=1
       Clamp Output=true
      )
     )
     Front End=1
    )
    Scale RGB=&Multiply_VectorByFloat(
     Vector=&Multiply_VectorByFloat(
      Vector=<1 1 1>
      Float=&FN_PickFloatDependOnPlatform(TargetPlatformActive=1, TargetPlatformNoActive=0.75)
     )
     Float=&FN_FrontEndFloatModifier(Gameplay=$System.TOD, Front End=1)
    )
    ScaleA=true
    ScaleRGB=true
   )
  ]
  rendererClasses [&NiagaraRibbonRendererProperties]
  renderers [
   {rendererClass=&NiagaraRibbonRendererProperties}
   {rendererClass=&NiagaraRibbonRendererProperties, rendererIndex=1}
   {rendererClass=&NiagaraRibbonRendererProperties, rendererIndex=2}
  ]
 }
 Drop_Flare {
  simTarget=GPUComputeSim
  emitterUpdateScript [
   &EmitterState()
   &SpawnRate(
    Spawn Probability=&FN_FrontEndFloatModifier(Gameplay=&Lerp_Float(A=0, B=1, Alpha=$System.Music_Reactive_preview), Front End=1)
    SpawnRate=&MakeCustomFloatFromBool(
     Bool=$User.IsSkydiving
     True Float=0
     False Float=&MakeCustomFloatFromBool(
      Bool=$User.bisInVehicle
      True Float=0
      False Float=&MakeCustomFloatFromBool(
       Bool=&SetBoolByFloatComparison(A=&VectorLength(Vector=$Engine.Owner.Velocity), B=1000)
       True Float=0
       False Float=20
      )
     )
    )
    Use Spawn Probability=true
   )
  ]
  particleSpawnScript [
   &InitializeParticle(
    Color=<rgba 2 0.9271 1.6421 1>
    Lifetime Max=1
    Lifetime Min=0.5
    Lifetime Mode=::NewEnumerator1
    Mesh Scale Mode=::NewEnumerator1
    Mesh Uniform Scale Max=0.5
    Mesh Uniform Scale Min=0.15
    Sprite Size Mode=::NewEnumerator1
    Uniform Sprite Size Max=12
    Uniform Sprite Size Min=7
    UsePositionOffset=false
    Write Material Random=true
   )
   &SkeletalMeshLocation(Bone Sampling Mode=::NewEnumerator4, MaskSpawnGroup=false, Sampled Position Offset=<-30 0 0>, Skeletal Mesh=$System.Skeletal Mesh debris)
   &SkeletalMeshLocation(
    Bone / Socket Index=2
    Bone Sampling Mode=::NewEnumerator4
    Interpolate Mesh Sampling=false
    MaskSpawnGroup=false
    Orientation Sampling=::NewEnumerator1
    Position Sampling=::NewEnumerator2
    Skeletal Mesh=$System.Skeletal Mesh debris
   ) {module=SkeletalMeshLocation001}
   &CylinderLocation(Cylinder Height=15, Cylinder Radius=25, MaskSpawnGroup=false, Override Seed=false, Surface Only=true, UseOwnerScale=true)
   &AddVelocityFromPoint(
    Module.Use Falloff Distance=false
    Origin Offset=<0 0 130>
    Velocity Origin:NiagaraPosition=$Particles.SkeletalMeshLocation001.SampledPosition
    Velocity Strength=&RandomRangeFloat(Minimum=30, Maximum=40)
   )
   &CameraOffset(Camera Offset Amount=5)
  ]
  particleUpdateScript [
   &Drag(Use Linear Drag=true, Use Rotational Drag=false)
   &CurlNoiseForce(
    Noise Strength=&FloatFromCurve(FloatCurve=DataInterface() {propertyValues {Curve {keys [<at 0 100>,<at 1 50>]}}})
    Pan Noise=false
   )
   &ParticleState()
   &GravityForce(Gravity=<0 0 -10>)
   &InheritSourceMovement(
    Applied Position Falloff Curve=DataInterface() {propertyValues {Curve {keys [{interpMode=RCIM_Cubic, tangentMode=RCTM_User, time=0, value=1}, {interpMode=RCIM_Cubic, tangentMode=RCTM_User, time=1, value=0}]}}}
    Applied Position Falloff Distance=150
    ApplyForce=false
    ApplyPosition=true
    ApplyPositionFalloff=true
    ApplyPositionFalloffCurve=true
   )
   &SolveForcesAndVelocity(Clamp Velocity=false, Limit Acceleration=false)
   &ScaleSpriteSize(Uniform Curve Sprite Scale=DataInterface() {propertyValues {Curve {keys [
        {arriveTangent=14.9673, interpMode=RCIM_Cubic, leaveTangent=14.9673, tangentMode=RCTM_User, time=0, value=0}
        <at 0.2 1 RCIM_Cubic>
        {arriveTangent=-3.2414, interpMode=RCIM_Cubic, leaveTangent=-3.2414, tangentMode=RCTM_User, time=1, value=0}
       ]}}}) {module=ScaleSpriteSize002}
   &ScaleColor(
    Scale Alpha=&Multiply_Float(
     A=$User.Debris Opacity
     B=&Lerp_Float(A=0, B=1, Alpha=&ClampFloat(Float=$System.Music_Reactive_preview))
    )
    Scale RGB=&Multiply_VectorByFloat(
     Vector=&VectorFromFloat(Value=&FloatFromCurve(
       FloatCurve=DataInterface() {propertyValues {Curve {keys [
           {arriveTangent=85.1377, interpMode=RCIM_Cubic, leaveTangent=85.1377, tangentMode=RCTM_User, time=0, value=0}
           {arriveTangentWeight=1.8705, interpMode=RCIM_Cubic, leaveTangentWeight=1.8705, tangentMode=RCTM_User, time=0.05, value=3}
           <at 0.1 0.5 RCIM_Cubic>
          ]}}}
       Scale Curve=&Multiply_Float(
        A=$System.TOD
        B=&FloatFromCurve(
         FloatCurve=DataInterface() {propertyValues {Curve {keys [
             {interpMode=RCIM_Cubic, tangentMode=RCTM_User, time=0, value=-0.025}
             {interpMode=RCIM_Cubic, tangentMode=RCTM_User, time=0.1488, value=1.1054}
             {arriveTangent=-0.3951, interpMode=RCIM_Cubic, leaveTangent=-0.3951, time=0.2612, value=-0.15}
             {arriveTangent=0.1378, interpMode=RCIM_Cubic, leaveTangent=0.1378, time=0.3572, value=1.023}
             {arriveTangent=-0.0897, interpMode=RCIM_Cubic, leaveTangent=-0.0897, time=0.496, value=-0.1176}
             {arriveTangent=0.3675, interpMode=RCIM_Cubic, leaveTangent=0.3675, time=0.6139, value=1}
             {arriveTangent=-0.2002, interpMode=RCIM_Cubic, leaveTangent=-0.2002, time=0.7481, value=-0.025}
             <at 0.8636 0.95 RCIM_Cubic>
             {tangentMode=RCTM_User, time=1, value=-0.025}
            ]}}}
         Scale Curve=1.5
        )
       )
      ))
     Float=&Lerp_Float(A=0.1, B=3, Alpha=$System.Music_Reactive_preview)
    )
    ScaleA=true
    ScaleRGB=true
   ) {module=ScaleColor001}
   &DynamicMaterialParameters(
    Index 0 Param 1=0.3
    Index 0 Param 2=&FloatFromCurve(FloatCurve=DataInterface() {propertyValues {Curve {keys [{arriveTangent=-12.4095, interpMode=RCIM_Cubic, leaveTangent=-12.4095, tangentMode=RCTM_User, time=0.5, value=2}, <at 1 0 RCIM_Cubic>]}}})
    Module.Param0WriteEnabled=true
    Module.Param1WriteEnabled=true
    Module.Param2WriteEnabled=true
    Module.Param3WriteEnabled=true
   ) {module=DynamicMaterialParameters001}
  ]
  rendererClasses [&NiagaraSpriteRendererProperties]
  renderers [{rendererClass=&NiagaraSpriteRendererProperties}]
 }
 Shoulder_Local {
  emitterUpdateScript [
   &EmitterState()
   &SpawnRate(
    Spawn Probability=&FN_FrontEndFloatModifier(Gameplay=&Lerp_Float(A=0, B=1, Alpha=$System.Music_Reactive_preview), Front End=1)
    SpawnRate=10
    Use Spawn Probability=true
   )
  ]
  particleSpawnScript [
   &InitializeParticle(
    Color=<rgba 1.5 0.045 0.2058 1>
    Lifetime Max=1
    Lifetime Mode=::NewEnumerator1
    Mesh Scale Mode=::NewEnumerator1
    Mesh Uniform Scale Max=0.35
    Mesh Uniform Scale Min=0.15
    Sprite Size Mode=::NewEnumerator3
    Uniform Sprite Size=5
    UsePositionOffset=false
    Write Material Random=true
   )
   &SphereLocation(
    MaskSpawnGroup=false
    Override Seed=false
    Sphere Origin:NiagaraPosition=&MakeVector(X=0, Y=0, Z=0)
    Sphere Radius=8
    Surface Only=true
    UseNonUniformScale=false
    UseOffset=true
    UseSphereOrientation=false
   )
   &InitialMeshOrientation(
    Rotation=&RandomRangeVector(Minimum=<0 0 0>, Maximum=<1 1 1>)
    Use Orientation Vector=true
    Use Rotation Vector=true
   ) {module=InitialMeshOrientation001}
   &AddVelocity(
    Scale Added Velocity=&Multiply_Vector(A=<1 1 1>, B=&VectorFromFloat(Value=0.1))
    Velocity=<100 0 100>
   )
  ]
  particleUpdateScript [
   &Drag(Use Linear Drag=true, Use Rotational Drag=false)
   &CurlNoiseForce(
    Noise Strength=&FloatFromCurve(FloatCurve=DataInterface() {propertyValues {Curve {keys [<at 0 100>,<at 1 50>]}}})
    Pan Noise=false
   )
   &ParticleState()
   &UpdateMeshOrientation(
    Rotation Rate=&FloatFromCurve(
     FloatCurve=DataInterface() {propertyValues {Curve {keys [
         {arriveTangent=-2.1159, interpMode=RCIM_Cubic, leaveTangent=-2.1159, tangentMode=RCTM_User, time=0, value=1}
         {interpMode=RCIM_Cubic, tangentMode=RCTM_Break, time=1, value=0.2}
        ]}}}
     Scale Curve=&RandomRangeFloat(Minimum=0.5, Maximum=0.7)
    )
    Rotation Vector=&RandomRangeVector(Minimum=<-1 -1 -1>, Maximum=<1 1 1>)
   )
   &SolveForcesAndVelocity(Clamp Velocity=false, Limit Acceleration=false)
   &SkeletalMeshLocation(
    Bone / Socket Index=2
    Bone Sampling Mode=::NewEnumerator4
    MaskSpawnGroup=false
    Orientation Sampling=::NewEnumerator1
    Sampled Position Offset=&Add_Vector(
     A=<20 0 7>
     B=&Add_Vector(
      A=Linked() {linkedVariable {name=Particles.Initial.Position, type=NiagaraPosition}}
      B=&MakeVector(
       X=&Lerp_Float(A=0, B=10, Alpha=$Particles.NormalizedAge)
       Y=&Lerp_Float(A=0, B=-5, Alpha=$Particles.NormalizedAge)
       Z=&Lerp_Float(A=0, B=10, Alpha=$Particles.NormalizedAge)
      )
     )
    )
    Skeletal Mesh=$System.Skeletal Mesh debris
   ) {module=SkeletalMeshLocation001}
   &ScaleMeshSize(Scale Factor=&VectorFromFloat(Value=&FloatFromCurve(FloatCurve=DataInterface() {propertyValues {Curve {keys [{interpMode=RCIM_Cubic, tangentMode=RCTM_User, time=0, value=1}, {interpMode=RCIM_Cubic, tangentMode=RCTM_User, time=1, value=0}]}}})))
   &ScaleColor(
    Scale Alpha=&Multiply_Float(
     A=&RandomRangeFloat(Minimum=0.1, Maximum=1)
     B=&Multiply_Float(
      A=&FloatFromCurve(FloatCurve=DataInterface() {propertyValues {Curve {keys [
           {interpMode=RCIM_Cubic, tangentMode=RCTM_User, time=0, value=0}
           {arriveTangent=-1, interpMode=RCIM_Cubic, leaveTangent=-1, time=0.2, value=0.95}
           {arriveTangent=-1.7637, interpMode=RCIM_Cubic, leaveTangent=-1.7637, tangentMode=RCTM_User, time=1, value=0}
          ]}}})
      B=&Multiply_Float(
       A=&Lerp_Float(A=0, B=1, Alpha=&ClampFloat(Float=$System.Music_Reactive_preview))
       B=$User.Opacity
      )
     )
    )
    Scale RGB=&Multiply_VectorByFloat(
     Vector=&Multiply_VectorByFloat(Vector=<1 1 1>, Float=$System.TOD)
     Float=&Multiply_Float(
      A=&FloatFromCurve(
       FloatCurve=DataInterface() {propertyValues {Curve {keys [
           {interpMode=RCIM_Cubic, tangentMode=RCTM_User, time=0, value=-0.025}
           {interpMode=RCIM_Cubic, tangentMode=RCTM_User, time=0.1488, value=1.1054}
           {arriveTangent=-0.3951, interpMode=RCIM_Cubic, leaveTangent=-0.3951, time=0.2612, value=-0.15}
           {arriveTangent=0.1378, interpMode=RCIM_Cubic, leaveTangent=0.1378, time=0.3572, value=1.023}
           {arriveTangent=-0.0897, interpMode=RCIM_Cubic, leaveTangent=-0.0897, time=0.496, value=-0.1176}
           {arriveTangent=0.3675, interpMode=RCIM_Cubic, leaveTangent=0.3675, time=0.6139, value=1}
           {arriveTangent=-0.2002, interpMode=RCIM_Cubic, leaveTangent=-0.2002, time=0.7481, value=-0.025}
           <at 0.8636 0.95 RCIM_Cubic>
           {tangentMode=RCTM_User, time=1, value=-0.025}
          ]}}}
       Scale Curve=5
      )
      B=&Lerp_Float(A=0, B=10, Alpha=$System.Music_Reactive_preview)
     )
    )
    ScaleA=true
    ScaleRGB=true
   ) {module=ScaleColor001}
  ]
  rendererClasses [&NiagaraMeshRendererProperties]
  renderers [{rendererClass=&NiagaraMeshRendererProperties, rendererIndex=0}]
 }
 Shoulder_Smoke {
  emitterUpdateScript [
   &EmitterState(
    Enable Distance Culling=true
    MaxDistance=10000
    Module.bUseMaxDistance=true
    Module.bUseMinDistance=false
    Scalability Mode=::NewEnumerator1
   )
   &SpawnRate(
    Spawn Probability=&FN_FrontEndFloatModifier(Gameplay=&Lerp_Float(A=0, B=1, Alpha=$System.Music_Reactive_preview), Front End=1)
    SpawnRate=5
    Use Spawn Probability=true
   )
  ]
  particleSpawnScript [
   /SetVariables(Particles.SpriteRotationRate=&RandomRangeFloat(Minimum=-3, Maximum=3, Evaluation Type=::NewEnumerator1)) {isSetParameters=true}
   &InitializeParticle(
    Color=&LerpLinearColors(
     StartColor=<rgba 1 0.1406 1 0.3>
     EndColor=<rgba 1 0.0091 0.2095 0.3>
     LerpFactor=&FloatFromCurve(
      FloatCurve=DataInterface() {propertyValues {Curve {keys [{arriveTangent=2.6771, interpMode=RCIM_Cubic, leaveTangent=2.6771, tangentMode=RCTM_User, time=0, value=0}, <at 1 1 RCIM_Cubic>]}}}
      CurveIndex=&RandomRangeFloat(Minimum=0, Maximum=1)
     )
    )
    Lifetime Max=1.5
    Lifetime Min=0.7
    Lifetime Mode=::NewEnumerator1
    Sprite Rotation Mode=::NewEnumerator2
    Sprite Size Mode=::NewEnumerator1
    Uniform Sprite Size Max=50
    Uniform Sprite Size Min=40
    UsePositionOffset=false
    Write Material Random=true
   ) {module=InitializeParticle001}
   &DynamicMaterialParameters(
    Index 0 Param 1=&RandomRangeFloat(Minimum=0, Maximum=1, Evaluation Type=::NewEnumerator1)
    Index 0 Param 2=&RandomRangeFloat(Minimum=0, Maximum=1, Evaluation Type=::NewEnumerator1)
    Module.Param0WriteEnabled=true
    Module.Param1WriteEnabled=true
    Module.Param2WriteEnabled=true
    Module.Param3WriteEnabled=true
   )
   &CameraOffset(Camera Offset Amount=&RandomRangeFloat(Minimum=0, Maximum=10, Evaluation Type=::NewEnumerator1))
   &AddVelocity(
    Scale Added Velocity=&Multiply_Vector(A=<1 1 1>, B=&VectorFromFloat(Value=0.1))
    Velocity=<100 0 50>
   )
   &AddVelocity(
    Coordinate Space=::World
    Scale Added Velocity=&Multiply_Vector(A=<1 1 1>, B=&VectorFromFloat(Value=0.5))
    Velocity=<0 0 6>
   ) {module=AddVelocity001}
  ]
  particleUpdateScript [
   &ParticleState()
   &Drag(Drag=2.5, Use Linear Drag=true, Use Rotational Drag=false)
   &SolveForcesAndVelocity(Clamp Velocity=false, Limit Acceleration=false)
   &SkeletalMeshLocation(
    Bone / Socket Index=2
    Bone Sampling Mode=::NewEnumerator4
    MaskSpawnGroup=false
    Sampled Position Offset=&Add_Vector(
     A=<20 0 2>
     B=&MakeVector(
      X=&Lerp_Float(A=0, B=10, Alpha=$Particles.NormalizedAge)
      Y=&Lerp_Float(A=0, B=-5, Alpha=$Particles.NormalizedAge)
      Z=&Lerp_Float(A=0, B=10, Alpha=$Particles.NormalizedAge)
     )
    )
    Skeletal Mesh=$System.Skeletal Mesh debris
   )
   &ScaleColor(
    Scale Alpha=&Multiply_Float(
     A=&FloatFromCurve(FloatCurve=DataInterface() {propertyValues {Curve {keys [
          {interpMode=RCIM_Cubic, tangentMode=RCTM_User, time=0, value=0}
          {interpMode=RCIM_Cubic, tangentMode=RCTM_User, time=0.5, value=1}
          {tangentMode=RCTM_User, time=1, value=0}
         ]}}})
     B=&Multiply_Float(
      A=$User.Opacity
      B=&Lerp_Float(A=0, B=1, Alpha=&ClampFloat(Float=$System.Music_Reactive_preview))
     )
    )
    Scale RGB=&Multiply_VectorByFloat(
     Vector=&Multiply_VectorByFloat(Vector=<1 1 1>, Float=$System.TOD)
     Float=&Multiply_Float(A=&Lerp_Float(A=0.5, B=2, Alpha=$System.Music_Reactive_preview), B=1)
    )
    ScaleA=true
    ScaleRGB=true
   )
  ]
  rendererClasses [&NiagaraSpriteRendererProperties]
  renderers [{rendererClass=&NiagaraSpriteRendererProperties}]
 }
 Skirt_Fragments {
  emitterUpdateScript [
   &EmitterState()
   &SpawnRate(
    Spawn Probability=&FN_FrontEndFloatModifier(Gameplay=&Lerp_Float(A=0, B=1, Alpha=$System.Music_Reactive_preview), Front End=1)
    SpawnRate=&MakeCustomFloatFromBool(
     Bool=$User.IsSkydiving
     True Float=0
     False Float=&MakeCustomFloatFromBool(
      Bool=$User.bisInVehicle
      True Float=0
      False Float=&MakeCustomFloatFromBool(
       Bool=&SetBoolByFloatComparison(A=&VectorLength(Vector=$Engine.Owner.Velocity), B=1000)
       True Float=0
       False Float=15
      )
     )
    )
    Use Spawn Probability=true
   )
  ]
  particleSpawnScript [
   &InitializeParticle(
    Color=<rgba 1.5 0.045 0.2058 1>
    Lifetime Max=1
    Lifetime Min=0.5
    Lifetime Mode=::NewEnumerator1
    Mesh Scale Mode=::NewEnumerator1
    Mesh Uniform Scale Max=0.5
    Mesh Uniform Scale Min=0.15
    Sprite Size Mode=::NewEnumerator3
    Uniform Sprite Size=5
    UsePositionOffset=false
    Write Material Random=true
   )
   &SkeletalMeshLocation(Bone Sampling Mode=::NewEnumerator4, MaskSpawnGroup=false, Sampled Position Offset=<-35 0 0>, Skeletal Mesh=$System.Skeletal Mesh debris)
   &SkeletalMeshLocation(
    Bone / Socket Index=1
    Bone Sampling Mode=::NewEnumerator4
    Interpolate Mesh Sampling=false
    MaskSpawnGroup=false
    Orientation Sampling=::NewEnumerator1
    Position Sampling=::NewEnumerator2
    Skeletal Mesh=$System.Skeletal Mesh debris
   ) {module=SkeletalMeshLocation001}
   &CylinderLocation(Cylinder Height=10, Cylinder Radius=25, MaskSpawnGroup=false, Override Seed=false, Surface Only=true, UseOwnerScale=true)
   &AddVelocityFromPoint(
    Module.Use Falloff Distance=false
    Velocity Origin:NiagaraPosition=$Particles.SkeletalMeshLocation001.SampledPosition
    Velocity Strength=&RandomRangeFloat(Minimum=30, Maximum=40)
   )
   &InitialMeshOrientation(
    Rotation=&RandomRangeVector(Minimum=<0 0 0>, Maximum=<1 1 1>)
    Use Orientation Vector=true
    Use Rotation Vector=true
   ) {module=InitialMeshOrientation001}
   &CameraOffset(Camera Offset Amount=&RandomRangeFloat(Minimum=0, Maximum=20, Evaluation Type=::NewEnumerator1))
  ]
  particleUpdateScript [
   &Drag(Use Linear Drag=true, Use Rotational Drag=false)
   &CurlNoiseForce(
    Noise Strength=&FloatFromCurve(FloatCurve=DataInterface() {propertyValues {Curve {keys [<at 0 100>,<at 1 50>]}}})
    Pan Noise=false
   )
   &ParticleState()
   &GravityForce(Gravity=<0 0 -15>)
   &InheritSourceMovement(
    Applied Position Falloff Curve=DataInterface() {propertyValues {Curve {keys [{interpMode=RCIM_Cubic, tangentMode=RCTM_User, time=0, value=1}, {interpMode=RCIM_Cubic, tangentMode=RCTM_User, time=1, value=0}]}}}
    Applied Position Falloff Distance=200
    ApplyForce=false
    ApplyPosition=true
    ApplyPositionFalloff=true
    ApplyPositionFalloffCurve=true
   )
   &UpdateMeshOrientation(
    Rotation Rate=&FloatFromCurve(
     FloatCurve=DataInterface() {propertyValues {Curve {keys [
         {arriveTangent=-2.1159, interpMode=RCIM_Cubic, leaveTangent=-2.1159, tangentMode=RCTM_User, time=0, value=1}
         {interpMode=RCIM_Cubic, tangentMode=RCTM_Break, time=1, value=0.2}
        ]}}}
     Scale Curve=&RandomRangeFloat(Minimum=0.7, Maximum=1.5)
    )
    Rotation Vector=&RandomRangeVector(Minimum=<-1 -1 -1>, Maximum=<1 1 1>)
   )
   &SolveForcesAndVelocity(Clamp Velocity=false, Limit Acceleration=false)
   &ScaleMeshSize(Scale Factor=&VectorFromFloat(Value=&FloatFromCurve(FloatCurve=DataInterface() {propertyValues {Curve {keys [<at 0.2 1 RCIM_Cubic>, {arriveTangent=-3.8833, interpMode=RCIM_Cubic, tangentMode=RCTM_Break, time=1, value=0}]}}})))
   &ScaleColor(
    Scale Alpha=&Multiply_Float(
     A=&RandomRangeFloat(Minimum=0.2, Maximum=1)
     B=&Multiply_Float(
      A=&FloatFromCurve(FloatCurve=DataInterface() {propertyValues {Curve {keys [
           {interpMode=RCIM_Cubic, tangentMode=RCTM_User, time=0, value=0}
           {arriveTangent=-1, interpMode=RCIM_Cubic, leaveTangent=-1, time=0.2, value=0.95}
           {arriveTangent=-1.7637, interpMode=RCIM_Cubic, leaveTangent=-1.7637, tangentMode=RCTM_User, time=1, value=0}
          ]}}})
      B=&Multiply_Float(
       A=&Lerp_Float(A=0, B=1, Alpha=&ClampFloat(Float=$System.Music_Reactive_preview))
       B=$User.Debris Opacity
      )
     )
    )
    Scale RGB=&VectorFromFloat(Value=&Multiply_Float(
      A=&RandomRangeFloat(Minimum=5, Maximum=10)
      B=&Multiply_Float(
       A=$System.TOD
       B=&FloatFromCurve(
        FloatCurve=DataInterface() {propertyValues {Curve {keys [
            {interpMode=RCIM_Cubic, tangentMode=RCTM_User, time=0, value=-0.025}
            {interpMode=RCIM_Cubic, tangentMode=RCTM_User, time=0.1488, value=1.1054}
            {arriveTangent=-0.3951, interpMode=RCIM_Cubic, leaveTangent=-0.3951, time=0.2612, value=-0.15}
            {arriveTangent=0.1378, interpMode=RCIM_Cubic, leaveTangent=0.1378, time=0.3572, value=1.023}
            {arriveTangent=-0.0897, interpMode=RCIM_Cubic, leaveTangent=-0.0897, time=0.496, value=-0.1176}
            {arriveTangent=0.3675, interpMode=RCIM_Cubic, leaveTangent=0.3675, time=0.6139, value=1}
            {arriveTangent=-0.2002, interpMode=RCIM_Cubic, leaveTangent=-0.2002, time=0.7481, value=-0.025}
            <at 0.8636 0.95 RCIM_Cubic>
            {tangentMode=RCTM_User, time=1, value=-0.025}
           ]}}}
        Scale Curve=2
       )
      )
     ))
    ScaleA=true
    ScaleRGB=true
   )
  ]
  rendererClasses [&NiagaraMeshRendererProperties]
  renderers [{rendererClass=&NiagaraMeshRendererProperties, rendererIndex=0}]
 }
}
summary {
 systemName=NS_RareDelightSail_Skin
 emitters [
  {emitterName=Bone_Flow_Right_shoulder, rendererClasses [&NiagaraRibbonRendererProperties]}
  {emitterName=Bone_Flow_LeftArm, rendererClasses [&NiagaraRibbonRendererProperties]}
  {emitterName=Shoulder_Smoke, rendererClasses [&NiagaraSpriteRendererProperties]}
  {emitterName=Shoulder_Local, rendererClasses [&NiagaraMeshRendererProperties]}
  {emitterName=Drop_Flare, simTarget=GPUComputeSim, rendererClasses [&NiagaraSpriteRendererProperties]}
  {emitterName=Bone_Flow_RightArm, rendererClasses [&NiagaraRibbonRendererProperties]}
  {emitterName=Skirt_Fragments, rendererClasses [&NiagaraMeshRendererProperties]}
  {bEnabled=false, emitterName=Bone_Flow_LeftArm001, rendererClasses [&NiagaraRibbonRendererProperties]}
  {emitterName=Bone_Flow_LeftArm003, rendererClasses [&NiagaraRibbonRendererProperties]}
 ]
 userVariables [
  User.bIsFrontEnd {type=NiagaraBool}
  User.bIsFrontEndPreview {type=NiagaraBool}
  User.bisInVehicle {type=NiagaraBool}
  User.Debris Opacity {type=NiagaraFloat, defaultValue=1, description=""}
  User.IsSkydiving {type=NiagaraBool}
  User.Opacity {type=NiagaraFloat, defaultValue=1, description=""}
 ]
}
