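# Control Rig for the RareDelightSail medium female player body cosmetic, handling corrective deformation and skirt bone placement.
# The first sequence branch runs the shared DeformRig function, blends left and right skirt base bones between pelvis and thighs, and stores spherical pose reader outputs for each thigh.
# The second branch uses those values to weight skirt transform offsets, then constrains front and back center bases to average their left and right neighbors.
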ControlRigBlueprint /BRCosmetics/Characters/Player/Female/Medium/Bodies/F_MED_RareDelightSail/Meshes/F_MED_RareDelightSail_CtrlRig
graphs [
 RigVMFunctionLibrary {links [], nodes []}
 RigVMModel {nodes [
   RigUnit_BeginExecution = &RigUnit_BeginExecution()
   RigVMFunction_Sequence = &RigVMFunction_Sequence(ExecuteContext=@RigUnit_BeginExecution.ExecutePin)
   FN_DeformRig_FNC = RigVMFunctionReferenceNode(
    DeformRig_ Curve="(deformRig_l_upperarm_up_alpha_72_#1=False,deformRig_r_upperarm_up_alpha_73_#2=False,deformRig_l_upperarm_fr_alpha_74_#3=False,deformRig_r_upperarm_fr_alpha_75_CB51B8 ...(truncated)"
    DeformRig_Bool="(upperarm_1_#4=True,elbow_3_#5=True,wrist_7_#6=True,groin_8_#7=True,glute_17_#8=True,knee_9_40C408324CA261 ...(truncated)"
    DeformRig_Debug=(upperarm_debug_18_#4=False,wrist_debug_20_#6=False,lat_debug_26_#9=False,groinGlute_debug_27_#7=False)
    DeformRig_Value="(upperarm_up_18_#10=(Rotation=(X=0,Y=0,Z=0,W=1),Translation=(X=0,Y=0,Z=8),Scale3D=(X=0.7,Y=1,Z=1)),upperarm_fwd_fr_19_#11=(Rotation=(X=0,Y=0,Z=0,W=1),Translation=(X=0,Y=10,Z=0),Scale3D=(X=0.5,Y=1,Z= ...(truncated)"
    ExecuteContext=@RigVMFunction_Sequence.A
   )
   > &RigUnit_ParentConstraint(
    AdvancedSettings=(InterpolationType=Average,RotationOrderForFilter=XZY)
    Child=(Type=Bone,Name="dyn_skirt_fr_a_base_l")
    Filter=(TranslationFilter=(bX=false,bY=false,bZ=false),RotationFilter=(bX=True,bY=True,bZ=True),ScaleFilter=(bX=True,bY=True,bZ=True))
    Parents=((Item=(Type=Bone,Name="pelvis"),Weight=0.1),(Item=(Type=Bone,Name="thigh_l"),Weight=0.9))
    Weight=1
   )
   > &RigUnit_ParentConstraint(
    AdvancedSettings=(InterpolationType=Average,RotationOrderForFilter=XZY)
    Child=(Type=Bone,Name="dyn_skirt_side_a_base_l")
    Filter=(TranslationFilter=(bX=false,bY=false,bZ=false),RotationFilter=(bX=True,bY=True,bZ=True),ScaleFilter=(bX=True,bY=True,bZ=True))
    Parents=((Item=(Type=Bone,Name="pelvis"),Weight=0.4),(Item=(Type=Bone,Name="thigh_l"),Weight=0.6))
    Weight=1
   )
   > &RigUnit_ParentConstraint(
    AdvancedSettings=(InterpolationType=Average,RotationOrderForFilter=XZY)
    Child=(Type=Bone,Name="dyn_skirt_bk_a_base_l")
    Filter=(TranslationFilter=(bX=false,bY=false,bZ=false),RotationFilter=(bX=True,bY=True,bZ=True),ScaleFilter=(bX=True,bY=True,bZ=True))
    Parents=((Item=(Type=Bone,Name="pelvis"),Weight=0.8),(Item=(Type=Bone,Name="thigh_l"),Weight=0.2))
    Weight=1
   )
   > &RigUnit_ParentConstraint(
    AdvancedSettings=(InterpolationType=Average,RotationOrderForFilter=XZY)
    Child=(Type=Bone,Name="dyn_skirt_fr_a_base_r")
    Filter=(TranslationFilter=(bX=false,bY=false,bZ=false),RotationFilter=(bX=True,bY=True,bZ=True),ScaleFilter=(bX=True,bY=True,bZ=True))
    Parents=((Item=(Type=Bone,Name="pelvis"),Weight=0.1),(Item=(Type=Bone,Name="thigh_r"),Weight=0.9))
    Weight=1
   )
   > &RigUnit_ParentConstraint(
    AdvancedSettings=(InterpolationType=Average,RotationOrderForFilter=XZY)
    Child=(Type=Bone,Name="dyn_skirt_side_a_base_r")
    Filter=(TranslationFilter=(bX=false,bY=false,bZ=false),RotationFilter=(bX=True,bY=True,bZ=True),ScaleFilter=(bX=True,bY=True,bZ=True))
    Parents=((Item=(Type=Bone,Name="pelvis"),Weight=0.4),(Item=(Type=Bone,Name="thigh_r"),Weight=0.6))
    Weight=1
   )
   > &RigUnit_ParentConstraint(
    AdvancedSettings=(InterpolationType=Average,RotationOrderForFilter=XZY)
    Child=(Type=Bone,Name="dyn_skirt_bk_a_base_r")
    Filter=(TranslationFilter=(bX=false,bY=false,bZ=false),RotationFilter=(bX=True,bY=True,bZ=True),ScaleFilter=(bX=True,bY=True,bZ=True))
    Parents=((Item=(Type=Bone,Name="pelvis"),Weight=0.8),(Item=(Type=Bone,Name="thigh_r"),Weight=0.2))
    Weight=1
   )
   > n1 = &RigUnit_SphericalPoseReader(
    ActiveRegionScaleFactors=(PositiveWidth=3.5,NegativeWidth=0.5,PositiveHeight=3,NegativeHeight=3)
    ActiveRegionSize=0.1
    Debug=(bDrawDebug=true,DebugScale=25,DebugSegments=20,DebugThickness=0.25)
    DriverAxis=(X=-1,Y=0,Z=0)
    DriverItem=(Type=Bone,Name="thigh_l")
    FalloffRegionScaleFactors=(PositiveWidth=1,NegativeWidth=1,PositiveHeight=1,NegativeHeight=1)
    FalloffSize=0.45
    OptionalParentItem=(Type=Bone)
    RotationOffset=(X=90,Y=0,Z=-15)
   )
   > n2 = &RigUnit_SphericalPoseReader(
    ActiveRegionScaleFactors=(PositiveWidth=3.5,NegativeWidth=0.5,PositiveHeight=3,NegativeHeight=3)
    ActiveRegionSize=0.1
    Debug=(bDrawDebug=true,DebugScale=25,DebugSegments=20,DebugThickness=0.25)
    DriverAxis=(X=1,Y=0,Z=0)
    DriverItem=(Type=Bone,Name="thigh_r")
    FalloffRegionScaleFactors=(PositiveWidth=1,NegativeWidth=1,PositiveHeight=1,NegativeHeight=1)
    FalloffSize=0.45
    FlipHeightScaling=false
    FlipWidthScaling=true
    OptionalParentItem=(Type=Bone)
    RotationOffset=(X=-90,Y=0,Z=-15)
   )
   > RigVMVariableNode(Value=@n1.OutputParam) {variable=thigh l rot}
   > RigVMVariableNode(Value=@n2.OutputParam) {variable=thigh r rot}
   &RigUnit_ModifyTransforms(
    ItemToModify="((Item=(Type=Bone,Name=\"dyn_skirt_fr_a_base_l\"),Transform=(Rotation=(X=0,Y=0,Z=0,W=1),Translation=(X=1,Y=0,Z=2),Scale3D=(X=1,Y=1,Z=1))),(Item=(Type=Bone,Name=\"dyn_skirt_side_a_base_l\"),Transform=(Rotation=(X=0,Y=0,Z=0,W=1),Translation=(X=0,Y=0,Z=3),Scale3D ...(truncated)"
    WeightMinimum=0
    ExecutePin=@RigVMFunction_Sequence.B
    Weight=RigVMVariableNode() {variable=thigh l rot}.Value
   )
   > &RigUnit_ModifyTransforms(
    ItemToModify="((Item=(Type=Bone,Name=\"dyn_skirt_fr_a_base_r\"),Transform=(Rotation=(X=0,Y=0,Z=0,W=1),Translation=(X=-1,Y=0,Z=-3),Scale3D=(X=1,Y=1,Z=1))),(Item=(Type=Bone,Name=\"dyn_skirt_side_a_base_r\"),Transform=(Rotation=(X=0,Y=0,Z=0,W=1),Translation=(X=0,Y=0,Z=-4),Scal ...(truncated)"
    WeightMinimum=0
    Weight=RigVMVariableNode() {variable=thigh r rot}.Value
   )
   > &RigUnit_ParentConstraint(
    AdvancedSettings=(InterpolationType=Average,RotationOrderForFilter=XYZ)
    Child=(Type=Bone,Name="dyn_skirt_fr_a_base")
    Parents=((Item=(Type=Bone,Name="dyn_skirt_fr_a_l"),Weight=1),(Item=(Type=Bone,Name="dyn_skirt_fr_a_r"),Weight=1))
    Weight=1
   )
   > &RigUnit_ParentConstraint(
    AdvancedSettings=(InterpolationType=Average,RotationOrderForFilter=XYZ)
    Child=(Type=Bone,Name="dyn_skirt_bk_a_base")
    Parents=((Item=(Type=Bone,Name="dyn_skirt_bk_a_l"),Weight=1),(Item=(Type=Bone,Name="dyn_skirt_bk_a_r"),Weight=1))
    Weight=1
   )
  ]}
]
hierarchy [
 root {loc=<0 0 0>, rot=<0 0 0 1>}
 pelvis {parent=root, loc=<0 -0.5301 98.0571>, rot=<0.0127 0.707 -0.0127 -0.707>}
 spine_01 {parent=pelvis, loc=<6.1208 -0.9328 0>, rot=<0 0 0.034 -0.9994>}
 spine_02 {parent=spine_01, loc=<6.5996 -0.2095 0>, rot=<0 0 0.0238 -0.9997>}
 spine_03 {parent=spine_02, loc=<6.4914 -0.1237 0>, rot=<0 0 0.0149 -0.9999>}
 spine_04 {parent=spine_03, loc=<6.9598 0.0794 0>, rot=<0 0 -0.0132 -0.9999>}
 spine_05 {parent=spine_04, loc=<6.9987 0 0>, rot=<0 0 -0.0015 -1>}
 attach_cape {parent=spine_05, loc=<9.3216 -8.4501 0>, rot=<0.0616 -0.7044 -0.0616 -0.7044>}
 neck_01 {parent=spine_05, loc=<11.3138 0.1529 0>, rot=<0 0 -0.2003 -0.9797>}
 neck_02 {parent=neck_01, loc=<4.5627 0 0>, rot=<0 0 0.0504 -0.9987>}
 head {parent=neck_02, loc=<5.3914 0 0>, rot=<0 0 0.115 -0.9934>}
 attach_backpack {parent=spine_05, loc=<0 -8.35 0>, rot=<-0.0093 -0.707 0.0093 -0.707>}
 clavicle_l {parent=spine_05, loc=<5.4831 -0.1053 -3.2946>, rot=<0.0465 -0.7289 -0.009 -0.6829>}
 upperarm_l {parent=clavicle_l, loc=<9.8658 -1.2136 -0.4105>, rot=<0.0164 -0.3245 -0.0269 -0.9454>}
 lowerarm_l {parent=upperarm_l, loc=<26.465 -0.8345 0.259>, rot=<0.0036 0.0161 -0.2298 -0.9731>}
 hand_l {parent=lowerarm_l, loc=<22.7355 -0.3203 -0.6105>, rot=<-0.6784 -0.0223 0.0735 -0.7307>}
 thumb_01_l {parent=hand_l, loc=<1.3118 -0.3191 -2.459>, rot=<0.6546 -0.4042 0.0664 -0.6354>}
 thumb_02_l {parent=thumb_01_l, loc=<5.9336 -0.0448 0.0401>, rot=<-0.0037 0.0213 0.298 -0.9543>}
 thumb_03_l {parent=thumb_02_l, loc=<2.2223 0.3682 -0.0907>, rot=<0.0118 -0.013 0.0066 -0.9998>}
 index_metacarpal_l {parent=hand_l, loc=<1.6535 0.0183 -2.96>, rot=<0.0158 -0.0124 -0.0173 -0.9996>}
 index_01_l {parent=index_metacarpal_l, loc=<8.3008 0.0357 -0.0524>, rot=<-0.0044 0.0485 0.0924 -0.9945>}
 index_02_l {parent=index_01_l, loc=<3.3587 -0.3004 0.0452>, rot=<-0.002 0.0027 0.3225 -0.9466>}
 index_03_l {parent=index_02_l, loc=<2.5578 -0.123 -0.0252>, rot=<-0.002 0.0067 0.2558 -0.9667>}
 middle_metacarpal_l {parent=hand_l, loc=<2.1856 -0.1725 -0.5984>, rot=<0.006 0.0168 -0.0329 -0.9993>}
 middle_01_l {parent=middle_metacarpal_l, loc=<7.3275 -0.2447 0.0129>, rot=<-0.0208 0.0042 0.0858 -0.9961>}
 middle_02_l {parent=middle_01_l, loc=<4.0123 0.1401 0.1175>, rot=<-0.0218 0.0258 0.3512 -0.9357>}
 middle_03_l {parent=middle_02_l, loc=<2.6077 0.0534 0.1373>, rot=<0.0082 0.0188 0.1708 -0.9851>}
 ring_metacarpal_l {parent=hand_l, loc=<2.2074 -0.3625 1.3638>, rot=<-0.0284 0.0722 -0.0458 -0.9959>}
 ring_01_l {parent=ring_metacarpal_l, loc=<7.0295 -0.6197 -0.1446>, rot=<-0.0489 -0.0252 0.1892 -0.9804>}
 ring_02_l {parent=ring_01_l, loc=<3.6174 0.4647 -0.1027>, rot=<0 0 0.2806 -0.9598>}
 ring_03_l {parent=ring_02_l, loc=<2.4152 0.114 0.0199>, rot=<0 0 0.1189 -0.9929>}
 pinky_metacarpal_l {parent=hand_l, loc=<1.7616 -0.9753 3.3814>, rot=<-0.0397 0.0814 -0.069 -0.9935>}
 pinky_01_l {parent=pinky_metacarpal_l, loc=<7.1709 -0.4831 -0.1094>, rot=<-0.0654 -0.0035 0.2242 -0.9723>}
 pinky_02_l {parent=pinky_01_l, loc=<2.8274 0.0893 -0.1043>, rot=<0.0037 0.0011 0.2854 -0.9584>}
 pinky_03_l {parent=pinky_02_l, loc=<2.2046 0.0307 -0.0078>, rot=<-0.0002 0.0098 0.1445 -0.9895>}
 deform_wrist_dn_base_l {parent=hand_l, loc=<0 -1.5326 0>, rot=<0 0 0 -1>}
 deform_wrist_dn_l {parent=deform_wrist_dn_base_l, loc=<0 0 0>, rot=<0 0 0 1>}
 deform_wrist_up_base_l {parent=hand_l, loc=<0 2.3009 0>, rot=<0 0 0 -1>}
 deform_wrist_up_l {parent=deform_wrist_up_base_l, loc=<0 0 0>, rot=<0 0 0 1>}
 lowerarm_twist_02_l {parent=lowerarm_l, loc=<10.6707 0 0>, rot=<0 0 0 1>}
 lowerarm_twist_01_l {parent=lowerarm_l, loc=<18.316 0 0>, rot=<0 0 0 1>}
 deform_elbow_in_base_l {parent=lowerarm_l, loc=<0.1834 3.6062 -0.2572>, rot=<0 0 0 -1>}
 deform_elbow_in_l {parent=deform_elbow_in_base_l, loc=<0 0 0>, rot=<0 0 0 1>}
 deform_elbow_out_base_l {parent=lowerarm_l, loc=<0 -4 0>, rot=<0 0 0 -1>}
 deform_elbow_out_l {parent=deform_elbow_out_base_l, loc=<0 0 0>, rot=<0 0 0 1>}
 upperarm_twist_01_l {parent=upperarm_l, loc=<7.2135 -0.1134 0>, rot=<0 0 0 1>}
 deform_upperarm_bk_base_l {parent=upperarm_twist_01_l, loc=<-6.4053 -4.3746 0.314>, rot=<0 0 0 -1>}
 deform_upperarm_bk_l {parent=deform_upperarm_bk_base_l, loc=<0 0 0>, rot=<0 0 0 1>}
 deform_upperarm_fr_base_l {parent=upperarm_twist_01_l, loc=<-5.6312 5.092 0.3354>, rot=<0 0 0 -1>}
 deform_upperarm_fr_l {parent=deform_upperarm_fr_base_l, loc=<0 0 0>, rot=<0 0 0 1>}
 deform_upperarm_up_base_l {parent=upperarm_twist_01_l, loc=<-8.891 0.4825 3.9517>, rot=<0 0 0 -1>}
 deform_upperarm_up_l {parent=deform_upperarm_up_base_l, loc=<0 0 0>, rot=<0 0 0 1>}
 upperarm_bk_volume_base_l {parent=upperarm_twist_01_l, loc=<-6.4053 -4.3746 0.314>, rot=<0 0 0 -1>}
 upperarm_bk_volume_l {parent=upperarm_bk_volume_base_l, loc=<0 0 0>, rot=<0 0 0 1>}
 upperarm_up_volume_base_l {parent=upperarm_twist_01_l, loc=<-8.891 0.4825 3.9517>, rot=<0 0 0 -1>}
 upperarm_up_volume_l {parent=upperarm_up_volume_base_l, loc=<0 0 0>, rot=<0 0 0 1>}
 dyn_shard_shoulder_d_base_l {parent=upperarm_twist_01_l, loc=<3.9836 -3.2714 12.4934>, rot=<-0.4842 -0.7152 0.0327 -0.5029>}
 dyn_shard_shoulder_d_l {parent=dyn_shard_shoulder_d_base_l, loc=<0 0 0>, rot=<0 0 0 1>}
 dyn_shard_shoulder_c_base_l {parent=upperarm_twist_01_l, loc=<-0.6235 -0.8239 7.8096>, rot=<-0.4077 -0.7477 0.0931 -0.5157>}
 dyn_shard_shoulder_c_l {parent=dyn_shard_shoulder_c_base_l, loc=<0 0 0>, rot=<0 0 0 1>}
 dyn_shard_shoulder_e_base_l {parent=upperarm_twist_01_l, loc=<-0.1469 -7.3698 21.4978>, rot=<-0.4077 -0.7477 0.0931 -0.5157>}
 dyn_shard_shoulder_e_l {parent=dyn_shard_shoulder_e_base_l, loc=<0 0 0>, rot=<0 0 0 1>}
 dyn_shard_shoulder_f_base_l {parent=upperarm_twist_01_l, loc=<-9.9638 -9.8667 21.3414>, rot=<-0.5093 -0.603 0.1234 -0.6014>}
 dyn_shard_shoulder_f_l {parent=dyn_shard_shoulder_f_base_l, loc=<0 0 0>, rot=<0 0 0 1>}
 dyn_shard_shoulder_h_base_l {parent=upperarm_twist_01_l, loc=<-7.2911 -5.5515 6.4869>, rot=<-0.3073 -0.6512 0.3074 -0.6221>}
 dyn_shard_shoulder_h_l {parent=dyn_shard_shoulder_h_base_l, loc=<0 0 0>, rot=<0 0 0 1>}
 upperarm_twist_02_l {parent=upperarm_l, loc=<19.7998 -0.1134 0>, rot=<0 0 0 1>}
 dyn_shard_shoulder_g_base_l {parent=clavicle_l, loc=<8.9077 -8.5497 0.4292>, rot=<-0.2093 -0.7688 0.4185 -0.4358>}
 dyn_shard_shoulder_g_l {parent=dyn_shard_shoulder_g_base_l, loc=<0 0 0>, rot=<0 0 0 1>}
 dyn_shard_shoulder_a_base_l {parent=clavicle_l, loc=<12.0909 7.2796 3.6242>, rot=<-0.2896 -0.8285 -0.0472 -0.4769>}
 dyn_shard_shoulder_a_l {parent=dyn_shard_shoulder_a_base_l, loc=<0 0 0>, rot=<0 0 0 1>}
 dyn_shard_shoulder_b_base_l {parent=clavicle_l, loc=<13.0666 -1.0356 11.2835>, rot=<-0.3426 -0.8154 -0.0135 -0.4664>}
 dyn_shard_shoulder_b_l {parent=dyn_shard_shoulder_b_base_l, loc=<0 0 0>, rot=<0 0 0 1>}
 clavicle_r {parent=spine_05, loc=<5.4831 -0.1053 3.2946>, rot=<0.7289 0.0465 -0.6829 0.009>}
 upperarm_r {parent=clavicle_r, loc=<-9.8658 1.2136 0.4104>, rot=<0.0164 -0.3245 -0.0269 -0.9454>}
 lowerarm_r {parent=upperarm_r, loc=<-26.4651 0.8345 -0.259>, rot=<0.0036 0.0161 -0.2298 -0.9731>}
 hand_r {parent=lowerarm_r, loc=<-22.7356 0.3203 0.6106>, rot=<-0.6784 -0.0223 0.0735 -0.7307>}
 thumb_01_r {parent=hand_r, loc=<-1.3119 0.3192 2.459>, rot=<0.6546 -0.4042 0.0664 -0.6354>}
 thumb_02_r {parent=thumb_01_r, loc=<-5.9335 0.0447 -0.0401>, rot=<-0.0037 0.0213 0.298 -0.9543>}
 thumb_03_r {parent=thumb_02_r, loc=<-2.2223 -0.3682 0.0907>, rot=<0.0118 -0.013 0.0066 -0.9998>}
 index_metacarpal_r {parent=hand_r, loc=<-1.6534 -0.0181 2.9601>, rot=<0.0158 -0.0124 -0.0173 -0.9996>}
 index_01_r {parent=index_metacarpal_r, loc=<-8.3008 -0.0357 0.0524>, rot=<-0.0044 0.0485 0.0924 -0.9945>}
 index_02_r {parent=index_01_r, loc=<-3.3589 0.3003 -0.0453>, rot=<-0.002 0.0027 0.3225 -0.9466>}
 index_03_r {parent=index_02_r, loc=<-2.5578 0.123 0.0252>, rot=<-0.002 0.0067 0.2558 -0.9667>}
 middle_metacarpal_r {parent=hand_r, loc=<-2.1856 0.1726 0.5984>, rot=<0.006 0.0168 -0.0329 -0.9993>}
 middle_01_r {parent=middle_metacarpal_r, loc=<-7.3273 0.2447 -0.0129>, rot=<-0.0208 0.0042 0.0858 -0.9961>}
 middle_02_r {parent=middle_01_r, loc=<-4.0124 -0.14 -0.1175>, rot=<-0.0218 0.0258 0.3512 -0.9357>}
 middle_03_r {parent=middle_02_r, loc=<-2.6077 -0.0533 -0.1373>, rot=<0.0082 0.0188 0.1708 -0.9851>}
 ring_metacarpal_r {parent=hand_r, loc=<-2.2073 0.3627 -1.3638>, rot=<-0.0284 0.0722 -0.0458 -0.9959>}
 ring_01_r {parent=ring_metacarpal_r, loc=<-7.0297 0.6197 0.1446>, rot=<-0.0489 -0.0252 0.1892 -0.9804>}
 ring_02_r {parent=ring_01_r, loc=<-3.6171 -0.4648 0.1027>, rot=<0 0 0.2806 -0.9598>}
 ring_03_r {parent=ring_02_r, loc=<-2.4153 -0.1141 -0.0199>, rot=<0 0 0.1189 -0.9929>}
 pinky_metacarpal_r {parent=hand_r, loc=<-1.7617 0.9753 -3.3814>, rot=<-0.0397 0.0814 -0.069 -0.9935>}
 pinky_01_r {parent=pinky_metacarpal_r, loc=<-7.1707 0.4832 0.1094>, rot=<-0.0654 -0.0035 0.2242 -0.9723>}
 pinky_02_r {parent=pinky_01_r, loc=<-2.8274 -0.0892 0.1043>, rot=<0.0037 0.0011 0.2854 -0.9584>}
 pinky_03_r {parent=pinky_02_r, loc=<-2.2045 -0.031 0.0078>, rot=<-0.0002 0.0098 0.1445 -0.9895>}
 deform_wrist_dn_base_r {parent=hand_r, loc=<-0.0001 1.5328 0>, rot=<0 0 0 -1>}
 deform_wrist_dn_r {parent=deform_wrist_dn_base_r, loc=<0 0 0>, rot=<0 0 0 1>}
 deform_wrist_up_base_r {parent=hand_r, loc=<-0.0002 -2.3007 -0.0001>, rot=<0 0 0 -1>}
 deform_wrist_up_r {parent=deform_wrist_up_base_r, loc=<0 0 0>, rot=<0 0 0 1>}
 lowerarm_twist_02_r {parent=lowerarm_r, loc=<-10.6707 0 0>, rot=<0 0 0 1>}
 dyn_shard_elbow_a_base_r {parent=lowerarm_twist_02_r, loc=<3.1568 2.1021 -7.6753>, rot=<0.2691 -0.2408 0.931 -0.0525>}
 dyn_shard_elbow_a_r {parent=dyn_shard_elbow_a_base_r, loc=<0 0 0>, rot=<0 0 0 1>}
 dyn_shard_elbow_b_base_r {parent=lowerarm_twist_02_r, loc=<11.9636 -1.0108 -7.5942>, rot=<0.0381 -0.2476 0.9681 -0.0097>}
 dyn_shard_elbow_b_r {parent=dyn_shard_elbow_b_base_r, loc=<0 0 0>, rot=<0 0 0 1>}
 lowerarm_twist_01_r {parent=lowerarm_r, loc=<-18.3161 0 0>, rot=<0 0 0 1>}
 deform_elbow_in_base_r {parent=lowerarm_r, loc=<-0.1834 -3.6063 0.2572>, rot=<0 0 0 -1>}
 deform_elbow_in_r {parent=deform_elbow_in_base_r, loc=<0 0 0>, rot=<0 0 0 1>}
 deform_elbow_out_base_r {parent=lowerarm_r, loc=<0 4 -0.0001>, rot=<0 0 0 -1>}
 deform_elbow_out_r {parent=deform_elbow_out_base_r, loc=<0 0 0>, rot=<0 0 0 1>}
 upperarm_twist_01_r {parent=upperarm_r, loc=<-7.2135 0.1134 0>, rot=<0 0 0 1>}
 deform_upperarm_bk_base_r {parent=upperarm_twist_01_r, loc=<6.4052 4.3746 -0.314>, rot=<0 0 0 -1>}
 deform_upperarm_bk_r {parent=deform_upperarm_bk_base_r, loc=<0 0 0>, rot=<0 0 0 1>}
 deform_upperarm_fr_base_r {parent=upperarm_twist_01_r, loc=<5.6311 -5.092 -0.3354>, rot=<0 0 0 -1>}
 deform_upperarm_fr_r {parent=deform_upperarm_fr_base_r, loc=<0 0 0>, rot=<0 0 0 1>}
 deform_upperarm_up_base_r {parent=upperarm_twist_01_r, loc=<8.891 -0.4825 -3.9517>, rot=<0 0 0 -1>}
 deform_upperarm_up_r {parent=deform_upperarm_up_base_r, loc=<0 0 0>, rot=<0 0 0 1>}
 upperarm_twist_02_r {parent=upperarm_r, loc=<-19.7998 0.1134 0>, rot=<0 0 0 1>}
 deform_pec_aim_base_l {parent=spine_04, loc=<3.2456 9.3309 -1.2237>, rot=<0.0603 -0.4033 0.3318 -0.8507>}
 deform_pec_aim_l {parent=deform_pec_aim_base_l, loc=<0 0 0>, rot=<0 0 0 1>}
 deform_pec_base_l {parent=deform_pec_aim_l, loc=<9.3665 0 0>, rot=<0 0 0 1>}
 deform_pec_l {parent=deform_pec_base_l, loc=<0 0 0>, rot=<0 0 0 1>}
 deform_pec_aim_base_r {parent=spine_04, loc=<3.2452 9.3309 1.2237>, rot=<-0.4033 -0.0603 0.8507 0.3318>}
 deform_pec_aim_r {parent=deform_pec_aim_base_r, loc=<0 0 0>, rot=<0 0 0 1>}
 deform_pec_base_r {parent=deform_pec_aim_r, loc=<-9.3667 -0.0003 -0.0004>, rot=<0 0 0 1>}
 deform_pec_r {parent=deform_pec_base_r, loc=<0 0 0>, rot=<0 0 0 1>}
 dyn_shard_torso_a_base_l {parent=spine_02, loc=<3.4826 -5.4296 -11.8095>, rot=<0.1606 0.1043 0.9708 0.1441>}
 dyn_shard_torso_a_l {parent=dyn_shard_torso_a_base_l, loc=<0 0 0>, rot=<0 0 0 1>}
 thigh_l {parent=pelvis, loc=<-1.0855 -0.2109 -8.7053>, rot=<0.0009 0.0355 0.0155 -0.9993>}
 calf_l {parent=thigh_l, loc=<-42.5284 -0.0019 -0.7297>, rot=<0.0006 -0.0161 -0.0391 -0.9991>}
 foot_l {parent=calf_l, loc=<-46.7402 -2.5014 -0.592>, rot=<0.0427 -0.0238 0.0353 -0.9982>}
 ball_l {parent=foot_l, loc=<-3.5658 10.8066 -0.1544>, rot=<0.0001 -0.0001 -0.7063 -0.7079>}
 calf_twist_02_l {parent=calf_l, loc=<-19.1546 -1.7943 -0.2638>, rot=<0.0237 0.0095 -0.0035 -0.9997>}
 calf_twist_01_l {parent=calf_l, loc=<-35.8483 -1.9197 -0.5794>, rot=<0.0237 0.0095 -0.0035 -0.9997>}
 deform_knee_in_base_l {parent=calf_l, loc=<0.8031 -5.0231 0.2456>, rot=<0 0 0 -1>}
 deform_knee_in_l {parent=deform_knee_in_base_l, loc=<0 0 0>, rot=<0 0 0 1>}
 deform_knee_out_base_l {parent=calf_l, loc=<0 4 0>, rot=<0 0 0 -1>}
 deform_knee_out_l {parent=deform_knee_out_base_l, loc=<0 0 0>, rot=<0 0 0 1>}
 dyn_shard_knee_a_base_l {parent=calf_l, loc=<-6.6355 -8.3828 -8.7764>, rot=<-0.1145 -0.5631 0.7916 0.2079>}
 dyn_shard_knee_a_l {parent=dyn_shard_knee_a_base_l, loc=<0 0 0>, rot=<0 0 0 1>}
 thigh_twist_01_l {parent=thigh_l, loc=<-27.952 0 0>, rot=<0.0013 0 0 -1>}
 deform_groin_base_l {parent=thigh_l, loc=<0.4801 5.6776 -0.2486>, rot=<0 0 0 -1>}
 deform_groin_l {parent=deform_groin_base_l, loc=<0 0 0>, rot=<0 0 0 1>}
 thigh_r {parent=pelvis, loc=<-1.0853 -0.2109 8.7053>, rot=<0.0355 -0.0009 0.9993 0.0155>}
 calf_r {parent=thigh_r, loc=<42.5286 0.0019 0.7297>, rot=<0.0006 -0.0161 -0.0391 -0.9991>}
 foot_r {parent=calf_r, loc=<46.7403 2.5014 0.592>, rot=<0.0427 -0.0238 0.0353 -0.9982>}
 ball_r {parent=foot_r, loc=<3.5658 -10.8066 0.1544>, rot=<0.0001 -0.0001 -0.7063 -0.7079>}
 calf_twist_02_r {parent=calf_r, loc=<19.1546 1.7943 0.2638>, rot=<0.0237 0.0095 -0.0035 -0.9997>}
 calf_twist_01_r {parent=calf_r, loc=<35.8485 1.9197 0.5794>, rot=<0.0237 0.0095 -0.0035 -0.9997>}
 deform_knee_in_base_r {parent=calf_r, loc=<-0.8032 5.0231 -0.2457>, rot=<0 0 0 -1>}
 deform_knee_in_r {parent=deform_knee_in_base_r, loc=<0 0 0>, rot=<0 0 0 1>}
 deform_knee_out_base_r {parent=calf_r, loc=<-0.0002 -4 0>, rot=<0 0 0 -1>}
 deform_knee_out_r {parent=deform_knee_out_base_r, loc=<0 0 0>, rot=<0 0 0 1>}
 thigh_twist_01_r {parent=thigh_r, loc=<27.952 0 0>, rot=<0.0013 0 0 -1>}
 deform_groin_base_r {parent=thigh_r, loc=<-0.48 -5.6776 0.2486>, rot=<0 0 0 -1>}
 deform_groin_r {parent=deform_groin_base_r, loc=<0 0 0>, rot=<0 0 0 1>}
 deform_glute_base_l {parent=pelvis, loc=<-7.3466 -9.0213 -9.1245>, rot=<0.0172 0.031 -0.4477 -0.8935>}
 deform_glute_l {parent=deform_glute_base_l, loc=<0 0 0>, rot=<0 0 0 1>}
 deform_glute_base_r {parent=pelvis, loc=<-7.3464 -9.0213 9.1245>, rot=<0.031 -0.0172 0.8935 -0.4477>}
 deform_glute_r {parent=deform_glute_base_r, loc=<0 0 0>, rot=<0 0 0 1>}
 dyn_skirt_side_a_base_l {parent=pelvis, loc=<5.2879 -1.0933 -12.4186>, rot=<-0.0176 -0.9831 0.0033 -0.1822>}
 dyn_skirt_side_a_l {parent=dyn_skirt_side_a_base_l, loc=<0 0 0>, rot=<0 0 0 1>}
 dyn_skirt_side_b_l {parent=dyn_skirt_side_a_l, loc=<9.9932 0 0>, rot=<0 0 0 1>}
 dyn_shard_skirt_d_base_l {parent=dyn_skirt_side_b_l, loc=<14.002 0.5753 1.9912>, rot=<0.9916 0.002 0.0484 -0.1203>}
 dyn_shard_skirt_d_l {parent=dyn_shard_skirt_d_base_l, loc=<0 0 0>, rot=<0 0 0 1>}
 dyn_shard_skirt_e_base_l {parent=dyn_skirt_side_b_l, loc=<13.3316 -1.982 -0.0175>, rot=<0.9756 -0.085 -0.012 -0.2019>}
 dyn_shard_skirt_e_l {parent=dyn_shard_skirt_e_base_l, loc=<0 0 0>, rot=<0 0 0 1>}
 dyn_skirt_bk_a_base_l {parent=pelvis, loc=<5.2752 -7.3968 -8.9596>, rot=<0.0483 -0.8854 0.439 -0.1449>}
 dyn_skirt_bk_a_l {parent=dyn_skirt_bk_a_base_l, loc=<0 0 0>, rot=<0 0 0 1>}
 dyn_skirt_bk_b_l {parent=dyn_skirt_bk_a_l, loc=<9.9932 0 0>, rot=<0 0 0 1>}
 dyn_fx_wide_a_l {parent=dyn_skirt_bk_b_l, loc=<14.535 2.6577 -0.3644>, rot=<0.0081 0.0327 -0.0339 -0.9989>}
 dyn_fx_wide_b_l {parent=dyn_fx_wide_a_l, loc=<21.4683 0 0>, rot=<0 0 0 1>}
 dyn_fx_wide_c_l {parent=dyn_fx_wide_b_l, loc=<22.3024 0 0.1309>, rot=<0 0.0273 0 -0.9996>}
 dyn_fx_narrow_a_l {parent=dyn_skirt_bk_b_l, loc=<13.993 2.6717 2.38>, rot=<0.0059 0.0972 -0.0344 -0.9947>}
 dyn_fx_narrow_b_l {parent=dyn_fx_narrow_a_l, loc=<21.4683 0 0>, rot=<0 0 0 1>}
 dyn_skirt_fr_a_base_l {parent=pelvis, loc=<5.392 6.3395 -8.312>, rot=<-0.0131 -0.8415 -0.5122 -0.1713>}
 dyn_skirt_fr_a_l {parent=dyn_skirt_fr_a_base_l, loc=<0 0 0>, rot=<0 0 0 1>}
 dyn_skirt_fr_b_l {parent=dyn_skirt_fr_a_l, loc=<9.9932 0 0>, rot=<0 0 0 1>}
 dyn_shard_skirt_a_base_l {parent=dyn_skirt_fr_b_l, loc=<14.4144 5.5939 1.6603>, rot=<0.9786 0.1485 -0.0104 0.1422>}
 dyn_shard_skirt_a_l {parent=dyn_shard_skirt_a_base_l, loc=<0 0 0>, rot=<0 0 0 1>}
 dyn_shard_skirt_b_base_l {parent=dyn_skirt_fr_b_l, loc=<12.9282 -0.437 0.8682>, rot=<0.9893 0.0614 -0.0004 -0.1322>}
 dyn_shard_skirt_b_l {parent=dyn_shard_skirt_b_base_l, loc=<0 0 0>, rot=<0 0 0 1>}
 dyn_shard_skirt_c_base_l {parent=dyn_skirt_fr_b_l, loc=<15.955 -5.8974 -2.9355>, rot=<0.9451 0.0352 0.0043 -0.3248>}
 dyn_shard_skirt_c_l {parent=dyn_shard_skirt_c_base_l, loc=<0 0 0>, rot=<0 0 0 1>}
 dyn_skirt_side_a_base_r {parent=pelvis, loc=<5.2881 -1.0933 12.4186>, rot=<0.9831 -0.0176 -0.1822 -0.0033>}
 dyn_skirt_side_a_r {parent=dyn_skirt_side_a_base_r, loc=<0 0 0>, rot=<0 0 0 1>}
 dyn_skirt_side_b_r {parent=dyn_skirt_side_a_r, loc=<-9.9934 0 0.0001>, rot=<0 0 0 1>}
 dyn_skirt_side_c_r {parent=dyn_skirt_side_b_r, loc=<-12.6529 0 0>, rot=<0 0 0 1>}
 dyn_shard_skirt_f_base_r {parent=dyn_skirt_side_c_r, loc=<-20.4255 -6.6149 1.6757>, rot=<-0.0411 0.2493 0.9675 0.0044>}
 dyn_shard_skirt_f_r {parent=dyn_shard_skirt_f_base_r, loc=<0 0 0>, rot=<0 0 0 1>}
 dyn_shard_skirt_e_base_r {parent=dyn_skirt_side_c_r, loc=<-11.9453 -9.6882 -3.1854>, rot=<-0.1283 0.2382 0.9626 -0.0174>}
 dyn_shard_skirt_e_r {parent=dyn_shard_skirt_e_base_r, loc=<0 0 0>, rot=<0 0 0 1>}
 dyn_shard_skirt_g_base_r {parent=dyn_skirt_side_b_r, loc=<-6.6056 -4.3461 -1.9976>, rot=<-0.1556 0.0889 0.9823 -0.0543>}
 dyn_shard_skirt_g_r {parent=dyn_shard_skirt_g_base_r, loc=<0 0 0>, rot=<0 0 0 1>}
 dyn_shard_skirt_j_base_r {parent=dyn_skirt_side_a_r, loc=<-3.7528 -0.0134 -4.5593>, rot=<-0.0894 -0.0847 0.9919 -0.0304>}
 dyn_shard_skirt_j_r {parent=dyn_shard_skirt_j_base_r, loc=<0 0 0>, rot=<0 0 0 1>}
 dyn_skirt_bk_a_base_r {parent=pelvis, loc=<5.2753 -7.3968 8.9596>, rot=<0.8839 0.0395 -0.1522 -0.4404>}
 dyn_skirt_bk_a_r {parent=dyn_skirt_bk_a_base_r, loc=<0 0 0>, rot=<0 0 0 1>}
 dyn_skirt_bk_b_r {parent=dyn_skirt_bk_a_r, loc=<-9.9933 0 0>, rot=<0 0 0 1>}
 dyn_skirt_bk_c_r {parent=dyn_skirt_bk_b_r, loc=<-15.6254 -3.7817 0.9235>, rot=<0 0 0 1>}
 dyn_shard_skirt_h_base_r {parent=dyn_skirt_bk_c_r, loc=<-8.6918 -2.5983 -1.422>, rot=<0.0229 0.0695 0.9968 -0.0317>}
 dyn_shard_skirt_h_r {parent=dyn_shard_skirt_h_base_r, loc=<0 0 0>, rot=<0 0 0 1>}
 dyn_shard_skirt_i_base_r {parent=dyn_skirt_bk_b_r, loc=<0.2229 -0.6058 -3.9109>, rot=<-0.0332 -0.0064 0.9993 0.0185>}
 dyn_shard_skirt_i_r {parent=dyn_shard_skirt_i_base_r, loc=<0 0 0>, rot=<0 0 0 1>}
 dyn_fx_wide_a_r {parent=dyn_skirt_bk_b_r, loc=<-14.5837 -2.1289 0.5199>, rot=<0.0118 0.0296 -0.0234 -0.9992>}
 dyn_fx_wide_b_r {parent=dyn_fx_wide_a_r, loc=<-21.4683 0 0>, rot=<0 0 0 1>}
 dyn_fx_wide_c_r {parent=dyn_fx_wide_b_r, loc=<-22.3024 0 -0.1309>, rot=<0 0.0273 0 -0.9996>}
 dyn_fx_narrow_a_r {parent=dyn_skirt_bk_b_r, loc=<-14.0579 -2.1736 -2.2272>, rot=<0.0103 0.0942 -0.0241 -0.9952>}
 dyn_fx_narrow_b_r {parent=dyn_fx_narrow_a_r, loc=<-21.4683 0 0.0001>, rot=<0 0 0 1>}
 dyn_skirt_fr_a_base_r {parent=pelvis, loc=<5.3924 6.3395 8.312>, rot=<-0.8443 0.0157 0.1524 -0.5135>}
 dyn_skirt_fr_a_r {parent=dyn_skirt_fr_a_base_r, loc=<0 0 0>, rot=<0 0 0 1>}
 dyn_skirt_fr_b_r {parent=dyn_skirt_fr_a_r, loc=<-9.9936 0 0.0001>, rot=<0 0 0 1>}
 dyn_skirt_fr_c_r {parent=dyn_skirt_fr_b_r, loc=<-12.6529 0 0>, rot=<0 0 0 1>}
 dyn_shard_skirt_c_base_r {parent=dyn_skirt_fr_c_r, loc=<-22.9051 -1.7726 -2.4473>, rot=<-0.0425 -0.0803 0.9959 0.0011>}
 dyn_shard_skirt_c_r {parent=dyn_shard_skirt_c_base_r, loc=<0 0 0>, rot=<0 0 0 1>}
 dyn_shard_skirt_b_base_r {parent=dyn_skirt_fr_c_r, loc=<-11.2068 0.0157 -2.3948>, rot=<-0.0946 0.0289 0.995 0.0159>}
 dyn_shard_skirt_b_r {parent=dyn_shard_skirt_b_base_r, loc=<0 0 0>, rot=<0 0 0 1>}
 dyn_shard_skirt_a_base_r {parent=dyn_skirt_fr_b_r, loc=<-16.1796 -8.4489 -1.2314>, rot=<-0.0736 0.0767 0.9943 0.0094>}
 dyn_shard_skirt_a_r {parent=dyn_shard_skirt_a_base_r, loc=<0 0 0>, rot=<0 0 0 1>}
 dyn_shard_skirt_d_base_r {parent=dyn_skirt_fr_b_r, loc=<-11.6886 5.5106 -1.3086>, rot=<-0.0453 -0.2811 0.9503 0.1259>}
 dyn_shard_skirt_d_r {parent=dyn_shard_skirt_d_base_r, loc=<0 0 0>, rot=<0 0 0 1>}
 dyn_skirt_fr_a_base {parent=pelvis, loc=<5.4025 8.5979 -0.0005>, rot=<0.0627 0.7032 0.7041 0.0758>}
 dyn_skirt_fr_a {parent=dyn_skirt_fr_a_base, loc=<0 0 0>, rot=<-0.0001 -0.0002 -0.0004 -1>}
 dyn_skirt_fr_b {parent=dyn_skirt_fr_a, loc=<9.9932 0 0>, rot=<0 0 0 1>}
 dyn_skirt_bk_a_base {parent=pelvis, loc=<5.3282 -9.8279 -0.0004>, rot=<0.0938 -0.7021 0.7007 -0.0847>}
 dyn_skirt_bk_a {parent=dyn_skirt_bk_a_base, loc=<0 0 0>, rot=<-0.0003 -0.0005 -0.0011 -1>}
 dyn_skirt_bk_b {parent=dyn_skirt_bk_a, loc=<9.9932 0 0>, rot=<0 0 0 1>}
 dyn_fx_wide_a {parent=dyn_skirt_bk_b, loc=<15.6965 0.2598 1.0357>, rot=<0 0.0519 0 -0.9987>}
 dyn_fx_wide_b {parent=dyn_fx_wide_a, loc=<21.4683 0 0>, rot=<0 0 0 1>}
 dyn_fx_wide_c {parent=dyn_fx_wide_b, loc=<22.3024 0 0.1309>, rot=<0 0.0273 0 -0.9996>}
 dyn_fx_narrow_a {parent=dyn_skirt_bk_b, loc=<15.1459 0.2357 3.8902>, rot=<0 0.0952 0 -0.9955>}
 dyn_fx_narrow_b {parent=dyn_fx_narrow_a, loc=<21.4683 0 0>, rot=<0 0 0 1>}
 dyn_fx_narrow_c {parent=dyn_fx_narrow_b, loc=<19.906 0 0>, rot=<0.0027 -0.0347 0.0013 -0.9994>}
]
variables [
 thigh l rot {type=double, default=0}
 thigh r rot {type=double, default=0}
]
