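# Post-animation AnimBP for the F_MED_RareDelightSail cosmetic body, driving secondary motion on skirt, shards and FX bones.
# It copies the parent mesh pose through a Control Rig, two RigidBody simulations and skirt constraints into a cached pose.
# Weighted constraints anchor shards to skirt and shoulder bones before AnimDynamics adds springy motion.
# Sprung FX bones follow from a crystal cache; state flags adjust bones for skydiving, vehicles and front end.
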AnimBlueprint /BRCosmetics/Characters/Player/Female/Medium/Bodies/F_MED_RareDelightSail/Meshes/F_MED_RareDelightSail_AnimBP
target_skeleton=/Game/Characters/Player/Female/Medium/Bodies/F_MED_RareDelightSail/Meshes/F_MED_RareDelightSail_Skeleton
graphs [
 AnimGraph {nodes [
   n1 = "Constraint - dyn_shard_shoulder_a_base_l (PPPP)"!(
    ComponentPose
    Alpha=1
    ConstraintWeights_0=0.45
    ConstraintWeights_1:float=0.2
    ConstraintWeights_2:float=0.1
    ConstraintWeights_3:float=0.25
   ) -> (Pose) {
    alphaBoolBlend {blendOption=Linear}
    boneToModify {boneName=dyn_shard_shoulder_a_base_l}
    constraintSetup [
     {targetBone {boneName=upperarm_twist_01_l}}
     {targetBone {boneName=spine_05}}
     {targetBone {boneName=clavicle_l}}
     {targetBone {boneName=deform_upperarm_fr_l}}
    ]
    constraintWeights [1,1,1,1]
   }
   "Constraint - dyn_shard_shoulder_b_base_l (PPPPP)"!(
    ComponentPose=@n1
    Alpha=1
    ConstraintWeights_0=0.2
    ConstraintWeights_1:float=0.45
    ConstraintWeights_2:float=0.2
    ConstraintWeights_3:float=0
    ConstraintWeights_4:float=0.15
   ) -> (Pose) {
    alphaBoolBlend {blendOption=Linear}
    boneToModify {boneName=dyn_shard_shoulder_b_base_l}
    constraintSetup [
     {targetBone {boneName=deform_upperarm_up_l}}
     {targetBone {boneName=clavicle_l}}
     {targetBone {boneName=upperarm_twist_01_l}}
     {targetBone {boneName=deform_upperarm_fr_l}}
     {targetBone {boneName=spine_05}}
    ]
    constraintWeights [1,1,1,1,1]
   }
   n2 = Anim Dynamics - Bone: dyn_fx_wide_a_r(ComponentPose=Local To Component(LocalPose=Use cached pose 'CrystalCachedPose'()), Alpha=1, ExternalForce="0,0,0") {
    angularDampingOverride=1
    angularSpringConstant=20000
    bAngularSpring=true
    bGravityOverrideInSimSpace=true
    bOverrideAngularDamping=true
    bUseGravityOverride=true
    gravityScale=0.35
    simulationSpace=RootRelative
    alphaBoolBlend {blendOption=Linear}
    boundBone {boneName=dyn_fx_wide_a_r}
    componentLinearAccScale=<1.3 1.3 1.3>
    componentLinearVelScale=<1.3 1.3 1.3>
    gravityOverride {z=-100}
    physicsBodyDefinitions [{
      sphereCollisionRadius=0
      boundBone {boneName=dyn_fx_wide_a_r}
      boxExtents=<20 30 30>
      constraintSetup {angularLimitsMax=<15 0 15>, angularLimitsMin=<-15 -30 -15>, angularTarget=<1 0 0>}
      localJointOffset=<10 0 0>
     }]
   }
   n3 = Anim Dynamics - Bone: dyn_fx_wide_a_l(ComponentPose=@n2, Alpha=1, ExternalForce="0,0,0") {
    angularDampingOverride=1
    angularSpringConstant=20000
    bAngularSpring=true
    bGravityOverrideInSimSpace=true
    bOverrideAngularDamping=true
    bUseGravityOverride=true
    gravityScale=0.35
    simulationSpace=RootRelative
    alphaBoolBlend {blendOption=Linear}
    boundBone {boneName=dyn_fx_wide_a_l}
    componentLinearAccScale=<1.3 1.3 1.3>
    componentLinearVelScale=<1.3 1.3 1.3>
    gravityOverride {z=-100}
    physicsBodyDefinitions [{
      sphereCollisionRadius=0
      boundBone {boneName=dyn_fx_wide_a_l}
      boxExtents=<20 30 30>
      constraintSetup {angularLimitsMax=<15 0 15>, angularLimitsMin=<-15 -30 -15>, angularTarget=<1 0 0>}
      localJointOffset=<10 0 0>
     }]
   }
   n4 = Anim Dynamics - Bone: dyn_fx_wide_a(ComponentPose=@n3, Alpha=1, ExternalForce="0,0,0") {
    angularDampingOverride=1
    angularSpringConstant=20000
    bAngularSpring=true
    bGravityOverrideInSimSpace=true
    bOverrideAngularDamping=true
    bUseGravityOverride=true
    gravityScale=0.35
    simulationSpace=RootRelative
    alphaBoolBlend {blendOption=Linear}
    boundBone {boneName=dyn_fx_wide_a}
    componentLinearAccScale=<1.3 1.3 1.3>
    componentLinearVelScale=<1.3 1.3 1.3>
    gravityOverride {z=-100}
    physicsBodyDefinitions [{
      sphereCollisionRadius=0
      boundBone {boneName=dyn_fx_wide_a}
      boxExtents=<20 30 30>
      constraintSetup {angularLimitsMax=<15 0 15>, angularLimitsMin=<-15 -30 -15>, angularTarget=<1 0 0>}
      localJointOffset=<10 0 0>
     }]
   }
   n5 = Anim Dynamics - Bone: dyn_fx_wide_b_r(ComponentPose=@n4, Alpha=1, ExternalForce="0,0,0") {
    angularDampingOverride=1
    angularSpringConstant=20000
    bAngularSpring=true
    bGravityOverrideInSimSpace=true
    bOverrideAngularDamping=true
    bUseGravityOverride=true
    gravityScale=0.35
    simulationSpace=RootRelative
    alphaBoolBlend {blendOption=Linear}
    boundBone {boneName=dyn_fx_wide_b_r}
    componentLinearAccScale=<1.3 1.3 1.3>
    componentLinearVelScale=<1.3 1.3 1.3>
    gravityOverride {z=-100}
    physicsBodyDefinitions [{
      sphereCollisionRadius=0
      boundBone {boneName=dyn_fx_wide_b_r}
      boxExtents=<20 30 30>
      constraintSetup {angularLimitsMax=<15 0 15>, angularLimitsMin=<-15 -30 -15>, angularTarget=<1 0 0>}
      localJointOffset=<10 0 0>
     }]
   }
   n6 = Anim Dynamics - Bone: dyn_fx_wide_b_l(ComponentPose=@n5, Alpha=1, ExternalForce="0,0,0") {
    angularDampingOverride=1
    angularSpringConstant=20000
    bAngularSpring=true
    bGravityOverrideInSimSpace=true
    bOverrideAngularDamping=true
    bUseGravityOverride=true
    gravityScale=0.35
    simulationSpace=RootRelative
    alphaBoolBlend {blendOption=Linear}
    boundBone {boneName=dyn_fx_wide_b_l}
    componentLinearAccScale=<1.3 1.3 1.3>
    componentLinearVelScale=<1.3 1.3 1.3>
    gravityOverride {z=-100}
    physicsBodyDefinitions [{
      sphereCollisionRadius=0
      boundBone {boneName=dyn_fx_wide_b_l}
      boxExtents=<20 30 30>
      constraintSetup {angularLimitsMax=<15 0 15>, angularLimitsMin=<-15 -30 -15>, angularTarget=<1 0 0>}
      localJointOffset=<10 0 0>
     }]
   }
   n7 = Anim Dynamics - Bone: dyn_fx_wide_b(ComponentPose=@n6, Alpha=1, ExternalForce="0,0,0") {
    angularDampingOverride=1
    angularSpringConstant=20000
    bAngularSpring=true
    bGravityOverrideInSimSpace=true
    bOverrideAngularDamping=true
    bUseGravityOverride=true
    gravityScale=0.35
    simulationSpace=RootRelative
    alphaBoolBlend {blendOption=Linear}
    boundBone {boneName=dyn_fx_wide_b}
    componentLinearAccScale=<1.3 1.3 1.3>
    componentLinearVelScale=<1.3 1.3 1.3>
    gravityOverride {z=-100}
    physicsBodyDefinitions [{
      sphereCollisionRadius=0
      boundBone {boneName=dyn_fx_wide_b}
      boxExtents=<20 30 30>
      constraintSetup {angularLimitsMax=<15 0 15>, angularLimitsMin=<-15 -30 -15>, angularTarget=<1 0 0>}
      localJointOffset=<10 0 0>
     }]
   }
   n8 = Anim Dynamics - Bone: dyn_fx_wide_c_r(ComponentPose=@n7, Alpha=1, ExternalForce="0,0,0") {
    angularDampingOverride=1
    angularSpringConstant=20000
    bAngularSpring=true
    bGravityOverrideInSimSpace=true
    bOverrideAngularDamping=true
    bUseGravityOverride=true
    gravityScale=0.35
    simulationSpace=RootRelative
    alphaBoolBlend {blendOption=Linear}
    boundBone {boneName=dyn_fx_wide_c_r}
    componentLinearAccScale=<1.3 1.3 1.3>
    componentLinearVelScale=<1.3 1.3 1.3>
    gravityOverride {z=-100}
    physicsBodyDefinitions [{
      sphereCollisionRadius=0
      boundBone {boneName=dyn_fx_wide_c_r}
      boxExtents=<20 30 30>
      constraintSetup {angularLimitsMax=<15 0 15>, angularLimitsMin=<-15 -30 -15>, angularTarget=<1 0 0>}
      localJointOffset=<10 0 0>
     }]
   }
   n9 = Anim Dynamics - Bone: dyn_fx_wide_c_l(ComponentPose=@n8, Alpha=1, ExternalForce="0,0,0") {
    angularDampingOverride=1
    angularSpringConstant=20000
    bAngularSpring=true
    bGravityOverrideInSimSpace=true
    bOverrideAngularDamping=true
    bUseGravityOverride=true
    gravityScale=0.35
    simulationSpace=RootRelative
    alphaBoolBlend {blendOption=Linear}
    boundBone {boneName=dyn_fx_wide_c_l}
    componentLinearAccScale=<1.3 1.3 1.3>
    componentLinearVelScale=<1.3 1.3 1.3>
    gravityOverride {z=-100}
    physicsBodyDefinitions [{
      sphereCollisionRadius=0
      boundBone {boneName=dyn_fx_wide_c_l}
      boxExtents=<20 30 30>
      constraintSetup {angularLimitsMax=<15 0 15>, angularLimitsMin=<-15 -30 -15>, angularTarget=<1 0 0>}
      localJointOffset=<10 0 0>
     }]
   }
   n10 = Anim Dynamics - Bone: dyn_fx_wide_c(ComponentPose=@n9, Alpha=1, ExternalForce="0,0,0") {
    angularDampingOverride=1
    angularSpringConstant=20000
    bAngularSpring=true
    bGravityOverrideInSimSpace=true
    bOverrideAngularDamping=true
    bUseGravityOverride=true
    gravityScale=0.35
    simulationSpace=RootRelative
    alphaBoolBlend {blendOption=Linear}
    boundBone {boneName=dyn_fx_wide_c}
    componentLinearAccScale=<1.3 1.3 1.3>
    componentLinearVelScale=<1.3 1.3 1.3>
    gravityOverride {z=-100}
    physicsBodyDefinitions [{
      sphereCollisionRadius=0
      boundBone {boneName=dyn_fx_wide_c}
      boxExtents=<20 30 30>
      constraintSetup {angularLimitsMax=<15 0 15>, angularLimitsMin=<-15 -30 -15>, angularTarget=<1 0 0>}
      localJointOffset=<10 0 0>
     }]
   }
   n11 = Anim Dynamics - Bone: dyn_fx_narrow_b_r(ComponentPose=@n10, Alpha=1, ExternalForce="0,0,0") {
    angularDampingOverride=1
    angularSpringConstant=20000
    bAngularSpring=true
    bGravityOverrideInSimSpace=true
    bOverrideAngularDamping=true
    bUseGravityOverride=true
    gravityScale=0.35
    simulationSpace=RootRelative
    alphaBoolBlend {blendOption=Linear}
    boundBone {boneName=dyn_fx_narrow_b_r}
    componentLinearAccScale=<1.3 1.3 1.3>
    componentLinearVelScale=<1.3 1.3 1.3>
    gravityOverride {z=-100}
    physicsBodyDefinitions [{
      sphereCollisionRadius=0
      boundBone {boneName=dyn_fx_narrow_b_r}
      boxExtents=<20 30 30>
      constraintSetup {angularLimitsMax=<15 0 15>, angularLimitsMin=<-15 -45 -15>, angularTarget=<1 0 0>}
      localJointOffset=<5 0 0>
     }]
   }
   n12 = Anim Dynamics - Bone: dyn_fx_narrow_a_l(ComponentPose=@n11, Alpha=1, ExternalForce="0,0,0") {
    angularDampingOverride=1
    angularSpringConstant=20000
    bAngularSpring=true
    bGravityOverrideInSimSpace=true
    bOverrideAngularDamping=true
    bUseGravityOverride=true
    gravityScale=0.35
    simulationSpace=RootRelative
    alphaBoolBlend {blendOption=Linear}
    boundBone {boneName=dyn_fx_narrow_a_l}
    componentLinearAccScale=<1.3 1.3 1.3>
    componentLinearVelScale=<1.3 1.3 1.3>
    gravityOverride {z=-100}
    physicsBodyDefinitions [{
      sphereCollisionRadius=0
      boundBone {boneName=dyn_fx_narrow_a_l}
      boxExtents=<20 30 30>
      constraintSetup {angularLimitsMax=<15 0 15>, angularLimitsMin=<-15 -45 -15>, angularTarget=<1 0 0>}
      localJointOffset=<5 0 0>
     }]
   }
   n13 = Anim Dynamics - Bone: dyn_fx_narrow_a(ComponentPose=@n12, Alpha=1, ExternalForce="0,0,0") {
    angularDampingOverride=1
    angularSpringConstant=20000
    bAngularSpring=true
    bGravityOverrideInSimSpace=true
    bOverrideAngularDamping=true
    bUseGravityOverride=true
    gravityScale=0.35
    simulationSpace=RootRelative
    alphaBoolBlend {blendOption=Linear}
    boundBone {boneName=dyn_fx_narrow_a}
    componentLinearAccScale=<1.3 1.3 1.3>
    componentLinearVelScale=<1.3 1.3 1.3>
    gravityOverride {z=-100}
    physicsBodyDefinitions [{
      sphereCollisionRadius=0
      boundBone {boneName=dyn_fx_narrow_a}
      boxExtents=<20 30 30>
      constraintSetup {angularLimitsMax=<15 0 15>, angularLimitsMin=<-15 -45 -15>, angularTarget=<1 0 0>}
      localJointOffset=<5 0 0>
     }]
   }
   n14 = Anim Dynamics - Bone: dyn_fx_narrow_b_l(ComponentPose=@n13, Alpha=1, ExternalForce="0,0,0") {
    angularDampingOverride=1
    angularSpringConstant=20000
    bAngularSpring=true
    bGravityOverrideInSimSpace=true
    bOverrideAngularDamping=true
    bUseGravityOverride=true
    gravityScale=0.35
    simulationSpace=RootRelative
    alphaBoolBlend {blendOption=Linear}
    boundBone {boneName=dyn_fx_narrow_b_l}
    componentLinearAccScale=<1.3 1.3 1.3>
    componentLinearVelScale=<1.3 1.3 1.3>
    gravityOverride {z=-100}
    physicsBodyDefinitions [{
      sphereCollisionRadius=0
      boundBone {boneName=dyn_fx_narrow_b_l}
      boxExtents=<20 30 30>
      constraintSetup {angularLimitsMax=<15 0 15>, angularLimitsMin=<-15 -45 -15>, angularTarget=<1 0 0>}
      localJointOffset=<5 0 0>
     }]
   }
   n15 = Anim Dynamics - Bone: dyn_fx_narrow_b(ComponentPose=@n14, Alpha=1, ExternalForce="0,0,0") {
    angularDampingOverride=1
    angularSpringConstant=20000
    bAngularSpring=true
    bGravityOverrideInSimSpace=true
    bOverrideAngularDamping=true
    bUseGravityOverride=true
    gravityScale=0.35
    simulationSpace=RootRelative
    alphaBoolBlend {blendOption=Linear}
    boundBone {boneName=dyn_fx_narrow_b}
    componentLinearAccScale=<1.3 1.3 1.3>
    componentLinearVelScale=<1.3 1.3 1.3>
    gravityOverride {z=-100}
    physicsBodyDefinitions [{
      sphereCollisionRadius=0
      boundBone {boneName=dyn_fx_narrow_b}
      boxExtents=<20 30 30>
      constraintSetup {angularLimitsMax=<15 0 15>, angularLimitsMin=<-15 -45 -15>, angularTarget=<1 0 0>}
      localJointOffset=<5 0 0>
     }]
   }
   n16 = Anim Dynamics - Bone: dyn_fx_narrow_a_r(ComponentPose=@n15, Alpha=1, ExternalForce="0,0,0") {
    angularDampingOverride=1
    angularSpringConstant=20000
    bAngularSpring=true
    bGravityOverrideInSimSpace=true
    bOverrideAngularDamping=true
    bUseGravityOverride=true
    gravityScale=0.35
    simulationSpace=RootRelative
    alphaBoolBlend {blendOption=Linear}
    boundBone {boneName=dyn_fx_narrow_a_r}
    componentLinearAccScale=<1.3 1.3 1.3>
    componentLinearVelScale=<1.3 1.3 1.3>
    gravityOverride {z=-100}
    physicsBodyDefinitions [{
      sphereCollisionRadius=0
      boundBone {boneName=dyn_fx_narrow_a_r}
      boxExtents=<20 30 30>
      constraintSetup {angularLimitsMax=<15 0 15>, angularLimitsMin=<-15 -45 -15>, angularTarget=<1 0 0>}
      localJointOffset=<5 0 0>
     }]
   }
   n17 = Anim Dynamics - Bone: dyn_fx_narrow_c(ComponentPose=@n16, Alpha=1, ExternalForce="0,0,0") {
    angularDampingOverride=1
    angularSpringConstant=20000
    bAngularSpring=true
    bGravityOverrideInSimSpace=true
    bOverrideAngularDamping=true
    bUseGravityOverride=true
    gravityScale=0.35
    simulationSpace=RootRelative
    alphaBoolBlend {blendOption=Linear}
    boundBone {boneName=dyn_fx_narrow_c}
    componentLinearAccScale=<1.3 1.3 1.3>
    componentLinearVelScale=<1.3 1.3 1.3>
    gravityOverride {z=-100}
    physicsBodyDefinitions [{
      sphereCollisionRadius=0
      boundBone {boneName=dyn_fx_narrow_c}
      boxExtents=<20 30 30>
      constraintSetup {angularLimitsMax=<15 0 15>, angularLimitsMin=<-15 -45 -15>, angularTarget=<1 0 0>}
      localJointOffset=<5 0 0>
     }]
   }
   n18 = Get bIsInVehicle() -> (bIsInVehicle:bool)
   n19 = "Transform (Modify) Bone - Bone: dyn_fx_wide_a_l"!(ComponentPose=@n17, bAlphaBoolEnabled:bool=@n18, Translation="0,0,0", Rotation="0,0,0", Scale="0.01,0.01,0.01") -> (Pose) {
    alphaInputType=Bool
    rotationMode=BMM_Additive
    rotationSpace=BCS_BoneSpace
    scaleMode=BMM_Additive
    scaleSpace=BCS_BoneSpace
    translationMode=BMM_Additive
    translationSpace=BCS_BoneSpace
    alphaBoolBlend {blendOption=Linear}
    boneToModify {boneName=dyn_fx_wide_a_l}
    scale=<0.01 0.01 0.01>
   }
   n20 = "Transform (Modify) Bone - Bone: dyn_fx_wide_a_r"!(ComponentPose=@n19, bAlphaBoolEnabled:bool=@n18, Translation="0,0,0", Rotation="0,0,0", Scale="0.01,0.01,0.01") -> (Pose) {
    alphaInputType=Bool
    rotationMode=BMM_Additive
    rotationSpace=BCS_BoneSpace
    scaleMode=BMM_Additive
    scaleSpace=BCS_BoneSpace
    translationMode=BMM_Additive
    translationSpace=BCS_BoneSpace
    alphaBoolBlend {blendOption=Linear}
    boneToModify {boneName=dyn_fx_wide_a_r}
    scale=<0.01 0.01 0.01>
   }
   n21 = "Transform (Modify) Bone - Bone: dyn_fx_wide_a"!(ComponentPose=@n20, bAlphaBoolEnabled:bool=@n18, Translation="0,0,0", Rotation="0,0,0", Scale="0.01,0.01,0.01") -> (Pose) {
    alphaInputType=Bool
    rotationMode=BMM_Additive
    rotationSpace=BCS_BoneSpace
    scaleMode=BMM_Additive
    scaleSpace=BCS_BoneSpace
    translationMode=BMM_Additive
    translationSpace=BCS_BoneSpace
    alphaBoolBlend {blendOption=Linear}
    boneToModify {boneName=dyn_fx_wide_a}
    scale=<0.01 0.01 0.01>
   }
   n22 = Get bIsInVehicle()
   n23 = "Transform (Modify) Bone - Bone: dyn_fx_narrow_a_l"!(ComponentPose=@n21, bAlphaBoolEnabled:bool=@n22, Translation="0,0,0", Rotation="0,0,0", Scale="0.01,0.01,0.01") -> (Pose) {
    alphaInputType=Bool
    rotationMode=BMM_Additive
    rotationSpace=BCS_BoneSpace
    scaleMode=BMM_Additive
    scaleSpace=BCS_BoneSpace
    translationMode=BMM_Additive
    translationSpace=BCS_BoneSpace
    alphaBoolBlend {blendOption=Linear}
    boneToModify {boneName=dyn_fx_narrow_a_l}
    scale=<0.01 0.01 0.01>
   }
   n24 = "Transform (Modify) Bone - Bone: dyn_fx_narrow_a_r"!(ComponentPose=@n23, bAlphaBoolEnabled:bool=@n22, Translation="0,0,0", Rotation="0,0,0", Scale="0.01,0.01,0.01") -> (Pose) {
    alphaInputType=Bool
    rotationMode=BMM_Additive
    rotationSpace=BCS_BoneSpace
    scaleMode=BMM_Additive
    scaleSpace=BCS_BoneSpace
    translationMode=BMM_Additive
    translationSpace=BCS_BoneSpace
    alphaBoolBlend {blendOption=Linear}
    boneToModify {boneName=dyn_fx_narrow_a_r}
    scale=<0.01 0.01 0.01>
   }
   n25 = "Transform (Modify) Bone - Bone: dyn_fx_narrow_a"!(ComponentPose=@n24, bAlphaBoolEnabled:bool=@n22, Translation="0,0,0", Rotation="0,0,0", Scale="0.01,0.01,0.01") -> (Pose) {
    alphaInputType=Bool
    rotationMode=BMM_Additive
    rotationSpace=BCS_BoneSpace
    scaleMode=BMM_Additive
    scaleSpace=BCS_BoneSpace
    translationMode=BMM_Additive
    translationSpace=BCS_BoneSpace
    alphaBoolBlend {blendOption=Linear}
    boneToModify {boneName=dyn_fx_narrow_a}
    scale=<0.01 0.01 0.01>
   }
   Output Pose(Result=Component To Local(ComponentPose=@n25)) {layerGroup=None}
   n26 = Control Rig(Source=Copy Pose From Mesh() {bUseAttachedParent=true}) {
    alphaBoolBlend {blendOption=Linear}
    controlRigAssetReference {blueprintRigClass=&/F_MED_RareDelightSail_CtrlRig.F_MED_RareDelightSail_CtrlRig_C}
   }
   n27 = "Constraint - simspace (P)"!(ComponentPose=Local To Component(LocalPose=@n26), bAlphaBoolEnabled:bool=Get bIsInFrontEnd(), ConstraintWeights_0=1) -> (Pose) {
    alphaInputType=Bool
    alphaBoolBlend {blendOption=Linear}
    boneToModify {boneName=simspace}
    constraintSetup [{targetBone {boneName=pelvis}}]
    constraintWeights [1]
   }
   n28 = RigidBody!(ComponentPose=@n27, Alpha, OverrideWorldGravity:Vector=Get GravityOverride(), ExternalForce) -> (Pose) {
    bClampLinearTranslationLimitToRefPose=true
    bOverrideWorldGravity=true
    overridePhysicsAsset=/BRCosmetics/Characters/Player/Female/Medium/Bodies/F_MED_Buffet_Shard/Meshes/F_MED_Buffet_Shard_Physics2
    simulationSpace=BaseBoneSpace
    alphaBoolBlend {blendOption=Linear}
    alphaScaleBiasClamp {bMapRange=true, outRange {max=0.5, min=1}}
    baseBoneRef {boneName=simspace}
    componentAppliedLinearAccClamp=<3000 3000 3000>
    componentLinearAccScale=<0.25 0.25 0.25>
    componentLinearVelScale=<0.25 0.25 0.25>
    overrideWorldGravity {z=-980}
   }
   n29 = "Constraint - dyn_skirt_fr_a_base_l (RR)"!(ComponentPose=@n28, Alpha=1, ConstraintWeights_0=0.1, ConstraintWeights_1:float=0.9) -> (Pose) {
    alphaBoolBlend {blendOption=Linear}
    boneToModify {boneName=dyn_skirt_fr_a_base_l}
    constraintSetup [{transformType=Rotation, targetBone {boneName=pelvis}}, {transformType=Rotation, targetBone {boneName=thigh_l}}]
    constraintWeights [1,1]
   }
   n30 = "Constraint - dyn_skirt_side_a_base_l (RR)"!(ComponentPose=@n29, Alpha=1, ConstraintWeights_0=0.4, ConstraintWeights_1:float=0.6) -> (Pose) {
    alphaBoolBlend {blendOption=Linear}
    boneToModify {boneName=dyn_skirt_side_a_base_l}
    constraintSetup [{transformType=Rotation, targetBone {boneName=pelvis}}, {transformType=Rotation, targetBone {boneName=thigh_l}}]
    constraintWeights [1,1]
   }
   n31 = "Constraint - dyn_skirt_bk_a_base_l (RR)"!(ComponentPose=@n30, Alpha=1, ConstraintWeights_0=0.8, ConstraintWeights_1:float=0.2) -> (Pose) {
    alphaBoolBlend {blendOption=Linear}
    boneToModify {boneName=dyn_skirt_bk_a_base_l}
    constraintSetup [{transformType=Rotation, targetBone {boneName=pelvis}}, {transformType=Rotation, targetBone {boneName=thigh_l}}]
    constraintWeights [1,1]
   }
   n32 = "Constraint - dyn_skirt_fr_a_base_r (RR)"!(ComponentPose=@n31, Alpha=1, ConstraintWeights_0=0.1, ConstraintWeights_1:float=0.9) -> (Pose) {
    alphaBoolBlend {blendOption=Linear}
    boneToModify {boneName=dyn_skirt_fr_a_base_r}
    constraintSetup [{transformType=Rotation, targetBone {boneName=pelvis}}, {transformType=Rotation, targetBone {boneName=thigh_r}}]
    constraintWeights [1,1]
   }
   n33 = "Constraint - dyn_skirt_side_a_base_r (RR)"!(ComponentPose=@n32, Alpha=1, ConstraintWeights_0=0.4, ConstraintWeights_1:float=0.6) -> (Pose) {
    alphaBoolBlend {blendOption=Linear}
    boneToModify {boneName=dyn_skirt_side_a_base_r}
    constraintSetup [{transformType=Rotation, targetBone {boneName=pelvis}}, {transformType=Rotation, targetBone {boneName=thigh_r}}]
    constraintWeights [1,1]
   }
   n34 = "Constraint - dyn_skirt_bk_a_base_r (RR)"!(ComponentPose=@n33, Alpha=1, ConstraintWeights_0=0.8, ConstraintWeights_1:float=0.2) -> (Pose) {
    alphaBoolBlend {blendOption=Linear}
    boneToModify {boneName=dyn_skirt_bk_a_base_r}
    constraintSetup [{transformType=Rotation, targetBone {boneName=pelvis}}, {transformType=Rotation, targetBone {boneName=thigh_r}}]
    constraintWeights [1,1]
   }
   n35 = RigidBody(ComponentPose=@n34, OverrideWorldGravity=Get gravityOverride() -> (gravityOverride:Vector)) {
    bClampLinearTranslationLimitToRefPose=true
    bOverrideWorldGravity=true
    overridePhysicsAsset=&/F_MED_RareDelightSail_Physics
    simulationSpace=BaseBoneSpace
    alphaBoolBlend {blendOption=Linear}
    alphaScaleBiasClamp {bMapRange=true, outRange {max=0.5, min=1}}
    baseBoneRef {boneName=root}
    componentAppliedLinearAccClamp=<3000 3000 3000>
    componentLinearAccScale=<0.25 0.25 0.25>
    componentLinearVelScale=<0.25 0.25 0.25>
    overrideWorldGravity {z=-980}
   }
   n36 = "Constraint - dyn_skirt_fr_a_base (PPP)"!(
    ComponentPose=@n35
    Alpha=Get skirtMidJntConstraintAlpha() -> (skirtMidJntConstraintAlpha:double)
    ConstraintWeights_0=0.5
    ConstraintWeights_1:float=0.5
    ConstraintWeights_2:float=0
   ) -> (Pose) {
    alphaBoolBlend {blendOption=Linear}
    alphaScaleBiasClamp {bClampResult=true, bInterpResult=true, clampMax=0.3, interpSpeedDecreasing=5, interpSpeedIncreasing=12}
    boneToModify {boneName=dyn_skirt_fr_a_base}
    constraintSetup [{targetBone {boneName=dyn_skirt_fr_a_l}}, {targetBone {boneName=dyn_skirt_fr_a_r}}, {targetBone {boneName=pelvis}}]
    constraintWeights [1,1,1]
   }
   n37 = "Constraint - dyn_skirt_bk_a_base (PPP)"!(ComponentPose=@n36, Alpha=1, ConstraintWeights_0=0.5, ConstraintWeights_1:float=0.5, ConstraintWeights_2:float=0) -> (Pose) {
    alphaBoolBlend {blendOption=Linear}
    boneToModify {boneName=dyn_skirt_bk_a_base}
    constraintSetup [{targetBone {boneName=dyn_skirt_bk_a_l}}, {targetBone {boneName=dyn_skirt_bk_a_r}}, {targetBone {boneName=pelvis}}]
    constraintWeights [1,1,1]
   }
   n38 = "Transform (Modify) Bone - Bone: dyn_skirt_bk_a_base_r"!(ComponentPose=@n37, Alpha=1, Rotation="5.000,0.000,-0.000") -> (Pose) {
    rotationMode=BMM_Replace
    rotationSpace=BCS_BoneSpace
    alphaBoolBlend {blendOption=Linear}
    boneToModify {boneName=dyn_skirt_bk_a_base_r}
    rotation=<rot 4.9999 0 0>
   }
   n39 = "Transform (Modify) Bone - Bone: dyn_skirt_side_a_base_r"!(ComponentPose=@n38, Alpha=1, Rotation="10.000,0.000,-0.000") -> (Pose) {
    rotationMode=BMM_Replace
    rotationSpace=BCS_BoneSpace
    alphaBoolBlend {blendOption=Linear}
    boneToModify {boneName=dyn_skirt_side_a_base_r}
    rotation=<rot 10 0 0>
   }
   n40 = "Transform (Modify) Bone - Bone: dyn_skirt_side_a_base_l"!(ComponentPose=@n39, Alpha=1, Rotation="5.000,0.000,-0.000") -> (Pose) {
    rotationMode=BMM_Replace
    rotationSpace=BCS_BoneSpace
    alphaBoolBlend {blendOption=Linear}
    boneToModify {boneName=dyn_skirt_side_a_base_l}
    rotation=<rot 4.9999 0 0>
   }
   Save cached pose 'SkirtCachePose'(Pose=Component To Local(ComponentPose=@n40)) {cacheName=SkirtCachePose}
   n41 = "Constraint - dyn_shard_shoulder_c_base_l (PPP)"!(
    ComponentPose=Local To Component(LocalPose=Use cached pose 'SkirtCachePose'())
    Alpha=1
    ConstraintWeights_0=0.95
    ConstraintWeights_1:float=0.04
    ConstraintWeights_2:float=0.01
   ) -> (Pose) {
    alphaBoolBlend {blendOption=Linear}
    boneToModify {boneName=dyn_shard_shoulder_c_base_l}
    constraintSetup [{targetBone {boneName=upperarm_twist_01_l}}, {targetBone {boneName=upperarm_up_volume_l}}, {targetBone {boneName=deform_upperarm_fr_l}}]
    constraintWeights [1,1,1]
   }
   n42 = "Constraint - dyn_shard_shoulder_d_base_l (PPP)"!(ComponentPose=@n41, Alpha=1, ConstraintWeights_0=0.1, ConstraintWeights_1:float=0.09, ConstraintWeights_2:float=0.81) -> (Pose) {
    alphaBoolBlend {blendOption=Linear}
    boneToModify {boneName=dyn_shard_shoulder_d_base_l}
    constraintSetup [{targetBone {boneName=upperarm_twist_02_l}}, {targetBone {boneName=upperarm_l}}, {targetBone {boneName=upperarm_twist_01_l}}]
    constraintWeights [1,1,1]
   }
   n43 = "Constraint - dyn_shard_shoulder_e_base_l (PP)"!(ComponentPose=@n42, Alpha=1, ConstraintWeights_0=0.2, ConstraintWeights_1:float=0.8) -> (Pose) {
    alphaBoolBlend {blendOption=Linear}
    boneToModify {boneName=dyn_shard_shoulder_e_base_l}
    constraintSetup [{targetBone {boneName=clavicle_l}}, {targetBone {boneName=upperarm_twist_01_l}}]
    constraintWeights [1,1]
   }
   n44 = "Constraint - dyn_shard_shoulder_f_base_l (PPPP)"!(
    ComponentPose=@n43
    Alpha=1
    ConstraintWeights_0=0.16
    ConstraintWeights_1:float=0.21
    ConstraintWeights_2:float=0.43
    ConstraintWeights_3:float=0.2
   ) -> (Pose) {
    alphaBoolBlend {blendOption=Linear}
    boneToModify {boneName=dyn_shard_shoulder_f_base_l}
    constraintSetup [
     {targetBone {boneName=upperarm_bk_volume_l}}
     {targetBone {boneName=clavicle_l}}
     {targetBone {boneName=upperarm_twist_01_l}}
     {targetBone {boneName=upperarm_up_volume_l}}
    ]
    constraintWeights [1,1,1,1]
   }
   n45 = "Constraint - dyn_shard_shoulder_g_base_l (PPPP)"!(
    ComponentPose=@n44
    Alpha=1
    ConstraintWeights_0=0.15
    ConstraintWeights_1:float=0.45
    ConstraintWeights_2:float=0.1
    ConstraintWeights_3:float=0.3
   ) -> (Pose) {
    alphaBoolBlend {blendOption=Linear}
    boneToModify {boneName=dyn_shard_shoulder_g_base_l}
    constraintSetup [
     {targetBone {boneName=upperarm_bk_volume_l}}
     {targetBone {boneName=clavicle_l}}
     {targetBone {boneName=spine_05}}
     {targetBone {boneName=upperarm_twist_01_l}}
    ]
    constraintWeights [1,1,1,1]
   }
   n46 = "Constraint - dyn_shard_shoulder_h_base_l (PPPP)"!(
    ComponentPose=@n45
    Alpha=1
    ConstraintWeights_0=0.26
    ConstraintWeights_1:float=0.22
    ConstraintWeights_2:float=0.38
    ConstraintWeights_3:float=0.14
   ) -> (Pose) {
    alphaBoolBlend {blendOption=Linear}
    boneToModify {boneName=dyn_shard_shoulder_h_base_l}
    constraintSetup [
     {targetBone {boneName=upperarm_bk_volume_l}}
     {targetBone {boneName=clavicle_l}}
     {targetBone {boneName=upperarm_twist_01_l}}
     {targetBone {boneName=upperarm_up_volume_l}}
    ]
    constraintWeights [1,1,1,1]
   }
   n47 = "Constraint - dyn_shard_skirt_a_base_l (PPP)"!(ComponentPose=@n46, Alpha=1, ConstraintWeights_0=0.02, ConstraintWeights_1:float=0.81, ConstraintWeights_2:float=0.17) -> (Pose) {
    alphaBoolBlend {blendOption=Linear}
    boneToModify {boneName=dyn_shard_skirt_a_base_l}
    constraintSetup [{targetBone {boneName=dyn_skirt_side_b_l}}, {targetBone {boneName=dyn_skirt_fr_b_l}}, {targetBone {boneName=dyn_skirt_fr_b}}]
    constraintWeights [1,1,1]
   }
   n48 = "Constraint - dyn_shard_skirt_b_base_l (PPP)"!(ComponentPose=@n47, Alpha=1, ConstraintWeights_0=0.12, ConstraintWeights_1:float=0.84, ConstraintWeights_2:float=0.04) -> (Pose) {
    alphaBoolBlend {blendOption=Linear}
    boneToModify {boneName=dyn_shard_skirt_b_base_l}
    constraintSetup [{targetBone {boneName=dyn_skirt_side_b_l}}, {targetBone {boneName=dyn_skirt_fr_b_l}}, {targetBone {boneName=dyn_skirt_fr_b}}]
    constraintWeights [1,1,1]
   }
   n49 = "Constraint - dyn_shard_skirt_c_base_l (PP)"!(ComponentPose=@n48, Alpha=1, ConstraintWeights_0=0.38, ConstraintWeights_1:float=0.62) -> (Pose) {
    alphaBoolBlend {blendOption=Linear}
    boneToModify {boneName=dyn_shard_skirt_c_base_l}
    constraintSetup [{targetBone {boneName=dyn_skirt_side_b_l}}, {targetBone {boneName=dyn_skirt_fr_b_l}}]
    constraintWeights [1,1]
   }
   n50 = "Constraint - dyn_shard_skirt_d_base_l (PPP)"!(ComponentPose=@n49, Alpha=1, ConstraintWeights_0=0.77, ConstraintWeights_1:float=0.14, ConstraintWeights_2:float=0.09) -> (Pose) {
    alphaBoolBlend {blendOption=Linear}
    boneToModify {boneName=dyn_shard_skirt_d_base_l}
    constraintSetup [{targetBone {boneName=dyn_skirt_side_b_l}}, {targetBone {boneName=dyn_skirt_fr_b_l}}, {targetBone {boneName=dyn_skirt_bk_b_l}}]
    constraintWeights [1,1,1]
   }
   n51 = "Constraint - dyn_shard_skirt_e_base_l (PPP)"!(ComponentPose=@n50, Alpha=1, ConstraintWeights_0=0.75, ConstraintWeights_1:float=0.06, ConstraintWeights_2:float=0.19) -> (Pose) {
    alphaBoolBlend {blendOption=Linear}
    boneToModify {boneName=dyn_shard_skirt_e_base_l}
    constraintSetup [{targetBone {boneName=dyn_skirt_side_b_l}}, {targetBone {boneName=dyn_skirt_fr_b_l}}, {targetBone {boneName=dyn_skirt_bk_b_l}}]
    constraintWeights [1,1,1]
   }
   n52 = "Constraint - dyn_shard_skirt_a_base_r (PPP)"!(ComponentPose=@n51, Alpha=1, ConstraintWeights_0=0.34, ConstraintWeights_1:float=0.01, ConstraintWeights_2:float=0.65) -> (Pose) {
    alphaBoolBlend {blendOption=Linear}
    boneToModify {boneName=dyn_shard_skirt_a_base_r}
    constraintSetup [{targetBone {boneName=dyn_skirt_fr_b}}, {targetBone {boneName=dyn_skirt_fr_b_l}}, {targetBone {boneName=dyn_skirt_fr_b_r}}]
    constraintWeights [1,1,1]
   }
   n53 = "Constraint - dyn_shard_skirt_b_base_r (PP)"!(ComponentPose=@n52, Alpha=1, ConstraintWeights_0=0.9, ConstraintWeights_1:float=0.1) -> (Pose) {
    alphaBoolBlend {blendOption=Linear}
    boneToModify {boneName=dyn_shard_skirt_b_base_r}
    constraintSetup [{targetBone {boneName=dyn_skirt_fr_c_r}}, {targetBone {boneName=dyn_skirt_side_c_r}}]
    constraintWeights [1,1]
   }
   n54 = "Constraint - dyn_shard_skirt_c_base_r (PP)"!(ComponentPose=@n53, Alpha=1, ConstraintWeights_0=0.9, ConstraintWeights_1:float=0.1) -> (Pose) {
    alphaBoolBlend {blendOption=Linear}
    boneToModify {boneName=dyn_shard_skirt_c_base_r}
    constraintSetup [{targetBone {boneName=dyn_skirt_fr_c_r}}, {targetBone {boneName=dyn_skirt_side_c_r}}]
    constraintWeights [1,1]
   }
   n55 = "Constraint - dyn_shard_skirt_d_base_r (PP)"!(ComponentPose=@n54, Alpha=1, ConstraintWeights_0=0.76, ConstraintWeights_1:float=0.24) -> (Pose) {
    alphaBoolBlend {blendOption=Linear}
    boneToModify {boneName=dyn_shard_skirt_d_base_r}
    constraintSetup [{targetBone {boneName=dyn_skirt_fr_b_r}}, {targetBone {boneName=dyn_skirt_side_b_r}}]
    constraintWeights [1,1]
   }
   n56 = "Constraint - dyn_shard_skirt_e_base_r (PP)"!(ComponentPose=@n55, Alpha=1, ConstraintWeights_0=0.5, ConstraintWeights_1:float=0.5) -> (Pose) {
    alphaBoolBlend {blendOption=Linear}
    boneToModify {boneName=dyn_shard_skirt_e_base_r}
    constraintSetup [{targetBone {boneName=dyn_skirt_fr_c_r}}, {targetBone {boneName=dyn_skirt_side_c_r}}]
    constraintWeights [1,1]
   }
   n57 = "Constraint - dyn_shard_skirt_f_base_r (PP)"!(ComponentPose=@n56, Alpha=1, ConstraintWeights_0=0.3, ConstraintWeights_1:float=0.7) -> (Pose) {
    alphaBoolBlend {blendOption=Linear}
    boneToModify {boneName=dyn_shard_skirt_f_base_r}
    constraintSetup [{targetBone {boneName=dyn_skirt_fr_c_r}}, {targetBone {boneName=dyn_skirt_side_c_r}}]
    constraintWeights [1,1]
   }
   n58 = "Constraint - dyn_shard_skirt_g_base_r (PPPP)"!(
    ComponentPose=@n57
    Alpha=1
    ConstraintWeights_0=0.03
    ConstraintWeights_1:float=0.03
    ConstraintWeights_2:float=0.41
    ConstraintWeights_3:float=0.53
   ) -> (Pose) {
    alphaBoolBlend {blendOption=Linear}
    boneToModify {boneName=dyn_shard_skirt_g_base_r}
    constraintSetup [
     {targetBone {boneName=dyn_skirt_side_a_r}}
     {targetBone {boneName=dyn_skirt_fr_a_r}}
     {targetBone {boneName=dyn_skirt_fr_b_r}}
     {targetBone {boneName=dyn_skirt_side_b_r}}
    ]
    constraintWeights [1,1,1,1]
   }
   n59 = "Constraint - dyn_shard_skirt_h_base_r (PP)"!(ComponentPose=@n58, Alpha=1, ConstraintWeights_0=0.7, ConstraintWeights_1:float=0.3) -> (Pose) {
    alphaBoolBlend {blendOption=Linear}
    boneToModify {boneName=dyn_shard_skirt_h_base_r}
    constraintSetup [{targetBone {boneName=dyn_skirt_bk_c_r}}, {targetBone {boneName=dyn_skirt_side_c_r}}]
    constraintWeights [1,1]
   }
   n60 = "Constraint - dyn_shard_skirt_i_base_r (PPPPPP)"!(
    ComponentPose=@n59
    Alpha=1
    ConstraintWeights_0=0.06
    ConstraintWeights_1:float=0.38
    ConstraintWeights_2:float=0.06
    ConstraintWeights_3:float=0.06
    ConstraintWeights_4:float=0.38
    ConstraintWeights_5:float=0.06
   ) -> (Pose) {
    alphaBoolBlend {blendOption=Linear}
    boneToModify {boneName=dyn_shard_skirt_i_base_r}
    constraintSetup [
     {targetBone {boneName=dyn_skirt_side_a_r}}
     {targetBone {boneName=dyn_skirt_bk_b_r}}
     {targetBone {boneName=dyn_skirt_bk_b}}
     {targetBone {boneName=dyn_skirt_bk_a}}
     {targetBone {boneName=dyn_skirt_bk_a_r}}
     {targetBone {boneName=dyn_skirt_side_b_r}}
    ]
    constraintWeights [1,1,1,1,1,1]
   }
   n61 = "Constraint - dyn_shard_skirt_j_base_r (PPPP)"!(
    ComponentPose=@n60
    Alpha=1
    ConstraintWeights_0=0.7
    ConstraintWeights_1:float=0.15
    ConstraintWeights_2:float=0.1
    ConstraintWeights_3:float=0.05
   ) -> (Pose) {
    alphaBoolBlend {blendOption=Linear}
    boneToModify {boneName=dyn_shard_skirt_j_base_r}
    constraintSetup [
     {targetBone {boneName=dyn_skirt_side_a_r}}
     {targetBone {boneName=pelvis}}
     {targetBone {boneName=dyn_skirt_fr_a_r}}
     {targetBone {boneName=dyn_skirt_bk_a_r}}
    ]
    constraintWeights [1,1,1,1]
   }
   n62 = Get bIsSkyDiving()
   n63 = "Transform (Modify) Bone - Bone: dyn_shard_shoulder_a_base_l"!(ComponentPose=@n61, bAlphaBoolEnabled:bool=@n62, Translation="0,0,2", Rotation="-10,0,0", Scale="1,1,1") -> (Pose) {
    alphaInputType=Bool
    rotationMode=BMM_Additive
    rotationSpace=BCS_BoneSpace
    translationMode=BMM_Additive
    translationSpace=BCS_BoneSpace
    alphaBoolBlend {blendOption=Linear}
    boneToModify {boneName=dyn_shard_shoulder_a_base_l}
    rotation {pitch=-10}
    translation {z=2}
   }
   n64 = "Transform (Modify) Bone - Bone: dyn_shard_shoulder_b_base_l"!(ComponentPose=@n63, bAlphaBoolEnabled:bool=@n62, Translation="0,1,6", Rotation="-15,0,0", Scale="1,1,1") -> (Pose) {
    alphaInputType=Bool
    rotationMode=BMM_Additive
    rotationSpace=BCS_BoneSpace
    translationMode=BMM_Additive
    translationSpace=BCS_BoneSpace
    alphaBoolBlend {blendOption=Linear}
    boneToModify {boneName=dyn_shard_shoulder_b_base_l}
    rotation {pitch=-15}
    translation {y=1, z=6}
   }
   n65 = Anim Dynamics - Bone: dyn_shard_skirt_a_base_l(ComponentPose=@n64, Alpha=0.3, ExternalForce="0,0,0") {
    angularDampingOverride=0.999
    angularSpringConstant=150
    bAngularSpring=true
    bGravityOverrideInSimSpace=true
    bLinearSpring=true
    bOverrideAngularDamping=true
    bOverrideLinearDamping=true
    bUseGravityOverride=true
    gravityScale=0
    linearDampingOverride=0.995
    linearSpringConstant=100
    simulationSpace=RootRelative
    alphaBoolBlend {blendOption=Linear}
    alphaScaleBiasClamp {interpSpeedDecreasing=3, interpSpeedIncreasing=5}
    boundBone {boneName=dyn_shard_skirt_a_base_l}
    componentAppliedLinearAccClamp=<6000 6000 6000>
    componentLinearAccScale=<0.05 0.05 0.05>
    componentLinearVelScale=<0.8 0.8 0.6>
    gravityOverride=<0.25 0.25 0.25>
    physicsBodyDefinitions [{
      sphereCollisionRadius=0
      boundBone {boneName=dyn_shard_skirt_a_base_l}
      boxExtents=<50 50 50>
      constraintSetup {angularLimitsMax=<0 0 0>, angularLimitsMin=<0 0 0>, angularTarget=<0 0 0>, linearAxesMax=<1 0.5 0>, linearAxesMin=<-0.5 -0.5 -0.5>}
      localJointOffset=<-3 0 0>
     }]
    relativeSpaceBone {boneName=root}
   }
   n66 = Anim Dynamics - Bone: dyn_shard_skirt_b_base_l(ComponentPose=@n65, Alpha=0.75, ExternalForce="0,0,0") {
    angularDampingOverride=0.999
    angularSpringConstant=150
    bAngularSpring=true
    bGravityOverrideInSimSpace=true
    bLinearSpring=true
    bOverrideAngularDamping=true
    bOverrideLinearDamping=true
    bUseGravityOverride=true
    gravityScale=0
    linearDampingOverride=0.995
    linearSpringConstant=100
    simulationSpace=RootRelative
    alphaBoolBlend {blendOption=Linear}
    alphaScaleBiasClamp {interpSpeedDecreasing=3, interpSpeedIncreasing=5}
    boundBone {boneName=dyn_shard_skirt_b_base_l}
    componentAppliedLinearAccClamp=<6000 6000 6000>
    componentLinearAccScale=<0.05 0.05 0.05>
    componentLinearVelScale=<0.8 0.8 0.6>
    gravityOverride=<0.25 0.25 0.25>
    physicsBodyDefinitions [{
      sphereCollisionRadius=0
      boundBone {boneName=dyn_shard_skirt_b_base_l}
      boxExtents=<50 50 50>
      constraintSetup {angularLimitsMax=<0 0 0>, angularLimitsMin=<0 0 0>, angularTarget=<0 0 0>, linearAxesMax=<1 0.5 0>, linearAxesMin=<-0.5 -0.5 -0.5>}
      localJointOffset=<-3 0 0>
     }]
    relativeSpaceBone {boneName=root}
   }
   n67 = Anim Dynamics - Bone: dyn_shard_skirt_c_base_l(ComponentPose=@n66, Alpha=0.7, ExternalForce="0,0,0") {
    angularDampingOverride=0.999
    angularSpringConstant=150
    bAngularSpring=true
    bGravityOverrideInSimSpace=true
    bLinearSpring=true
    bOverrideAngularDamping=true
    bOverrideLinearDamping=true
    bUseGravityOverride=true
    gravityScale=0
    linearDampingOverride=0.995
    linearSpringConstant=100
    simulationSpace=RootRelative
    alphaBoolBlend {blendOption=Linear}
    alphaScaleBiasClamp {interpSpeedDecreasing=3, interpSpeedIncreasing=5}
    boundBone {boneName=dyn_shard_skirt_c_base_l}
    componentAppliedLinearAccClamp=<6000 6000 6000>
    componentLinearAccScale=<0.05 0.05 0.05>
    componentLinearVelScale=<0.8 0.8 0.6>
    gravityOverride=<0.25 0.25 0.25>
    physicsBodyDefinitions [{
      sphereCollisionRadius=0
      boundBone {boneName=dyn_shard_skirt_c_base_l}
      boxExtents=<50 50 50>
      constraintSetup {angularLimitsMax=<0 0 0>, angularLimitsMin=<0 0 0>, angularTarget=<0 0 0>, linearAxesMax=<1 0.5 0>, linearAxesMin=<0 -0.5 -1>}
      localJointOffset=<-3 0 0>
     }]
    relativeSpaceBone {boneName=root}
   }
   n68 = Anim Dynamics - Bone: dyn_shard_skirt_d_base_l(ComponentPose=@n67, Alpha=0.7, ExternalForce="0,0,0") {
    angularDampingOverride=0.999
    angularSpringConstant=150
    bAngularSpring=true
    bGravityOverrideInSimSpace=true
    bLinearSpring=true
    bOverrideAngularDamping=true
    bOverrideLinearDamping=true
    bUseGravityOverride=true
    gravityScale=0
    linearDampingOverride=0.995
    linearSpringConstant=100
    simulationSpace=RootRelative
    alphaBoolBlend {blendOption=Linear}
    alphaScaleBiasClamp {interpSpeedDecreasing=3, interpSpeedIncreasing=5}
    boundBone {boneName=dyn_shard_skirt_d_base_l}
    componentAppliedLinearAccClamp=<6000 6000 6000>
    componentLinearAccScale=<0.05 0.05 0.05>
    componentLinearVelScale=<0.8 0.8 0.6>
    gravityOverride=<0.25 0.25 0.25>
    physicsBodyDefinitions [{
      sphereCollisionRadius=0
      boundBone {boneName=dyn_shard_skirt_d_base_l}
      boxExtents=<50 50 50>
      constraintSetup {angularLimitsMax=<0 0 0>, angularLimitsMin=<0 0 0>, angularTarget=<0 0 0>, linearAxesMax=<1 0.5 0>, linearAxesMin=<-1 -0.5 -1>}
      localJointOffset=<-3 0 0>
     }]
    relativeSpaceBone {boneName=root}
   }
   n69 = Anim Dynamics - Bone: dyn_shard_skirt_e_base_l(ComponentPose=@n68, Alpha=0.5, ExternalForce="0,0,0") {
    angularDampingOverride=0.999
    angularSpringConstant=150
    bAngularSpring=true
    bGravityOverrideInSimSpace=true
    bLinearSpring=true
    bOverrideAngularDamping=true
    bOverrideLinearDamping=true
    bUseGravityOverride=true
    gravityScale=0
    linearDampingOverride=0.995
    linearSpringConstant=100
    simulationSpace=RootRelative
    alphaBoolBlend {blendOption=Linear}
    alphaScaleBiasClamp {interpSpeedDecreasing=3, interpSpeedIncreasing=5}
    boundBone {boneName=dyn_shard_skirt_e_base_l}
    componentAppliedLinearAccClamp=<6000 6000 6000>
    componentLinearAccScale=<0.05 0.05 0.05>
    componentLinearVelScale=<0.8 0.8 0.6>
    gravityOverride=<0.25 0.25 0.25>
    physicsBodyDefinitions [{
      sphereCollisionRadius=0
      boundBone {boneName=dyn_shard_skirt_e_base_l}
      boxExtents=<50 50 50>
      constraintSetup {angularLimitsMax=<0 0 0>, angularLimitsMin=<0 0 0>, angularTarget=<0 0 0>, linearAxesMax=<0.5 0.5 0>, linearAxesMin=<-0.5 -0.5 -0.5>}
      localJointOffset=<-3 0 0>
     }]
    relativeSpaceBone {boneName=root}
   }
   n70 = Anim Dynamics - Bone: dyn_shard_skirt_a_base_r(ComponentPose=@n69, Alpha=0.75, ExternalForce="0,0,0") {
    angularDampingOverride=0.999
    angularSpringConstant=150
    bAngularSpring=true
    bGravityOverrideInSimSpace=true
    bLinearSpring=true
    bOverrideAngularDamping=true
    bOverrideLinearDamping=true
    bUseGravityOverride=true
    gravityScale=0
    linearDampingOverride=0.995
    linearSpringConstant=100
    simulationSpace=RootRelative
    alphaBoolBlend {blendOption=Linear}
    alphaScaleBiasClamp {interpSpeedDecreasing=3, interpSpeedIncreasing=5}
    boundBone {boneName=dyn_shard_skirt_a_base_r}
    componentAppliedLinearAccClamp=<6000 6000 6000>
    componentLinearAccScale=<0.05 0.05 0.05>
    componentLinearVelScale=<0.8 0.8 0.6>
    gravityOverride=<0.25 0.25 0.25>
    physicsBodyDefinitions [{
      sphereCollisionRadius=0
      boundBone {boneName=dyn_shard_skirt_a_base_r}
      boxExtents=<50 50 50>
      constraintSetup {angularLimitsMax=<0 0 0>, angularLimitsMin=<0 0 0>, angularTarget=<0 0 0>, linearAxesMax=<0 0.5 0>, linearAxesMin=<-0.5 -0.5 -1>}
      localJointOffset=<-3 0 0>
     }]
    relativeSpaceBone {boneName=root}
   }
   n71 = Anim Dynamics - Bone: dyn_shard_skirt_b_r(ComponentPose=@n70, Alpha=1, ExternalForce="0,0,0") {
    angularDampingOverride=0.999
    angularSpringConstant=150
    bAngularSpring=true
    bGravityOverrideInSimSpace=true
    bLinearSpring=true
    bOverrideAngularDamping=true
    bOverrideLinearDamping=true
    bUseGravityOverride=true
    gravityScale=0
    linearDampingOverride=0.995
    linearSpringConstant=100
    simulationSpace=RootRelative
    alphaBoolBlend {blendOption=Linear}
    alphaScaleBiasClamp {interpSpeedDecreasing=3, interpSpeedIncreasing=5}
    boundBone {boneName=dyn_shard_skirt_b_r}
    componentAppliedLinearAccClamp=<6000 6000 6000>
    componentLinearAccScale=<0.05 0.05 0.05>
    componentLinearVelScale=<0.8 0.8 0.6>
    gravityOverride=<0.25 0.25 0.25>
    physicsBodyDefinitions [{
      sphereCollisionRadius=0
      boundBone {boneName=dyn_shard_skirt_b_r}
      boxExtents=<50 50 50>
      constraintSetup {angularLimitsMax=<0 0 0>, angularLimitsMin=<0 0 0>, angularTarget=<0 0 0>, linearAxesMax=<0 0.5 0>, linearAxesMin=<-2 -0.5 -1.5>}
      localJointOffset=<-3 0 0>
     }]
    relativeSpaceBone {boneName=root}
   }
   n72 = Anim Dynamics - Bone: dyn_shard_skirt_c_base_r(ComponentPose=@n71, Alpha=0.5, ExternalForce="0,0,0") {
    angularDampingOverride=0.999
    angularSpringConstant=150
    bAngularSpring=true
    bGravityOverrideInSimSpace=true
    bLinearSpring=true
    bOverrideAngularDamping=true
    bOverrideLinearDamping=true
    bUseGravityOverride=true
    gravityScale=0
    linearDampingOverride=0.995
    linearSpringConstant=100
    simulationSpace=RootRelative
    alphaBoolBlend {blendOption=Linear}
    alphaScaleBiasClamp {interpSpeedDecreasing=3, interpSpeedIncreasing=5}
    boundBone {boneName=dyn_shard_skirt_c_base_r}
    componentAppliedLinearAccClamp=<6000 6000 6000>
    componentLinearAccScale=<0.05 0.05 0.05>
    componentLinearVelScale=<0.8 0.8 0.6>
    gravityOverride=<0.25 0.25 0.25>
    physicsBodyDefinitions [{
      sphereCollisionRadius=0
      boundBone {boneName=dyn_shard_skirt_c_base_r}
      boxExtents=<50 50 50>
      constraintSetup {angularLimitsMax=<0 0 0>, angularLimitsMin=<0 0 0>, angularTarget=<0 0 0>, linearAxesMax=<0 0.5 0>, linearAxesMin=<-0.5 -0.5 -1>}
      localJointOffset=<-3 0 0>
     }]
    relativeSpaceBone {boneName=root}
   }
   n73 = Anim Dynamics - Bone: dyn_shard_skirt_d_base_r(ComponentPose=@n72, Alpha=0.4, ExternalForce="0,0,0") {
    angularDampingOverride=0.999
    angularSpringConstant=150
    bAngularSpring=true
    bGravityOverrideInSimSpace=true
    bLinearSpring=true
    bOverrideAngularDamping=true
    bOverrideLinearDamping=true
    bUseGravityOverride=true
    gravityScale=0
    linearDampingOverride=0.995
    linearSpringConstant=100
    simulationSpace=RootRelative
    alphaBoolBlend {blendOption=Linear}
    alphaScaleBiasClamp {interpSpeedDecreasing=3, interpSpeedIncreasing=5}
    boundBone {boneName=dyn_shard_skirt_d_base_r}
    componentAppliedLinearAccClamp=<6000 6000 6000>
    componentLinearAccScale=<0.05 0.05 0.05>
    componentLinearVelScale=<0.8 0.8 0.6>
    gravityOverride=<0.25 0.25 0.25>
    physicsBodyDefinitions [{
      sphereCollisionRadius=0
      boundBone {boneName=dyn_shard_skirt_d_base_r}
      boxExtents=<50 50 50>
      constraintSetup {angularLimitsMax=<0 0 0>, angularLimitsMin=<0 0 0>, angularTarget=<0 0 0>, linearAxesMax=<0 0.5 0>, linearAxesMin=<-1 -0.5 -1>}
      localJointOffset=<-3 0 0>
     }]
    relativeSpaceBone {boneName=root}
   }
   n74 = Anim Dynamics - Bone: dyn_shard_skirt_e_base_r(ComponentPose=@n73, Alpha=0.4, ExternalForce="0,0,0") {
    angularDampingOverride=0.999
    angularSpringConstant=150
    bAngularSpring=true
    bGravityOverrideInSimSpace=true
    bLinearSpring=true
    bOverrideAngularDamping=true
    bOverrideLinearDamping=true
    bUseGravityOverride=true
    gravityScale=0
    linearDampingOverride=0.995
    linearSpringConstant=100
    simulationSpace=RootRelative
    alphaBoolBlend {blendOption=Linear}
    alphaScaleBiasClamp {interpSpeedDecreasing=3, interpSpeedIncreasing=5}
    boundBone {boneName=dyn_shard_skirt_e_base_r}
    componentAppliedLinearAccClamp=<6000 6000 6000>
    componentLinearAccScale=<0.05 0.05 0.05>
    componentLinearVelScale=<0.8 0.8 0.6>
    gravityOverride=<0.25 0.25 0.25>
    physicsBodyDefinitions [{
      sphereCollisionRadius=0
      boundBone {boneName=dyn_shard_skirt_e_base_r}
      boxExtents=<50 50 50>
      constraintSetup {angularLimitsMax=<0 0 0>, angularLimitsMin=<0 0 0>, angularTarget=<0 0 0>, linearAxesMax=<1 0.5 0>, linearAxesMin=<-1 -0.5 -1>}
      localJointOffset=<-3 0 0>
     }]
    relativeSpaceBone {boneName=root}
   }
   n75 = Anim Dynamics - Bone: dyn_shard_skirt_f_base_r(ComponentPose=@n74, Alpha=1, ExternalForce="0,0,0") {
    angularDampingOverride=0.999
    angularSpringConstant=150
    bAngularSpring=true
    bGravityOverrideInSimSpace=true
    bLinearSpring=true
    bOverrideAngularDamping=true
    bOverrideLinearDamping=true
    bUseGravityOverride=true
    gravityScale=0
    linearDampingOverride=0.995
    linearSpringConstant=100
    simulationSpace=RootRelative
    alphaBoolBlend {blendOption=Linear}
    alphaScaleBiasClamp {interpSpeedDecreasing=3, interpSpeedIncreasing=5}
    boundBone {boneName=dyn_shard_skirt_f_base_r}
    componentAppliedLinearAccClamp=<6000 6000 6000>
    componentLinearAccScale=<0.05 0.05 0.05>
    componentLinearVelScale=<0.8 0.8 0.6>
    gravityOverride=<0.25 0.25 0.25>
    physicsBodyDefinitions [{
      sphereCollisionRadius=0
      boundBone {boneName=dyn_shard_skirt_f_base_r}
      boxExtents=<50 50 50>
      constraintSetup {angularLimitsMax=<0 0 0>, angularLimitsMin=<0 0 0>, angularTarget=<0 0 0>, linearAxesMax=<1 0.5 0.5>, linearAxesMin=<-2 -0.5 -2>}
      localJointOffset=<-3 0 0>
     }]
    relativeSpaceBone {boneName=root}
   }
   n76 = Anim Dynamics - Bone: dyn_shard_skirt_g_base_r(ComponentPose=@n75, Alpha=0.3, ExternalForce="0,0,0") {
    angularDampingOverride=0.999
    angularSpringConstant=150
    bAngularSpring=true
    bGravityOverrideInSimSpace=true
    bLinearSpring=true
    bOverrideAngularDamping=true
    bOverrideLinearDamping=true
    bUseGravityOverride=true
    gravityScale=0
    linearDampingOverride=0.995
    linearSpringConstant=100
    simulationSpace=RootRelative
    alphaBoolBlend {blendOption=Linear}
    alphaScaleBiasClamp {interpSpeedDecreasing=3, interpSpeedIncreasing=5}
    boundBone {boneName=dyn_shard_skirt_g_base_r}
    componentAppliedLinearAccClamp=<6000 6000 6000>
    componentLinearAccScale=<0.05 0.05 0.05>
    componentLinearVelScale=<0.8 0.8 0.6>
    gravityOverride=<0.25 0.25 0.25>
    physicsBodyDefinitions [{
      sphereCollisionRadius=0
      boundBone {boneName=dyn_shard_skirt_g_base_r}
      boxExtents=<50 50 50>
      constraintSetup {angularLimitsMax=<0 0 0>, angularLimitsMin=<0 0 0>, angularTarget=<0 0 0>, linearAxesMax=<0.5 0.5 0.5>, linearAxesMin=<-0.5 -0.5 -0.5>}
      localJointOffset=<-3 0 0>
     }]
    relativeSpaceBone {boneName=root}
   }
   n77 = Anim Dynamics - Bone: dyn_shard_skirt_h_base_r(ComponentPose=@n76, Alpha=1, ExternalForce="0,0,0") {
    angularDampingOverride=0.999
    angularSpringConstant=150
    bAngularSpring=true
    bGravityOverrideInSimSpace=true
    bLinearSpring=true
    bOverrideAngularDamping=true
    bOverrideLinearDamping=true
    bUseGravityOverride=true
    gravityScale=0
    linearDampingOverride=0.995
    linearSpringConstant=100
    simulationSpace=RootRelative
    alphaBoolBlend {blendOption=Linear}
    alphaScaleBiasClamp {interpSpeedDecreasing=3, interpSpeedIncreasing=5}
    boundBone {boneName=dyn_shard_skirt_h_base_r}
    componentAppliedLinearAccClamp=<6000 6000 6000>
    componentLinearAccScale=<0.05 0.05 0.05>
    componentLinearVelScale=<0.8 0.8 0.6>
    gravityOverride=<0.25 0.25 0.25>
    physicsBodyDefinitions [{
      sphereCollisionRadius=0
      boundBone {boneName=dyn_shard_skirt_h_base_r}
      boxExtents=<50 50 50>
      constraintSetup {angularLimitsMax=<0 0 0>, angularLimitsMin=<0 0 0>, angularTarget=<0 0 0>, linearAxesMax=<1 0.5 0.5>, linearAxesMin=<-1 -0.5 -0.5>}
      localJointOffset=<-3 0 0>
     }]
    relativeSpaceBone {boneName=root}
   }
   n78 = Anim Dynamics - Bone: dyn_shard_skirt_i_base_r(ComponentPose=@n77, Alpha=0.3, ExternalForce="0,0,0") {
    angularDampingOverride=0.999
    angularSpringConstant=150
    bAngularSpring=true
    bGravityOverrideInSimSpace=true
    bLinearSpring=true
    bOverrideAngularDamping=true
    bOverrideLinearDamping=true
    bUseGravityOverride=true
    gravityScale=0
    linearDampingOverride=0.995
    linearSpringConstant=100
    simulationSpace=RootRelative
    alphaBoolBlend {blendOption=Linear}
    alphaScaleBiasClamp {interpSpeedDecreasing=3, interpSpeedIncreasing=5}
    boundBone {boneName=dyn_shard_skirt_i_base_r}
    componentAppliedLinearAccClamp=<6000 6000 6000>
    componentLinearAccScale=<0.05 0.05 0.05>
    componentLinearVelScale=<0.8 0.8 0.6>
    gravityOverride=<0.25 0.25 0.25>
    physicsBodyDefinitions [{
      sphereCollisionRadius=0
      boundBone {boneName=dyn_shard_skirt_i_base_r}
      boxExtents=<50 50 50>
      constraintSetup {angularLimitsMax=<0 0 0>, angularLimitsMin=<0 0 0>, angularTarget=<0 0 0>, linearAxesMax=<0.5 0.5 0.5>, linearAxesMin=<-0.5 -0.5 -0.5>}
      localJointOffset=<-3 0 0>
     }]
    relativeSpaceBone {boneName=root}
   }
   n79 = Anim Dynamics - Bone: dyn_shard_skirt_j_base_r(ComponentPose=@n78, Alpha=0.3, ExternalForce="0,0,0") {
    angularDampingOverride=0.999
    angularSpringConstant=150
    bAngularSpring=true
    bGravityOverrideInSimSpace=true
    bLinearSpring=true
    bOverrideAngularDamping=true
    bOverrideLinearDamping=true
    bUseGravityOverride=true
    gravityScale=0
    linearDampingOverride=0.995
    linearSpringConstant=100
    simulationSpace=RootRelative
    alphaBoolBlend {blendOption=Linear}
    alphaScaleBiasClamp {interpSpeedDecreasing=3, interpSpeedIncreasing=5}
    boundBone {boneName=dyn_shard_skirt_j_base_r}
    componentAppliedLinearAccClamp=<6000 6000 6000>
    componentLinearAccScale=<0.05 0.05 0.05>
    componentLinearVelScale=<0.8 0.8 0.6>
    gravityOverride=<0.25 0.25 0.25>
    physicsBodyDefinitions [{
      sphereCollisionRadius=0
      boundBone {boneName=dyn_shard_skirt_j_base_r}
      boxExtents=<50 50 50>
      constraintSetup {angularLimitsMax=<0 0 0>, angularLimitsMin=<0 0 0>, angularTarget=<0 0 0>, linearAxesMax=<1 0.5 0.5>, linearAxesMin=<-1 -0.5 -0.5>}
      localJointOffset=<-3 0 0>
     }]
    relativeSpaceBone {boneName=root}
   }
   n80 = Anim Dynamics - Bone: dyn_shard_shoulder_a_base_l(ComponentPose=@n79, Alpha=0.25, ExternalForce="0,0,0") {
    angularDampingOverride=0.999
    angularSpringConstant=150
    bAngularSpring=true
    bGravityOverrideInSimSpace=true
    bLinearSpring=true
    bOverrideAngularDamping=true
    bOverrideLinearDamping=true
    bUseGravityOverride=true
    gravityScale=0
    linearDampingOverride=0.995
    linearSpringConstant=100
    simulationSpace=RootRelative
    alphaBoolBlend {blendOption=Linear}
    alphaScaleBiasClamp {interpSpeedDecreasing=3, interpSpeedIncreasing=5}
    boundBone {boneName=dyn_shard_shoulder_a_base_l}
    componentAppliedLinearAccClamp=<6000 6000 6000>
    componentLinearAccScale=<0.05 0.05 0.05>
    componentLinearVelScale=<0.8 0.8 0.6>
    gravityOverride=<0.25 0.25 0.25>
    physicsBodyDefinitions [{
      sphereCollisionRadius=0
      boundBone {boneName=dyn_shard_shoulder_a_base_l}
      boxExtents=<50 50 50>
      constraintSetup {angularLimitsMax=<0 0 0>, angularLimitsMin=<0 0 0>, angularTarget=<0 0 0>, linearAxesMax=<0.5 0.5 0.5>, linearAxesMin=<-0.5 -0.5 -0.5>}
      localJointOffset=<-3 0 0>
     }]
    relativeSpaceBone {boneName=root}
   }
   n81 = Anim Dynamics - Bone: dyn_shard_shoulder_b_base_l(ComponentPose=@n80, Alpha=0.3, ExternalForce="0,0,0") {
    angularDampingOverride=0.999
    angularSpringConstant=150
    bAngularSpring=true
    bGravityOverrideInSimSpace=true
    bLinearSpring=true
    bOverrideAngularDamping=true
    bOverrideLinearDamping=true
    bUseGravityOverride=true
    gravityScale=0
    linearDampingOverride=0.995
    linearSpringConstant=100
    simulationSpace=RootRelative
    alphaBoolBlend {blendOption=Linear}
    alphaScaleBiasClamp {interpSpeedDecreasing=3, interpSpeedIncreasing=5}
    boundBone {boneName=dyn_shard_shoulder_b_base_l}
    componentAppliedLinearAccClamp=<6000 6000 6000>
    componentLinearAccScale=<0.05 0.05 0.05>
    componentLinearVelScale=<0.8 0.8 0.6>
    gravityOverride=<0.25 0.25 0.25>
    physicsBodyDefinitions [{
      sphereCollisionRadius=0
      boundBone {boneName=dyn_shard_shoulder_b_base_l}
      boxExtents=<50 50 50>
      constraintSetup {angularLimitsMax=<0 0 0>, angularLimitsMin=<0 0 0>, angularTarget=<0 0 0>, linearAxesMax=<0.5 0.5 0.5>, linearAxesMin=<-0.5 -0.5 -0.5>}
      localJointOffset=<-3 0 0>
     }]
    relativeSpaceBone {boneName=root}
   }
   n82 = Anim Dynamics - Bone: dyn_shard_shoulder_c_base_l(ComponentPose=@n81, Alpha=0.8, ExternalForce="0,0,0") {
    angularDampingOverride=0.999
    angularSpringConstant=150
    bAngularSpring=true
    bGravityOverrideInSimSpace=true
    bLinearSpring=true
    bOverrideAngularDamping=true
    bOverrideLinearDamping=true
    bUseGravityOverride=true
    gravityScale=0
    linearDampingOverride=0.995
    linearSpringConstant=100
    simulationSpace=RootRelative
    alphaBoolBlend {blendOption=Linear}
    alphaScaleBiasClamp {interpSpeedDecreasing=3, interpSpeedIncreasing=5}
    boundBone {boneName=dyn_shard_shoulder_c_base_l}
    componentAppliedLinearAccClamp=<6000 6000 6000>
    componentLinearAccScale=<0.05 0.05 0.05>
    componentLinearVelScale=<0.8 0.8 0.6>
    gravityOverride=<0.25 0.25 0.25>
    physicsBodyDefinitions [{
      sphereCollisionRadius=0
      boundBone {boneName=dyn_shard_shoulder_c_base_l}
      boxExtents=<50 50 50>
      constraintSetup {angularLimitsMax=<0 0 0>, angularLimitsMin=<0 0 0>, angularTarget=<0 0 0>, linearAxesMax=<0.5 0.5 0.5>, linearAxesMin=<-0.5 -0.5 -0.5>}
      localJointOffset=<-3 0 0>
     }]
    relativeSpaceBone {boneName=root}
   }
   n83 = Anim Dynamics - Bone: dyn_shard_shoulder_d_base_l(ComponentPose=@n82, Alpha=0.9, ExternalForce="0,0,0") {
    angularDampingOverride=0.999
    angularSpringConstant=150
    bAngularSpring=true
    bGravityOverrideInSimSpace=true
    bLinearSpring=true
    bOverrideAngularDamping=true
    bOverrideLinearDamping=true
    bUseGravityOverride=true
    gravityScale=0
    linearDampingOverride=0.995
    linearSpringConstant=100
    simulationSpace=RootRelative
    alphaBoolBlend {blendOption=Linear}
    alphaScaleBiasClamp {interpSpeedDecreasing=3, interpSpeedIncreasing=5}
    boundBone {boneName=dyn_shard_shoulder_d_base_l}
    componentAppliedLinearAccClamp=<6000 6000 6000>
    componentLinearAccScale=<0.05 0.05 0.05>
    componentLinearVelScale=<0.8 0.8 0.6>
    gravityOverride=<0.25 0.25 0.25>
    physicsBodyDefinitions [{
      sphereCollisionRadius=0
      boundBone {boneName=dyn_shard_shoulder_d_base_l}
      boxExtents=<50 50 50>
      constraintSetup {angularLimitsMax=<0 0 0>, angularLimitsMin=<0 0 0>, angularTarget=<0 0 0>, linearAxesMax=<0.5 0.5 0.5>, linearAxesMin=<-0.5 -0.5 -0.5>}
      localJointOffset=<-3 0 0>
     }]
    relativeSpaceBone {boneName=root}
   }
   n84 = Anim Dynamics - Bone: dyn_shard_shoulder_e_base_l(ComponentPose=@n83, Alpha=0.7, ExternalForce="0,0,0") {
    angularDampingOverride=0.999
    angularSpringConstant=150
    bAngularSpring=true
    bGravityOverrideInSimSpace=true
    bLinearSpring=true
    bOverrideAngularDamping=true
    bOverrideLinearDamping=true
    bUseGravityOverride=true
    gravityScale=0
    linearDampingOverride=0.995
    linearSpringConstant=100
    simulationSpace=RootRelative
    alphaBoolBlend {blendOption=Linear}
    alphaScaleBiasClamp {interpSpeedDecreasing=3, interpSpeedIncreasing=5}
    boundBone {boneName=dyn_shard_shoulder_e_base_l}
    componentAppliedLinearAccClamp=<6000 6000 6000>
    componentLinearAccScale=<0.05 0.05 0.05>
    componentLinearVelScale=<0.8 0.8 0.6>
    gravityOverride=<0.25 0.25 0.25>
    physicsBodyDefinitions [{
      sphereCollisionRadius=0
      boundBone {boneName=dyn_shard_shoulder_e_base_l}
      boxExtents=<50 50 50>
      constraintSetup {angularLimitsMax=<0 0 0>, angularLimitsMin=<0 0 0>, angularTarget=<0 0 0>, linearAxesMax=<0.5 0.5 0.5>, linearAxesMin=<-0.5 -0.5 -0.5>}
      localJointOffset=<-3 0 0>
     }]
    relativeSpaceBone {boneName=root}
   }
   n85 = Anim Dynamics - Bone: dyn_shard_shoulder_f_base_l(ComponentPose=@n84, Alpha=1, ExternalForce="0,0,0") {
    angularDampingOverride=0.999
    angularSpringConstant=150
    bAngularSpring=true
    bGravityOverrideInSimSpace=true
    bLinearSpring=true
    bOverrideAngularDamping=true
    bOverrideLinearDamping=true
    bUseGravityOverride=true
    gravityScale=0
    linearDampingOverride=0.995
    linearSpringConstant=100
    simulationSpace=RootRelative
    alphaBoolBlend {blendOption=Linear}
    alphaScaleBiasClamp {interpSpeedDecreasing=3, interpSpeedIncreasing=5}
    boundBone {boneName=dyn_shard_shoulder_f_base_l}
    componentAppliedLinearAccClamp=<6000 6000 6000>
    componentLinearAccScale=<0.05 0.05 0.05>
    componentLinearVelScale=<0.8 0.8 0.6>
    gravityOverride=<0.25 0.25 0.25>
    physicsBodyDefinitions [{
      sphereCollisionRadius=0
      boundBone {boneName=dyn_shard_shoulder_f_base_l}
      boxExtents=<50 50 50>
      constraintSetup {angularLimitsMax=<0 0 0>, angularLimitsMin=<0 0 0>, angularTarget=<0 0 0>, linearAxesMax=<0.5 0.5 0.5>, linearAxesMin=<-0.5 -0.5 -0.5>}
      localJointOffset=<-3 0 0>
     }]
    relativeSpaceBone {boneName=root}
   }
   n86 = Anim Dynamics - Bone: dyn_shard_shoulder_g_base_l(ComponentPose=@n85, Alpha=0.4, ExternalForce="0,0,0") {
    angularDampingOverride=0.999
    angularSpringConstant=150
    bAngularSpring=true
    bGravityOverrideInSimSpace=true
    bLinearSpring=true
    bOverrideAngularDamping=true
    bOverrideLinearDamping=true
    bUseGravityOverride=true
    gravityScale=0
    linearDampingOverride=0.995
    linearSpringConstant=100
    simulationSpace=RootRelative
    alphaBoolBlend {blendOption=Linear}
    alphaScaleBiasClamp {interpSpeedDecreasing=3, interpSpeedIncreasing=5}
    boundBone {boneName=dyn_shard_shoulder_g_base_l}
    componentAppliedLinearAccClamp=<6000 6000 6000>
    componentLinearAccScale=<0.05 0.05 0.05>
    componentLinearVelScale=<0.8 0.8 0.6>
    gravityOverride=<0.25 0.25 0.25>
    physicsBodyDefinitions [{
      sphereCollisionRadius=0
      boundBone {boneName=dyn_shard_shoulder_g_base_l}
      boxExtents=<50 50 50>
      constraintSetup {angularLimitsMax=<0 0 0>, angularLimitsMin=<0 0 0>, angularTarget=<0 0 0>, linearAxesMax=<0.5 0.5 0.5>, linearAxesMin=<-0.5 -0.5 -0.5>}
      localJointOffset=<-3 0 0>
     }]
    relativeSpaceBone {boneName=root}
   }
   n87 = Anim Dynamics - Bone: dyn_shard_shoulder_h_base_l(ComponentPose=@n86, Alpha=0.5, ExternalForce="0,0,0") {
    angularDampingOverride=0.999
    angularSpringConstant=150
    bAngularSpring=true
    bGravityOverrideInSimSpace=true
    bLinearSpring=true
    bOverrideAngularDamping=true
    bOverrideLinearDamping=true
    bUseGravityOverride=true
    gravityScale=0
    linearDampingOverride=0.995
    linearSpringConstant=100
    simulationSpace=RootRelative
    alphaBoolBlend {blendOption=Linear}
    alphaScaleBiasClamp {interpSpeedDecreasing=3, interpSpeedIncreasing=5}
    boundBone {boneName=dyn_shard_shoulder_h_base_l}
    componentAppliedLinearAccClamp=<6000 6000 6000>
    componentLinearAccScale=<0.05 0.05 0.05>
    componentLinearVelScale=<0.8 0.8 0.6>
    gravityOverride=<0.25 0.25 0.25>
    physicsBodyDefinitions [{
      sphereCollisionRadius=0
      boundBone {boneName=dyn_shard_shoulder_h_base_l}
      boxExtents=<50 50 50>
      constraintSetup {angularLimitsMax=<0 0 0>, angularLimitsMin=<0 0 0>, angularTarget=<0 0 0>, linearAxesMax=<0.5 0.5 0.5>, linearAxesMin=<-0.5 -0.5 -0.5>}
      localJointOffset=<-3 0 0>
     }]
    relativeSpaceBone {boneName=root}
   }
   n88 = Anim Dynamics - Bone: dyn_shard_elbow_a_base_r(ComponentPose=@n87, Alpha=0.7, ExternalForce="0,0,0") {
    angularDampingOverride=0.999
    angularSpringConstant=150
    bAngularSpring=true
    bGravityOverrideInSimSpace=true
    bLinearSpring=true
    bOverrideAngularDamping=true
    bOverrideLinearDamping=true
    bUseGravityOverride=true
    gravityScale=0
    linearDampingOverride=0.995
    linearSpringConstant=100
    simulationSpace=RootRelative
    alphaBoolBlend {blendOption=Linear}
    alphaScaleBiasClamp {interpSpeedDecreasing=3, interpSpeedIncreasing=5}
    boundBone {boneName=dyn_shard_elbow_a_base_r}
    componentAppliedLinearAccClamp=<6000 6000 6000>
    componentLinearAccScale=<0.05 0.05 0.05>
    componentLinearVelScale=<0.8 0.8 0.6>
    gravityOverride=<0.25 0.25 0.25>
    physicsBodyDefinitions [{
      sphereCollisionRadius=0
      boundBone {boneName=dyn_shard_elbow_a_base_r}
      boxExtents=<50 50 50>
      constraintSetup {angularLimitsMax=<0 0 0>, angularLimitsMin=<0 0 0>, angularTarget=<0 0 0>, linearAxesMax=<0.5 0.5 0.5>, linearAxesMin=<-0.5 -0.5 -0.5>}
      localJointOffset=<-3 0 0>
     }]
    relativeSpaceBone {boneName=root}
   }
   n89 = Anim Dynamics - Bone: dyn_shard_elbow_b_base_r(ComponentPose=@n88, Alpha=0.5, ExternalForce="0,0,0") {
    angularDampingOverride=0.999
    angularSpringConstant=150
    bAngularSpring=true
    bGravityOverrideInSimSpace=true
    bLinearSpring=true
    bOverrideAngularDamping=true
    bOverrideLinearDamping=true
    bUseGravityOverride=true
    gravityScale=0
    linearDampingOverride=0.995
    linearSpringConstant=100
    simulationSpace=RootRelative
    alphaBoolBlend {blendOption=Linear}
    alphaScaleBiasClamp {interpSpeedDecreasing=3, interpSpeedIncreasing=5}
    boundBone {boneName=dyn_shard_elbow_b_base_r}
    componentAppliedLinearAccClamp=<6000 6000 6000>
    componentLinearAccScale=<0.05 0.05 0.05>
    componentLinearVelScale=<0.8 0.8 0.6>
    gravityOverride=<0.25 0.25 0.25>
    physicsBodyDefinitions [{
      sphereCollisionRadius=0
      boundBone {boneName=dyn_shard_elbow_b_base_r}
      boxExtents=<50 50 50>
      constraintSetup {angularLimitsMax=<0 0 0>, angularLimitsMin=<0 0 0>, angularTarget=<0 0 0>, linearAxesMax=<0.5 0.5 0.5>, linearAxesMin=<-0.5 -0.5 -0.5>}
      localJointOffset=<-3 0 0>
     }]
    relativeSpaceBone {boneName=root}
   }
   n90 = Anim Dynamics - Bone: dyn_shard_knee_a_base_l(ComponentPose=@n89, Alpha=1, ExternalForce="0,0,0") {
    angularDampingOverride=0.999
    angularSpringConstant=150
    bAngularSpring=true
    bGravityOverrideInSimSpace=true
    bLinearSpring=true
    bOverrideAngularDamping=true
    bOverrideLinearDamping=true
    bUseGravityOverride=true
    gravityScale=0
    linearDampingOverride=0.995
    linearSpringConstant=100
    simulationSpace=RootRelative
    alphaBoolBlend {blendOption=Linear}
    alphaScaleBiasClamp {interpSpeedDecreasing=3, interpSpeedIncreasing=5}
    boundBone {boneName=dyn_shard_knee_a_base_l}
    componentAppliedLinearAccClamp=<6000 6000 6000>
    componentLinearAccScale=<0.05 0.05 0.05>
    componentLinearVelScale=<0.8 0.8 0.6>
    gravityOverride=<0.25 0.25 0.25>
    physicsBodyDefinitions [{
      sphereCollisionRadius=0
      boundBone {boneName=dyn_shard_knee_a_base_l}
      boxExtents=<50 50 50>
      constraintSetup {angularLimitsMax=<0 0 0>, angularLimitsMin=<0 0 0>, angularTarget=<0 0 0>, linearAxesMax=<0.5 0.5 0.5>, linearAxesMin=<-0.5 -0.5 -0.5>}
      localJointOffset=<-3 0 0>
     }]
    relativeSpaceBone {boneName=root}
   }
   n91 = Anim Dynamics - Bone: dyn_shard_torso_a_base_l(ComponentPose=@n90, Alpha=1, ExternalForce="0,0,0") {
    angularDampingOverride=0.999
    angularSpringConstant=150
    bAngularSpring=true
    bGravityOverrideInSimSpace=true
    bLinearSpring=true
    bOverrideAngularDamping=true
    bOverrideLinearDamping=true
    bUseGravityOverride=true
    gravityScale=0
    linearDampingOverride=0.995
    linearSpringConstant=100
    simulationSpace=RootRelative
    alphaBoolBlend {blendOption=Linear}
    alphaScaleBiasClamp {interpSpeedDecreasing=3, interpSpeedIncreasing=5}
    boundBone {boneName=dyn_shard_torso_a_base_l}
    componentAppliedLinearAccClamp=<6000 6000 6000>
    componentLinearAccScale=<0.05 0.05 0.05>
    componentLinearVelScale=<0.8 0.8 0.6>
    gravityOverride=<0.25 0.25 0.25>
    physicsBodyDefinitions [{
      sphereCollisionRadius=0
      boundBone {boneName=dyn_shard_torso_a_base_l}
      boxExtents=<50 50 50>
      constraintSetup {angularLimitsMax=<0 0 0>, angularLimitsMin=<0 0 0>, angularTarget=<0 0 0>, linearAxesMax=<0.5 0.5 0.5>, linearAxesMin=<-0.5 -0.5 -0.5>}
      localJointOffset=<-3 0 0>
     }]
    relativeSpaceBone {boneName=root}
   }
   n92 = "Transform (Modify) Bone - Bone: dyn_shard_skirt_a_l"(ComponentPose=@n91, Alpha=0.3, Translation=Get floatA() -> (floatA:Vector), Rotation="0,0,0", Scale="1,1,1") {
    rotationMode=BMM_Additive
    rotationSpace=BCS_BoneSpace
    scaleMode=BMM_Additive
    scaleSpace=BCS_BoneSpace
    translationMode=BMM_Additive
    translationSpace=BCS_BoneSpace
    alphaBoolBlend {blendOption=Linear}
    boneToModify {boneName=dyn_shard_skirt_a_l}
   }
   n93 = "Transform (Modify) Bone - Bone: dyn_shard_skirt_b_l"(ComponentPose=@n92, Alpha=0.7, Translation=Get floatB() -> (floatB:Vector), Rotation="0,0,0", Scale="1,1,1") {
    rotationMode=BMM_Additive
    rotationSpace=BCS_BoneSpace
    scaleMode=BMM_Additive
    scaleSpace=BCS_BoneSpace
    translationMode=BMM_Additive
    translationSpace=BCS_BoneSpace
    alphaBoolBlend {blendOption=Linear}
    boneToModify {boneName=dyn_shard_skirt_b_l}
   }
   n94 = "Transform (Modify) Bone - Bone: dyn_shard_skirt_c_l"(ComponentPose=@n93, Alpha=1, Translation=Get floatA(), Rotation="0,0,0", Scale="1,1,1") {
    rotationMode=BMM_Additive
    rotationSpace=BCS_BoneSpace
    scaleMode=BMM_Additive
    scaleSpace=BCS_BoneSpace
    translationMode=BMM_Additive
    translationSpace=BCS_BoneSpace
    alphaBoolBlend {blendOption=Linear}
    boneToModify {boneName=dyn_shard_skirt_c_l}
   }
   n95 = "Transform (Modify) Bone - Bone: dyn_shard_skirt_d_l"(ComponentPose=@n94, Alpha=0.7, Translation=Get floatB(), Rotation="0,0,0", Scale="1,1,1") {
    rotationMode=BMM_Additive
    rotationSpace=BCS_BoneSpace
    scaleMode=BMM_Additive
    scaleSpace=BCS_BoneSpace
    translationMode=BMM_Additive
    translationSpace=BCS_BoneSpace
    alphaBoolBlend {blendOption=Linear}
    boneToModify {boneName=dyn_shard_skirt_d_l}
   }
   n96 = "Transform (Modify) Bone - Bone: dyn_shard_skirt_e_l"(ComponentPose=@n95, Alpha=0.5, Translation=Get floatA(), Rotation="0,0,0", Scale="1,1,1") {
    rotationMode=BMM_Additive
    rotationSpace=BCS_BoneSpace
    scaleMode=BMM_Additive
    scaleSpace=BCS_BoneSpace
    translationMode=BMM_Additive
    translationSpace=BCS_BoneSpace
    alphaBoolBlend {blendOption=Linear}
    boneToModify {boneName=dyn_shard_skirt_e_l}
   }
   n97 = "Transform (Modify) Bone - Bone: dyn_shard_skirt_a_r"(ComponentPose=@n96, Alpha=0.75, Translation=Get floatA(), Rotation="0,0,0", Scale="1,1,1") {
    rotationMode=BMM_Additive
    rotationSpace=BCS_BoneSpace
    scaleMode=BMM_Additive
    scaleSpace=BCS_BoneSpace
    translationMode=BMM_Additive
    translationSpace=BCS_BoneSpace
    alphaBoolBlend {blendOption=Linear}
    boneToModify {boneName=dyn_shard_skirt_a_r}
   }
   n98 = "Transform (Modify) Bone - Bone: dyn_shard_skirt_b_r"(ComponentPose=@n97, Alpha=1, Translation=Get floatB(), Rotation="0,0,0", Scale="1,1,1") {
    rotationMode=BMM_Additive
    rotationSpace=BCS_BoneSpace
    scaleMode=BMM_Additive
    scaleSpace=BCS_BoneSpace
    translationMode=BMM_Additive
    translationSpace=BCS_BoneSpace
    alphaBoolBlend {blendOption=Linear}
    boneToModify {boneName=dyn_shard_skirt_b_r}
   }
   n99 = "Transform (Modify) Bone - Bone: dyn_shard_skirt_c_r"(ComponentPose=@n98, Alpha=1, Translation=Get floatA(), Rotation="0,0,0", Scale="1,1,1") {
    rotationMode=BMM_Additive
    rotationSpace=BCS_BoneSpace
    scaleMode=BMM_Additive
    scaleSpace=BCS_BoneSpace
    translationMode=BMM_Additive
    translationSpace=BCS_BoneSpace
    alphaBoolBlend {blendOption=Linear}
    boneToModify {boneName=dyn_shard_skirt_c_r}
   }
   n100 = "Transform (Modify) Bone - Bone: dyn_shard_skirt_d_r"(ComponentPose=@n99, Alpha=0.3, Translation=Get floatB(), Rotation="0,0,0", Scale="1,1,1") {
    rotationMode=BMM_Additive
    rotationSpace=BCS_BoneSpace
    scaleMode=BMM_Additive
    scaleSpace=BCS_BoneSpace
    translationMode=BMM_Additive
    translationSpace=BCS_BoneSpace
    alphaBoolBlend {blendOption=Linear}
    boneToModify {boneName=dyn_shard_skirt_d_r}
   }
   n101 = "Transform (Modify) Bone - Bone: dyn_shard_skirt_e_r"(ComponentPose=@n100, Alpha=1, Translation=Get floatA(), Rotation="0,0,0", Scale="1,1,1") {
    rotationMode=BMM_Additive
    rotationSpace=BCS_BoneSpace
    scaleMode=BMM_Additive
    scaleSpace=BCS_BoneSpace
    translationMode=BMM_Additive
    translationSpace=BCS_BoneSpace
    alphaBoolBlend {blendOption=Linear}
    boneToModify {boneName=dyn_shard_skirt_e_r}
   }
   n102 = "Transform (Modify) Bone - Bone: dyn_shard_skirt_f_r"(ComponentPose=@n101, Alpha=1, Translation=Get floatB(), Rotation="0,0,0", Scale="1,1,1") {
    rotationMode=BMM_Additive
    rotationSpace=BCS_BoneSpace
    scaleMode=BMM_Additive
    scaleSpace=BCS_BoneSpace
    translationMode=BMM_Additive
    translationSpace=BCS_BoneSpace
    alphaBoolBlend {blendOption=Linear}
    boneToModify {boneName=dyn_shard_skirt_f_r}
   }
   n103 = "Transform (Modify) Bone - Bone: dyn_shard_skirt_g_r"(ComponentPose=@n102, Alpha=0.3, Translation=Get floatA(), Rotation="0,0,0", Scale="1,1,1") {
    rotationMode=BMM_Additive
    rotationSpace=BCS_BoneSpace
    scaleMode=BMM_Additive
    scaleSpace=BCS_BoneSpace
    translationMode=BMM_Additive
    translationSpace=BCS_BoneSpace
    alphaBoolBlend {blendOption=Linear}
    boneToModify {boneName=dyn_shard_skirt_g_r}
   }
   n104 = "Transform (Modify) Bone - Bone: dyn_shard_skirt_h_r"(ComponentPose=@n103, Alpha=1, Translation=Get floatB(), Rotation="0,0,0", Scale="1,1,1") {
    rotationMode=BMM_Additive
    rotationSpace=BCS_BoneSpace
    scaleMode=BMM_Additive
    scaleSpace=BCS_BoneSpace
    translationMode=BMM_Additive
    translationSpace=BCS_BoneSpace
    alphaBoolBlend {blendOption=Linear}
    boneToModify {boneName=dyn_shard_skirt_h_r}
   }
   n105 = "Transform (Modify) Bone - Bone: dyn_shard_skirt_i_r"(ComponentPose=@n104, Alpha=0.5, Translation=Get floatA(), Rotation="0,0,0", Scale="1,1,1") {
    rotationMode=BMM_Additive
    rotationSpace=BCS_BoneSpace
    scaleMode=BMM_Additive
    scaleSpace=BCS_BoneSpace
    translationMode=BMM_Additive
    translationSpace=BCS_BoneSpace
    alphaBoolBlend {blendOption=Linear}
    boneToModify {boneName=dyn_shard_skirt_i_r}
   }
   n106 = "Transform (Modify) Bone - Bone: dyn_shard_skirt_j_r"(ComponentPose=@n105, Alpha=0.5, Translation=Get floatB(), Rotation="0,0,0", Scale="1,1,1") {
    rotationMode=BMM_Additive
    rotationSpace=BCS_BoneSpace
    scaleMode=BMM_Additive
    scaleSpace=BCS_BoneSpace
    translationMode=BMM_Additive
    translationSpace=BCS_BoneSpace
    alphaBoolBlend {blendOption=Linear}
    boneToModify {boneName=dyn_shard_skirt_j_r}
   }
   n107 = "Transform (Modify) Bone - Bone: dyn_shard_elbow_a_r"(ComponentPose=@n106, Alpha=1, Translation=Get floatB(), Rotation="0,0,0", Scale="1,1,1") {
    rotationMode=BMM_Additive
    rotationSpace=BCS_BoneSpace
    scaleMode=BMM_Additive
    scaleSpace=BCS_BoneSpace
    translationMode=BMM_Additive
    translationSpace=BCS_BoneSpace
    alphaBoolBlend {blendOption=Linear}
    boneToModify {boneName=dyn_shard_elbow_a_r}
   }
   n108 = "Transform (Modify) Bone - Bone: dyn_shard_elbow_b_r"(ComponentPose=@n107, Alpha=0.5, Translation=Get floatA(), Rotation="0,0,0", Scale="1,1,1") {
    rotationMode=BMM_Additive
    rotationSpace=BCS_BoneSpace
    scaleMode=BMM_Additive
    scaleSpace=BCS_BoneSpace
    translationMode=BMM_Additive
    translationSpace=BCS_BoneSpace
    alphaBoolBlend {blendOption=Linear}
    boneToModify {boneName=dyn_shard_elbow_b_r}
   }
   n109 = "Transform (Modify) Bone - Bone: dyn_shard_knee_a_l"(ComponentPose=@n108, Alpha=1, Translation=Get floatB(), Rotation="0,0,0", Scale="1,1,1") {
    rotationMode=BMM_Additive
    rotationSpace=BCS_BoneSpace
    scaleMode=BMM_Additive
    scaleSpace=BCS_BoneSpace
    translationMode=BMM_Additive
    translationSpace=BCS_BoneSpace
    alphaBoolBlend {blendOption=Linear}
    boneToModify {boneName=dyn_shard_knee_a_l}
   }
   n110 = "Transform (Modify) Bone - Bone: dyn_shard_torso_a_l"(ComponentPose=@n109, Alpha=0.7, Translation=Get floatA(), Rotation="0,0,0", Scale="1,1,1") {
    rotationMode=BMM_Additive
    rotationSpace=BCS_BoneSpace
    scaleMode=BMM_Additive
    scaleSpace=BCS_BoneSpace
    translationMode=BMM_Additive
    translationSpace=BCS_BoneSpace
    alphaBoolBlend {blendOption=Linear}
    boneToModify {boneName=dyn_shard_torso_a_l}
   }
   n111 = "Transform (Modify) Bone - Bone: dyn_shard_shoulder_a_l"(ComponentPose=@n110, Alpha=0.5, Translation=Get floatA(), Rotation="0,0,0", Scale="1,1,1") {
    rotationMode=BMM_Additive
    rotationSpace=BCS_BoneSpace
    scaleMode=BMM_Additive
    scaleSpace=BCS_BoneSpace
    translationMode=BMM_Additive
    translationSpace=BCS_BoneSpace
    alphaBoolBlend {blendOption=Linear}
    boneToModify {boneName=dyn_shard_shoulder_a_l}
   }
   n112 = "Transform (Modify) Bone - Bone: dyn_shard_shoulder_b_l"(ComponentPose=@n111, Alpha=0.7, Translation=Get floatB(), Rotation="0,0,0", Scale="1,1,1") {
    rotationMode=BMM_Additive
    rotationSpace=BCS_BoneSpace
    scaleMode=BMM_Additive
    scaleSpace=BCS_BoneSpace
    translationMode=BMM_Additive
    translationSpace=BCS_BoneSpace
    alphaBoolBlend {blendOption=Linear}
    boneToModify {boneName=dyn_shard_shoulder_b_l}
   }
   n113 = "Transform (Modify) Bone - Bone: dyn_shard_shoulder_c_l"(ComponentPose=@n112, Alpha=1, Translation=Get floatA(), Rotation="0,0,0", Scale="1,1,1") {
    rotationMode=BMM_Additive
    rotationSpace=BCS_BoneSpace
    scaleMode=BMM_Additive
    scaleSpace=BCS_BoneSpace
    translationMode=BMM_Additive
    translationSpace=BCS_BoneSpace
    alphaBoolBlend {blendOption=Linear}
    boneToModify {boneName=dyn_shard_shoulder_c_l}
   }
   n114 = "Transform (Modify) Bone - Bone: dyn_shard_shoulder_d_l"(ComponentPose=@n113, Alpha=1, Translation=Get floatB(), Rotation="0,0,0", Scale="1,1,1") {
    rotationMode=BMM_Additive
    rotationSpace=BCS_BoneSpace
    scaleMode=BMM_Additive
    scaleSpace=BCS_BoneSpace
    translationMode=BMM_Additive
    translationSpace=BCS_BoneSpace
    alphaBoolBlend {blendOption=Linear}
    boneToModify {boneName=dyn_shard_shoulder_d_l}
   }
   n115 = "Transform (Modify) Bone - Bone: dyn_shard_shoulder_e_l"(ComponentPose=@n114, Alpha=1, Translation=Get floatA(), Rotation="0,0,0", Scale="1,1,1") {
    rotationMode=BMM_Additive
    rotationSpace=BCS_BoneSpace
    scaleMode=BMM_Additive
    scaleSpace=BCS_BoneSpace
    translationMode=BMM_Additive
    translationSpace=BCS_BoneSpace
    alphaBoolBlend {blendOption=Linear}
    boneToModify {boneName=dyn_shard_shoulder_e_l}
   }
   n116 = "Transform (Modify) Bone - Bone: dyn_shard_shoulder_f_l"(ComponentPose=@n115, Alpha=1, Translation=Get floatB(), Rotation="0,0,0", Scale="1,1,1") {
    rotationMode=BMM_Additive
    rotationSpace=BCS_BoneSpace
    scaleMode=BMM_Additive
    scaleSpace=BCS_BoneSpace
    translationMode=BMM_Additive
    translationSpace=BCS_BoneSpace
    alphaBoolBlend {blendOption=Linear}
    boneToModify {boneName=dyn_shard_shoulder_f_l}
   }
   n117 = "Transform (Modify) Bone - Bone: dyn_shard_shoulder_g_l"(ComponentPose=@n116, Alpha=0.5, Translation=Get floatA(), Rotation="0,0,0", Scale="1,1,1") {
    rotationMode=BMM_Additive
    rotationSpace=BCS_BoneSpace
    scaleMode=BMM_Additive
    scaleSpace=BCS_BoneSpace
    translationMode=BMM_Additive
    translationSpace=BCS_BoneSpace
    alphaBoolBlend {blendOption=Linear}
    boneToModify {boneName=dyn_shard_shoulder_g_l}
   }
   n118 = "Transform (Modify) Bone - Bone: dyn_shard_shoulder_h_l"(ComponentPose=@n117, Alpha=1, Translation=Get floatB(), Rotation="0,0,0", Scale="1,1,1") {
    rotationMode=BMM_Additive
    rotationSpace=BCS_BoneSpace
    scaleMode=BMM_Additive
    scaleSpace=BCS_BoneSpace
    translationMode=BMM_Additive
    translationSpace=BCS_BoneSpace
    alphaBoolBlend {blendOption=Linear}
    boneToModify {boneName=dyn_shard_shoulder_h_l}
   }
   Save cached pose 'CrystalCachedPose'(Pose=Component To Local(ComponentPose=@n118)) {cacheName=CrystalCachedPose}
  ]}
 EventGraph {nodes [
   Event BlueprintUpdateAnimation()
   > n1 = Sequence!(execute) -> (then_0, then_1, then_2:Exec)
   > n2 = Gravity Override State Function DEPRECATED(
    CharacterPartAnimInstance=Self-Reference()
    GameTimeInSeconds=GetGameTimeInSeconds()
    GravityOverrideParams=Get GravityOverrideParams() -> (GravityOverrideParams:Gravity Override Params Struct Structure)
   )
   > Set GravityOverride(GravityOverride=@n2)
   Set floatA(
    execute=@n1.then_1
    floatA:Vector=MakeWaveVector!(Amplitudes:Rotator="1,0.5,1.75", Frequencies:Rotator="100,50,80", Offsets:Rotator="0,0,0", bUseNoise:bool=false, Min:Vector, Max:Vector) -> (ReturnValue:Vector)
   ) -> (Output_Get:Vector)
   > Set floatB(floatB:Vector=MakeWaveVector(Amplitudes="1,0.5,1.75", Frequencies="100,50,80", Offsets="-45,-45,90", bUseNoise=false)) -> (Output_Get:Vector)
   n3 = Set isPistol(
    execute=@n1.then_2
    isPistol="Equal (Enum)"(A:EFortWeaponCoreAnimation Enum=Get WeaponCoreAnim!() -> (WeaponCoreAnim), B:EFortWeaponCoreAnimation Enum=Pistol)
   )
   > Set Pistol(Pistol:double="To Float (Boolean)"(InBool=@n3)) -> (Output_Get:double)
   > n4 = Branch(Condition=OR Boolean!(A=Get bIsCrouching(), B=Get bIsMoving2D(), C:bool=Get bIsJumping(), D:bool=Get bIsFalling()) -> (ReturnValue))
   > Set skirtMidJntConstraintAlpha(skirtMidJntConstraintAlpha:double=1) -> (Output_Get:double)
   Set skirtMidJntConstraintAlpha(execute=@n4.else, skirtMidJntConstraintAlpha=0)
   Event BlueprintInitializeAnimation()
   > Set Body_Mesh(Body_Mesh=GetOwnerSkeletalMeshComponent())
   > n5 = Cast To PlayerPawn_Athena(Object=TryGetPawnOwner())
   > Set Pawn Owner(Pawn Owner:Player Pawn Athena Object Reference=@n5) -> (Output_Get:Player Pawn Athena Object Reference)
  ]}
]
variables [
 Body_Mesh {type=SkeletalMeshComponent}
 floatA {type=Vector}
 floatB {type=Vector}
 GravityOverride {type=Vector}
 GravityOverrideParams {type=GravityOverrideParamsStruct}
 isPistol {type=boolean}
 Pawn Owner {type=PlayerPawn_Athena_C}
 Pistol {type=number}
 skirtMidJntConstraintAlpha {type=number}
]
